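; Athena rift portal actor that teleports players, vehicles and AI pawns to a generated exit location, offsetting later arrivals.
; Overlap events run eligibility checks, cancel abilities, suppress fall damage, teleport after a delay, and update quests outside Creative.
; The first successful teleport starts a lifetime timer and opening FX; expiry destroys the rift, or resets it if permanent.
; Proximity overlaps widen rift visuals, and child classes can override success and rejection hooks.
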(event Custom|DestroyRift
  ;[1 1 1] On Death
  (CallFunction|DestroyRiftInternal))

(event Custom|DestroyRift_Internal
  (bind _audio_looping (Variables|Sound|GetAudio_Looping))
  ;[1 1 1] On Death
  (Variables|RiftPortal|SetPortalOpen false)
  (if (not (Variables|Default|GetPermanentRift))
    (Utilities|FlowControl|Sequence
      (:then_0
        (Networking|CanExecuteCosmeticEvents
          (:True
            (Development|SetHiddeninGame (Variables|Default|GetSM_Rift) true false)
            (Audio|PlaySoundatLocation (Variables|Sound|GetRiftDestroyed) (Transformation|GetActorLocation)) ; Play Deactivate Sound
            (bind (_direction _pitchcurve) (Timeline|Audio%20RiftDestroyed Play)
              (:Update
                (Audio|Components|Audio|SetPitchMultiplier _audio_looping _pitchcurve))
              (:Finished
                (Audio|Components|Audio|FadeOut _audio_looping 1 0.0 "Linear" :!class /Script/Engine.AudioComponent))))))
      (:then_1
        (Utilities|FlowControl|Delay (+ (ScalableFloat|GetValueatLevel (Variables|RiftPortal|GetHangtime)) 1.0))
        (Actor|DestroyActor)))
    (else
      (CallFunction|ResetRift)
      (Variables|RiftPortal|SetPortalOpen true))))

(event EventBeginPlay
  ;[1 1 1] EVENT--> BeginPlay
  (|CallParentFunction)
  (Variables|RiftPortal|SetPortalOpen true)
  ;[1 1 1] Should we spawn Rifts?
  (if (or (> (ScalableFloat|GetValueatLevel (Variables|Default|GetShouldSpawnRift?)) 0.0) (Variables|Default|GetPermanentRift))
    ;[1 1 1] Check if In Creative Mode or not for Quest
    (bind _asfort_game_state_athena (Utilities|Casting|CastToFortGameStateAthena (Game|GetGameState)))
    (if (Variables|/Script/FortniteGame.FortGameStateAthena|GetbPlaylistDataIsLoaded _asfort_game_state_athena)
      (bind _cancompletequest (Class|AthenaQuestFunctionLibrary|AthenaCheckCreativeMode (Game|GetGameState)))
      (if _cancompletequest
        (Variables|QuestStuff|SetCreativeMode false)
        (else
          (Variables|QuestStuff|SetCreativeMode true)))
      (else
        (Playlist|BindEventtoOnCurrentPlaylistLoaded _asfort_game_state_athena (delegate Custom|PlaylistLoaded))))
    (else
      (Actor|DestroyActor))))

(event OnComponentBeginOverlap(Capsule)
  (params (OverlappedComponent :type Primitive%20Component%20Object%20Reference) (OtherActor :type Actor%20Object%20Reference) (OtherComp :type Primitive%20Component%20Object%20Reference) (OtherBodyIndex :type Integer) (bFromSweep :type Boolean) (SweepResult :type Hit%20Result%20Structure%20%28by%20ref%29))
  (bind _sm_rift (Variables|Default|GetSM_Rift))
  ;[1 1 1] On Overlap - Toggle Post Process
  (Utilities|Casting|CastToFortPlayerPawn OtherActor)
  (CallFunction|PlayExpandorCollapseAudio true)
  (if (CallFunction|RiftVFXEnabled)
    ;;[1 1 1] Widen rift if player is close
    (bind (_direction _open) (Timeline|VectorScale Play)
      (:Update
        (Rendering|Material|SetScalarParameterValueonMaterials _sm_rift "Open" (Math|Float|SelectFloat 0.0 _open (Components|Timeline|IsPlaying (Variables|BGA_RiftPortal_Athena|GetFragmentMovement))))
        (Transformation|SetRelativeRotation _sm_rift (Math|Rotator|MakeRotator (* _open -2e+01)) false false)
        (Effects|Components|ParticleSystem|SetFloatParameter (Variables|Default|GetNS_Rift_Portal_Idle_Shimmer) "Open" _open)))))

(event Custom|PlaylistLoaded
  (params (PlaylistName :type Name) (PlaylistContextTags :type Gameplay%20Tag%20Container%20Structure%20%28by%20ref%29))
  ;[1 1 1] Check if In Creative Mode or not for Quest
  (bind _cancompletequest (Class|AthenaQuestFunctionLibrary|AthenaCheckCreativeMode (Game|GetGameState)))
  (if _cancompletequest
    (Variables|QuestStuff|SetCreativeMode false)
    (else
      (Variables|QuestStuff|SetCreativeMode true))))

(event OnComponentEndOverlap(Capsule)
  (params (OverlappedComponent :type Primitive%20Component%20Object%20Reference) (OtherActor :type Actor%20Object%20Reference) (OtherComp :type Primitive%20Component%20Object%20Reference) (OtherBodyIndex :type Integer))
  (bind _sm_rift (Variables|Default|GetSM_Rift))
  ;[1 1 1] On Overlap - Toggle Post Process
  (if (<= (Utilities|Array|Length (Collision|GetOverlappingActors OtherActor "/Script/FortniteGame.FortPlayerPawn")) 0)
    (CallFunction|PlayExpandorCollapseAudio false)
    (if (CallFunction|RiftVFXEnabled)
      ;;[1 1 1] Widen rift if player is close
      (Timeline|VectorScale Reverse))))

(event Custom|PlayTeleportFX
  (params (Location :type Vector))
  ;[1 1 1] Spawn teleport visuals and audio for clients
  (Utilities|FlowControl|DoOnce false
    (:Completed
      (bind _returnvalue (Effects|Components|ParticleSystem|SpawnFXAtLocation (Variables|VFX|GetEnterRift_NS) Location))
      (if (Variables|VFX|GetIS_Color%20Variant)
        (bind _asniagara_particle_system_component (Utilities|Casting|CastToNiagaraComponent _returnvalue))
        (Niagara|NiagaraSetColorArray _asniagara_particle_system_component "Color Array" (Variables|VFX|GetNS%20Color%20Array%20Data)))
      (if (not (Utilities|Equal(Object) (Variables|RiftPortal|GetOverlapActor) (Game|GetPlayerPawn 0)))
        (Audio|PlaySoundatLocation (Variables|Sound|GetEnteredRift) Location)
        (Utilities|FlowControl|Delay 1.0)
        (Audio|SpawnSoundatLocation (Variables|Sound|GetPlayerExitSkyCrack) (Variables|RiftPortal|GetTeleportLocation))))))

(event Custom|ResetFX
  ;[1 1 1] Spawn teleport visuals and audio for clients
  (Utilities|FlowControl|DoOnce false
    (:Completed
      (bind _returnvalue (Effects|Components|ParticleSystem|SpawnFXAtLocation (Variables|VFX|GetEnterRift_NS) (stale Custom|PlayTeleportFX Location)))
      (if (Variables|VFX|GetIS_Color%20Variant)
        (bind _asniagara_particle_system_component (Utilities|Casting|CastToNiagaraComponent _returnvalue))
        (Niagara|NiagaraSetColorArray _asniagara_particle_system_component "Color Array" (Variables|VFX|GetNS%20Color%20Array%20Data)))
      (if (not (Utilities|Equal(Object) (Variables|RiftPortal|GetOverlapActor) (Game|GetPlayerPawn 0)))
        (Audio|PlaySoundatLocation (Variables|Sound|GetEnteredRift) (stale Custom|PlayTeleportFX Location))
        (Utilities|FlowControl|Delay 1.0)
        (Audio|SpawnSoundatLocation (Variables|Sound|GetPlayerExitSkyCrack) (Variables|RiftPortal|GetTeleportLocation))))))

(event Custom|PortalOpened
  (bind (_returnvalue_location _returnvalue_rotation _returnvalue_scale) (Math|Transform|BreakTransform (Transformation|GetWorldTransform (Variables|Default|GetSM_Rift))))
  (bind _sm_rift (Variables|Default|GetSM_Rift))
  ;[1 1 1] Toggle FX and animate shards after opening
  (Utilities|FlowControl|Sequence
    (:then_0
      (if (CallFunction|RiftVFXEnabled)
        (if (not (Variables|Default|GetPermanentRift))
          (Components|Activation|SetActive (Variables|Default|GetNS_Rift_Portal_Idle_Shimmer) false false)
          (Niagara|SpawnSystematLocation "/Game/Athena/Items/ForagedItems/Rift/FX/NS_Rift_Portal_Closing.NS_Rift_Portal_Closing" _returnvalue_location _returnvalue_rotation _returnvalue_scale) ; Rift Closing Niagara System
        )
        (bind (_direction _open _glow) (Timeline|FragmentMovement Play)
          (:Update
            (Rendering|Material|SetScalarParameterValueonMaterials _sm_rift "OpenAdd" _open)
            (Rendering|Material|SetScalarParameterValueonMaterials _sm_rift "Emissive" _glow)))))
    (:then_1
      (Utilities|FlowControl|Delay 1.0)
      (bind _returnvalue (Effects|Components|ParticleSystem|SpawnFXAtLocation (Variables|VFX|GetCrackExitRift) (+ (Variables|RiftPortal|GetTeleportLocation) "0, 0,200.000000") (Math|Rotator|MakeRotator 0.0 0.0 (Math|Random|RandomFloatinRange -1.8e+02 1.8e+02)) "1.000000,1.000000,1.000000" true "None" true))
      (LOD|SetMaxDrawDistance _returnvalue 0.0))))

(event Custom|TryRiftTeleport
  (params (TargetActor :type Actor%20Object%20Reference))
  (bind _playerpawn (Variables|RiftPortal|GetPlayerPawn))
  (bind _returnvalue (Vehicle|GetVehicleInterface _playerpawn))
  (bind _playerpawn_2 (Variables|RiftPortal|GetPlayerPawn))
  (bind _returnvalue_2 (Vehicle|GetVehicleInterface _playerpawn_2))
  (bind _playerpawn_3 (Variables|RiftPortal|GetPlayerPawn))
  (bind _overlapactor (Variables|RiftPortal|GetOverlapActor))
  (bind _playerpawn_4 (Variables|RiftPortal|GetPlayerPawn))
  (bind _playerpawn_5 (Variables|RiftPortal|GetPlayerPawn))
  (bind (_actorrotation_roll _actorrotation_pitch _actorrotation_yaw) (Math|Rotator|BreakRotator (Variables|GetActorRotation_Roll)))
  (bind _fortvehicleinterface (Variables|Default|GetFortVehicleInterface))
  ;[1 1 1] Teleport Player and/or Vehicle
  (Variables|RiftPortal|SetOverlapActor TargetActor)
  (if (Variables|RiftPortal|GetPortalOpen)
    (Variables|RiftPortal|SetActorRotation (Transformation|GetActorRotation (Variables|RiftPortal|GetOverlapActor)))
    (Utilities|FlowControl|Sequence
      (:then_0
        (Utilities|FlowControl|DoOnce false
          (:Completed
            (Variables|RiftPortal|SetTeleportLocation (CallFunction|GenerateTeleportLocation)))))
      (:then_1
        (CallFunction|ResetFX)
        (Utilities|IsValid _overlapactor
          (:Is%20Valid
            (bind _asfort_player_pawn (Utilities|Casting|CastToFortPlayerPawn _overlapactor)
              (:then
                (if (CallFunction|CanPawnbeTeleported _asfort_player_pawn)
                  (bind _output_get (Variables|RiftPortal|SetPlayerPawn _asfort_player_pawn))
                  (CallFunction|BlockInteractionsWIthDisguises _output_get)
                  ;;[1 1 1] Quest for Different Location Hook In
                  (if (not (Variables|QuestStuff|GetCreativeMode))
                    (CallFunction|CheckforRiftLocUpdate (Pawn|GetController _output_get)))
                  (if (Networking|IsServer)
                    (CallFunction|PlayTeleportFX (Transformation|GetActorLocation (Variables|RiftPortal|GetPlayerPawn))))
                  (if (Vehicle|IsinVehicle (Variables|RiftPortal|GetPlayerPawn))
                    (Utilities|IsValid (Utilities|ToObject(Interface) _returnvalue)
                      (:Is%20Valid
                        (if (Vehicle|CanTeleport _returnvalue)
                          (if (not (Utilities|IsValid (Utilities|ToObject(Interface) _returnvalue_2) :!class /Script/Engine.KismetSystemLibrary))
                            (Variables|Default|SetFortVehicleInterface _returnvalue_2)
                            (CallFunction|VehicleTelportEvent))
                          (else
                            (GameplayEffects|RemoveActiveEffectswithGrantedTags (Ability|GetAbilitySystemComponent (Variables|RiftPortal|GetPlayerPawn)) "(GameplayTags=((TagName=\"Gameplay.Action.Player.AthenaInVehicle\")))") ; If vehicle isn't supposed to teleport then kick the player from it
                          )))))
                  (Networking|SwitchHasAuthority
                    (:Authority
                      (CallFunction|CancelPlayerPawnAbilities)
                      (RemoteControlPawn|ForceDetachfromRemoteControlledPawn _playerpawn_5)
                      (Variables|/Script/FortniteGame.FortPawn|SetbIgnoreNextFallingDamage true _playerpawn_5) ; Have to do this because of the delay. Otherwise player can die from fall damage if they fall from high distance.
                      (Utilities|FlowControl|Delay (Variables|RiftPortal|GetTeleportDelay))
                      (bind _returnvalue_3 (CallFunction|IsFlying))
                      (if (not _returnvalue_3)
                        (Parachute|EndSkydiving _playerpawn_4) ; End Skydiving if player was previously
                        (Transform|Actor|TeleportPlayerPawn _playerpawn_4 (Variables|RiftPortal|GetTeleportLocation) (Math|Rotator|MakeRotator 0.0 0.0 _actorrotation_yaw) true false)
                        (CallFunction|OnRiftTeleportSucceed (Variables|RiftPortal|GetPlayerPawn))
                        (Variables|RiftPortal|SetTeleportLocation (+ (+ (Variables|RiftPortal|GetTeleportLocation) "0,50.000000, 0") (Math|Vector|MakeVector (Math|Random|RandomFloatinRange (* (Variables|Default|GetRandomRange) -1.0) (Variables|Default|GetRandomRange)) (Math|Random|RandomFloatinRange (* (Variables|Default|GetRandomRange) -1.0) (Variables|Default|GetRandomRange))))) ; Offset players so they don't all teleport on top of each other
                        (Utilities|FlowControl|Sequence
                          (:then_0
                            ;[1 1 1] SERVER: Check to Award Single Use Quest Progression if not in Creative Mode
                            (if (not (Variables|QuestStuff|GetCreativeMode))
                              (Networking|SwitchHasAuthority
                                (:Authority
                                  (Development|PrintString "QUEST GOES HERE!" false false "(R=0.331582,G=1.000000,B=0.000000,A=1.000000)" 5.0)
                                  (Utilities|IsValid (Variables|RiftPortal|GetPlayerPawn)
                                    (:Is%20Valid
                                      (CallFunction|UseRiftQuestUpdate (Pawn|GetController _playerpawn_3))
                                      (CallFunction|QuestUpdateComplexCustom _playerpawn_3))
                                    (:Is%20Not%20Valid
                                      ;;[1 1 1] For AthenaVehicle - fire Complex Custom and Rift Quest update events once for each player in the vehicle
                                      (Utilities|IsValid (Utilities|ToObject(Interface) _fortvehicleinterface)
                                        (:Is%20Valid
                                          (if (Vehicle|ContainsPlayers _fortvehicleinterface)
                                            (for _array_element (Vehicle|GetAllPassengers _fortvehicleinterface)
                                              (Utilities|IsValid _array_element
                                                (:Is%20Valid
                                                  (CallFunction|UseRiftQuestUpdate (Pawn|GetController _array_element))
                                                  (CallFunction|QuestUpdateComplexCustom _array_element)))))))))))))
                          (:then_1
                            (Utilities|FlowControl|DoOnce false
                              (:Completed
                                (Utilities|Time|SetTimerbyEvent (EventDispatchers|CreateEvent :!event Custom|DestroyRift) (ScalableFloat|GetValueatLevel (Variables|RiftPortal|GetActiveDuration)))
                                (CallFunction|PortalOpened))))))))
                  (else
                    (CallFunction|OnRiftTeleportRejected _asfort_player_pawn))))
              (:CastFailed
                (bind _asfort_vehicle_interface (Utilities|Casting|CastToFortVehicleInterface (Variables|RiftPortal|GetOverlapActor))
                  (:then
                    (Variables|Default|SetFortVehicleInterface _asfort_vehicle_interface)
                    (CallFunction|VehicleTelportEvent))
                  (:CastFailed
                    (bind _asfort_aipawn (Utilities|Casting|CastToFortAIPawn (Variables|RiftPortal|GetOverlapActor)))
                    (Utilities|IsValid (.AttachedPlayerPawn (Default|IsPlayerAttachedtoAIPawn _asfort_aipawn))
                      (:Is%20Valid
                        (CallFunction|TryRiftTeleport (.AttachedPlayerPawn (Default|IsPlayerAttachedtoAIPawn _asfort_aipawn))))
                      (:Is%20Not%20Valid
                        _))
                    (CallFunction|PlayTeleportFX (Transformation|GetActorLocation _asfort_aipawn))
                    (AI|TeleportAIPawn _asfort_aipawn (Math|Vector|vector+vector (Variables|RiftPortal|GetTeleportLocation) (1e+02 1e+02 0.0)) "0, 0, 0" true)
                    (Variables|/Script/FortniteGame.FortPawn|SetbIgnoreNextFallingDamage true _asfort_aipawn)
                    (CallFunction|OnRiftTeleportSucceed _asfort_aipawn)))))))))))

(event EventOverlappingOnBeginPlay
  (params (Actor :type Actor%20Object%20Reference))
  (bind _playerpawn (Variables|RiftPortal|GetPlayerPawn))
  (bind _returnvalue (Vehicle|GetVehicleInterface _playerpawn))
  (bind _playerpawn_2 (Variables|RiftPortal|GetPlayerPawn))
  (bind _returnvalue_2 (Vehicle|GetVehicleInterface _playerpawn_2))
  (bind _playerpawn_3 (Variables|RiftPortal|GetPlayerPawn))
  (bind _overlapactor (Variables|RiftPortal|GetOverlapActor))
  (bind _playerpawn_4 (Variables|RiftPortal|GetPlayerPawn))
  (bind _playerpawn_5 (Variables|RiftPortal|GetPlayerPawn))
  (bind (_actorrotation_roll _actorrotation_pitch _actorrotation_yaw) (Math|Rotator|BreakRotator (Variables|GetActorRotation_Roll)))
  (bind _fortvehicleinterface (Variables|Default|GetFortVehicleInterface))
  ;[1 1 1] Teleport Player and/or Vehicle
  (Variables|RiftPortal|SetOverlapActor Actor)
  (if (Variables|RiftPortal|GetPortalOpen)
    (Variables|RiftPortal|SetActorRotation (Transformation|GetActorRotation (Variables|RiftPortal|GetOverlapActor)))
    (Utilities|FlowControl|Sequence
      (:then_0
        (Utilities|FlowControl|DoOnce false
          (:Completed
            (Variables|RiftPortal|SetTeleportLocation (CallFunction|GenerateTeleportLocation)))))
      (:then_1
        (CallFunction|ResetFX)
        (Utilities|IsValid _overlapactor
          (:Is%20Valid
            (bind _asfort_player_pawn (Utilities|Casting|CastToFortPlayerPawn _overlapactor)
              (:then
                (if (CallFunction|CanPawnbeTeleported _asfort_player_pawn)
                  (bind _output_get (Variables|RiftPortal|SetPlayerPawn _asfort_player_pawn))
                  (CallFunction|BlockInteractionsWIthDisguises _output_get)
                  ;;[1 1 1] Quest for Different Location Hook In
                  (if (not (Variables|QuestStuff|GetCreativeMode))
                    (CallFunction|CheckforRiftLocUpdate (Pawn|GetController _output_get)))
                  (if (Networking|IsServer)
                    (CallFunction|PlayTeleportFX (Transformation|GetActorLocation (Variables|RiftPortal|GetPlayerPawn))))
                  (if (Vehicle|IsinVehicle (Variables|RiftPortal|GetPlayerPawn))
                    (Utilities|IsValid (Utilities|ToObject(Interface) _returnvalue)
                      (:Is%20Valid
                        (if (Vehicle|CanTeleport _returnvalue)
                          (if (not (Utilities|IsValid (Utilities|ToObject(Interface) _returnvalue_2) :!class /Script/Engine.KismetSystemLibrary))
                            (Variables|Default|SetFortVehicleInterface _returnvalue_2)
                            (CallFunction|VehicleTelportEvent))
                          (else
                            (GameplayEffects|RemoveActiveEffectswithGrantedTags (Ability|GetAbilitySystemComponent (Variables|RiftPortal|GetPlayerPawn)) "(GameplayTags=((TagName=\"Gameplay.Action.Player.AthenaInVehicle\")))") ; If vehicle isn't supposed to teleport then kick the player from it
                          )))))
                  (Networking|SwitchHasAuthority
                    (:Authority
                      (CallFunction|CancelPlayerPawnAbilities)
                      (RemoteControlPawn|ForceDetachfromRemoteControlledPawn _playerpawn_5)
                      (Variables|/Script/FortniteGame.FortPawn|SetbIgnoreNextFallingDamage true _playerpawn_5) ; Have to do this because of the delay. Otherwise player can die from fall damage if they fall from high distance.
                      (Utilities|FlowControl|Delay (Variables|RiftPortal|GetTeleportDelay))
                      (bind _returnvalue_3 (CallFunction|IsFlying))
                      (if (not _returnvalue_3)
                        (Parachute|EndSkydiving _playerpawn_4) ; End Skydiving if player was previously
                        (Transform|Actor|TeleportPlayerPawn _playerpawn_4 (Variables|RiftPortal|GetTeleportLocation) (Math|Rotator|MakeRotator 0.0 0.0 _actorrotation_yaw) true false)
                        (CallFunction|OnRiftTeleportSucceed (Variables|RiftPortal|GetPlayerPawn))
                        (Variables|RiftPortal|SetTeleportLocation (+ (+ (Variables|RiftPortal|GetTeleportLocation) "0,50.000000, 0") (Math|Vector|MakeVector (Math|Random|RandomFloatinRange (* (Variables|Default|GetRandomRange) -1.0) (Variables|Default|GetRandomRange)) (Math|Random|RandomFloatinRange (* (Variables|Default|GetRandomRange) -1.0) (Variables|Default|GetRandomRange))))) ; Offset players so they don't all teleport on top of each other
                        (Utilities|FlowControl|Sequence
                          (:then_0
                            ;[1 1 1] SERVER: Check to Award Single Use Quest Progression if not in Creative Mode
                            (if (not (Variables|QuestStuff|GetCreativeMode))
                              (Networking|SwitchHasAuthority
                                (:Authority
                                  (Development|PrintString "QUEST GOES HERE!" false false "(R=0.331582,G=1.000000,B=0.000000,A=1.000000)" 5.0)
                                  (Utilities|IsValid (Variables|RiftPortal|GetPlayerPawn)
                                    (:Is%20Valid
                                      (CallFunction|UseRiftQuestUpdate (Pawn|GetController _playerpawn_3))
                                      (CallFunction|QuestUpdateComplexCustom _playerpawn_3))
                                    (:Is%20Not%20Valid
                                      ;;[1 1 1] For AthenaVehicle - fire Complex Custom and Rift Quest update events once for each player in the vehicle
                                      (Utilities|IsValid (Utilities|ToObject(Interface) _fortvehicleinterface)
                                        (:Is%20Valid
                                          (if (Vehicle|ContainsPlayers _fortvehicleinterface)
                                            (for _array_element (Vehicle|GetAllPassengers _fortvehicleinterface)
                                              (Utilities|IsValid _array_element
                                                (:Is%20Valid
                                                  (CallFunction|UseRiftQuestUpdate (Pawn|GetController _array_element))
                                                  (CallFunction|QuestUpdateComplexCustom _array_element)))))))))))))
                          (:then_1
                            (Utilities|FlowControl|DoOnce false
                              (:Completed
                                (Utilities|Time|SetTimerbyEvent (EventDispatchers|CreateEvent :!event Custom|DestroyRift) (ScalableFloat|GetValueatLevel (Variables|RiftPortal|GetActiveDuration)))
                                (CallFunction|PortalOpened))))))))
                  (else
                    (CallFunction|OnRiftTeleportRejected _asfort_player_pawn))))
              (:CastFailed
                (bind _asfort_vehicle_interface (Utilities|Casting|CastToFortVehicleInterface (Variables|RiftPortal|GetOverlapActor))
                  (:then
                    (Variables|Default|SetFortVehicleInterface _asfort_vehicle_interface)
                    (CallFunction|VehicleTelportEvent))
                  (:CastFailed
                    (bind _asfort_aipawn (Utilities|Casting|CastToFortAIPawn (Variables|RiftPortal|GetOverlapActor)))
                    (Utilities|IsValid (.AttachedPlayerPawn (Default|IsPlayerAttachedtoAIPawn _asfort_aipawn))
                      (:Is%20Valid
                        (CallFunction|TryRiftTeleport (.AttachedPlayerPawn (Default|IsPlayerAttachedtoAIPawn _asfort_aipawn))))
                      (:Is%20Not%20Valid
                        _))
                    (CallFunction|PlayTeleportFX (Transformation|GetActorLocation _asfort_aipawn))
                    (AI|TeleportAIPawn _asfort_aipawn (Math|Vector|vector+vector (Variables|RiftPortal|GetTeleportLocation) (1e+02 1e+02 0.0)) "0, 0, 0" true)
                    (Variables|/Script/FortniteGame.FortPawn|SetbIgnoreNextFallingDamage true _asfort_aipawn)
                    (CallFunction|OnRiftTeleportSucceed _asfort_aipawn)))))))))))

(event OnComponentBeginOverlap(OverlapCapsule)
  (params (OverlappedComponent :type Primitive%20Component%20Object%20Reference) (OtherActor :type Actor%20Object%20Reference) (OtherComp :type Primitive%20Component%20Object%20Reference) (OtherBodyIndex :type Integer) (bFromSweep :type Boolean) (SweepResult :type Hit%20Result%20Structure%20%28by%20ref%29))
  (bind _playerpawn (Variables|RiftPortal|GetPlayerPawn))
  (bind _returnvalue (Vehicle|GetVehicleInterface _playerpawn))
  (bind _playerpawn_2 (Variables|RiftPortal|GetPlayerPawn))
  (bind _returnvalue_2 (Vehicle|GetVehicleInterface _playerpawn_2))
  (bind _playerpawn_3 (Variables|RiftPortal|GetPlayerPawn))
  (bind _overlapactor (Variables|RiftPortal|GetOverlapActor))
  (bind _playerpawn_4 (Variables|RiftPortal|GetPlayerPawn))
  (bind _playerpawn_5 (Variables|RiftPortal|GetPlayerPawn))
  (bind (_actorrotation_roll _actorrotation_pitch _actorrotation_yaw) (Math|Rotator|BreakRotator (Variables|GetActorRotation_Roll)))
  (bind _fortvehicleinterface (Variables|Default|GetFortVehicleInterface))
  ;[1 1 1] Teleport Player and/or Vehicle
  (Variables|RiftPortal|SetOverlapActor OtherActor)
  (if (Variables|RiftPortal|GetPortalOpen)
    (Variables|RiftPortal|SetActorRotation (Transformation|GetActorRotation (Variables|RiftPortal|GetOverlapActor)))
    (Utilities|FlowControl|Sequence
      (:then_0
        (Utilities|FlowControl|DoOnce false
          (:Completed
            (Variables|RiftPortal|SetTeleportLocation (CallFunction|GenerateTeleportLocation)))))
      (:then_1
        (CallFunction|ResetFX)
        (Utilities|IsValid _overlapactor
          (:Is%20Valid
            (bind _asfort_player_pawn (Utilities|Casting|CastToFortPlayerPawn _overlapactor)
              (:then
                (if (CallFunction|CanPawnbeTeleported _asfort_player_pawn)
                  (bind _output_get (Variables|RiftPortal|SetPlayerPawn _asfort_player_pawn))
                  (CallFunction|BlockInteractionsWIthDisguises _output_get)
                  ;;[1 1 1] Quest for Different Location Hook In
                  (if (not (Variables|QuestStuff|GetCreativeMode))
                    (CallFunction|CheckforRiftLocUpdate (Pawn|GetController _output_get)))
                  (if (Networking|IsServer)
                    (CallFunction|PlayTeleportFX (Transformation|GetActorLocation (Variables|RiftPortal|GetPlayerPawn))))
                  (if (Vehicle|IsinVehicle (Variables|RiftPortal|GetPlayerPawn))
                    (Utilities|IsValid (Utilities|ToObject(Interface) _returnvalue)
                      (:Is%20Valid
                        (if (Vehicle|CanTeleport _returnvalue)
                          (if (not (Utilities|IsValid (Utilities|ToObject(Interface) _returnvalue_2) :!class /Script/Engine.KismetSystemLibrary))
                            (Variables|Default|SetFortVehicleInterface _returnvalue_2)
                            (CallFunction|VehicleTelportEvent))
                          (else
                            (GameplayEffects|RemoveActiveEffectswithGrantedTags (Ability|GetAbilitySystemComponent (Variables|RiftPortal|GetPlayerPawn)) "(GameplayTags=((TagName=\"Gameplay.Action.Player.AthenaInVehicle\")))") ; If vehicle isn't supposed to teleport then kick the player from it
                          )))))
                  (Networking|SwitchHasAuthority
                    (:Authority
                      (CallFunction|CancelPlayerPawnAbilities)
                      (RemoteControlPawn|ForceDetachfromRemoteControlledPawn _playerpawn_5)
                      (Variables|/Script/FortniteGame.FortPawn|SetbIgnoreNextFallingDamage true _playerpawn_5) ; Have to do this because of the delay. Otherwise player can die from fall damage if they fall from high distance.
                      (Utilities|FlowControl|Delay (Variables|RiftPortal|GetTeleportDelay))
                      (bind _returnvalue_3 (CallFunction|IsFlying))
                      (if (not _returnvalue_3)
                        (Parachute|EndSkydiving _playerpawn_4) ; End Skydiving if player was previously
                        (Transform|Actor|TeleportPlayerPawn _playerpawn_4 (Variables|RiftPortal|GetTeleportLocation) (Math|Rotator|MakeRotator 0.0 0.0 _actorrotation_yaw) true false)
                        (CallFunction|OnRiftTeleportSucceed (Variables|RiftPortal|GetPlayerPawn))
                        (Variables|RiftPortal|SetTeleportLocation (+ (+ (Variables|RiftPortal|GetTeleportLocation) "0,50.000000, 0") (Math|Vector|MakeVector (Math|Random|RandomFloatinRange (* (Variables|Default|GetRandomRange) -1.0) (Variables|Default|GetRandomRange)) (Math|Random|RandomFloatinRange (* (Variables|Default|GetRandomRange) -1.0) (Variables|Default|GetRandomRange))))) ; Offset players so they don't all teleport on top of each other
                        (Utilities|FlowControl|Sequence
                          (:then_0
                            ;[1 1 1] SERVER: Check to Award Single Use Quest Progression if not in Creative Mode
                            (if (not (Variables|QuestStuff|GetCreativeMode))
                              (Networking|SwitchHasAuthority
                                (:Authority
                                  (Development|PrintString "QUEST GOES HERE!" false false "(R=0.331582,G=1.000000,B=0.000000,A=1.000000)" 5.0)
                                  (Utilities|IsValid (Variables|RiftPortal|GetPlayerPawn)
                                    (:Is%20Valid
                                      (CallFunction|UseRiftQuestUpdate (Pawn|GetController _playerpawn_3))
                                      (CallFunction|QuestUpdateComplexCustom _playerpawn_3))
                                    (:Is%20Not%20Valid
                                      ;;[1 1 1] For AthenaVehicle - fire Complex Custom and Rift Quest update events once for each player in the vehicle
                                      (Utilities|IsValid (Utilities|ToObject(Interface) _fortvehicleinterface)
                                        (:Is%20Valid
                                          (if (Vehicle|ContainsPlayers _fortvehicleinterface)
                                            (for _array_element (Vehicle|GetAllPassengers _fortvehicleinterface)
                                              (Utilities|IsValid _array_element
                                                (:Is%20Valid
                                                  (CallFunction|UseRiftQuestUpdate (Pawn|GetController _array_element))
                                                  (CallFunction|QuestUpdateComplexCustom _array_element)))))))))))))
                          (:then_1
                            (Utilities|FlowControl|DoOnce false
                              (:Completed
                                (Utilities|Time|SetTimerbyEvent (EventDispatchers|CreateEvent :!event Custom|DestroyRift) (ScalableFloat|GetValueatLevel (Variables|RiftPortal|GetActiveDuration)))
                                (CallFunction|PortalOpened))))))))
                  (else
                    (CallFunction|OnRiftTeleportRejected _asfort_player_pawn))))
              (:CastFailed
                (bind _asfort_vehicle_interface (Utilities|Casting|CastToFortVehicleInterface (Variables|RiftPortal|GetOverlapActor))
                  (:then
                    (Variables|Default|SetFortVehicleInterface _asfort_vehicle_interface)
                    (CallFunction|VehicleTelportEvent))
                  (:CastFailed
                    (bind _asfort_aipawn (Utilities|Casting|CastToFortAIPawn (Variables|RiftPortal|GetOverlapActor)))
                    (Utilities|IsValid (.AttachedPlayerPawn (Default|IsPlayerAttachedtoAIPawn _asfort_aipawn))
                      (:Is%20Valid
                        (CallFunction|TryRiftTeleport (.AttachedPlayerPawn (Default|IsPlayerAttachedtoAIPawn _asfort_aipawn))))
                      (:Is%20Not%20Valid
                        _))
                    (CallFunction|PlayTeleportFX (Transformation|GetActorLocation _asfort_aipawn))
                    (AI|TeleportAIPawn _asfort_aipawn (Math|Vector|vector+vector (Variables|RiftPortal|GetTeleportLocation) (1e+02 1e+02 0.0)) "0, 0, 0" true)
                    (Variables|/Script/FortniteGame.FortPawn|SetbIgnoreNextFallingDamage true _asfort_aipawn)
                    (CallFunction|OnRiftTeleportSucceed _asfort_aipawn)))))))))))

(event Custom|ResetRift
  ;[1 1 1] Teleport Player and/or Vehicle
  (Utilities|FlowControl|DoOnce false
    (:Completed
      (Utilities|Time|SetTimerbyEvent (EventDispatchers|CreateEvent :!event Custom|DestroyRift) (ScalableFloat|GetValueatLevel (Variables|RiftPortal|GetActiveDuration)))
      (CallFunction|PortalOpened))))

(event Custom|Vehicle%20Telport%20Event
  (bind _returnvalue (Utilities|ToObject(Interface) (Variables|Default|GetFortVehicleInterface)))
  (bind _fortvehicleinterface (Variables|Default|GetFortVehicleInterface))
  (bind _returnvalue_2 (Vehicle|GetAllPassengers _fortvehicleinterface))
  (bind _fortvehicleinterface_2 (Variables|Default|GetFortVehicleInterface))
  (bind _playerpawn (Variables|RiftPortal|GetPlayerPawn))
  (bind _fortvehicleinterface_3 (Variables|Default|GetFortVehicleInterface))
  ;[1 1 1] Teleport Player and/or Vehicle
  (Utilities|IsValid (Utilities|ToObject(Interface) _fortvehicleinterface_2)
    (:Is%20Valid
      (if (Vehicle|CanTeleport _fortvehicleinterface_2)
        (CallFunction|PlayTeleportFX (Class|FortVehicleInterface|GetOwnerLocation _fortvehicleinterface_2))
        (Vehicle|SetIgnoreNextFallingDamage (Variables|Default|GetFortVehicleInterface) true)
        (Utilities|FlowControl|Delay (Variables|RiftPortal|GetTeleportDelay))
        (bind _returnvalue_3 (CallFunction|IsFlying))
        (if (not _returnvalue_3)
          (bind _asfort_player_pawn (Utilities|Casting|CastToFortPlayerPawn _returnvalue)
            (:then
              (Transformation|Teleport _asfort_player_pawn (Variables|RiftPortal|GetTeleportLocation) (Transformation|GetActorRotation)))
            (:CastFailed
              ;;[1 0.1349 0.1249] Swapping to teleport from teleport vehicle to work around an issue where teleport vehicle would fail (for a code reason I can't see here). This may cause fallout so passing to an egineer to review the teleport vehicle code 8/5/19
              (bind _asfort_athena_vehicle (Utilities|Casting|CastToFortAthenaVehicle _returnvalue))
              (CallFunction|TeleportVehicle _asfort_athena_vehicle (Variables|RiftPortal|GetTeleportLocation))))
          (Class|FortVehicleInterface|DisableTrickCredit _fortvehicleinterface 0.1)
          (CallFunction|OnRiftTeleportSucceed _asfort_athena_vehicle)
          (if (>= (Utilities|Array|Length _returnvalue_2) 1)
            (bind _output_get (Variables|RiftPortal|SetPassengerArray _returnvalue_2))
            (for _array_element _output_get
              (RemoteControlPawn|ForceDetachfromRemoteControlledPawn _array_element)
              (CallFunction|OnRiftTeleportSucceed _array_element)))
          (Variables|RiftPortal|SetTeleportLocation (+ (+ (Variables|RiftPortal|GetTeleportLocation) "0,50.000000, 0") (Math|Vector|MakeVector (Math|Random|RandomFloatinRange (* (Variables|Default|GetRandomRange) -1.0) (Variables|Default|GetRandomRange)) (Math|Random|RandomFloatinRange (* (Variables|Default|GetRandomRange) -1.0) (Variables|Default|GetRandomRange))))) ; Offset players so they don't all teleport on top of each other
          (Utilities|FlowControl|Sequence
            (:then_0
              ;[1 1 1] SERVER: Check to Award Single Use Quest Progression if not in Creative Mode
              (if (not (Variables|QuestStuff|GetCreativeMode))
                (Networking|SwitchHasAuthority
                  (:Authority
                    (Development|PrintString "QUEST GOES HERE!" false false "(R=0.331582,G=1.000000,B=0.000000,A=1.000000)" 5.0)
                    (Utilities|IsValid (Variables|RiftPortal|GetPlayerPawn)
                      (:Is%20Valid
                        (CallFunction|UseRiftQuestUpdate (Pawn|GetController _playerpawn))
                        (CallFunction|QuestUpdateComplexCustom _playerpawn))
                      (:Is%20Not%20Valid
                        ;;[1 1 1] For AthenaVehicle - fire Complex Custom and Rift Quest update events once for each player in the vehicle
                        (Utilities|IsValid (Utilities|ToObject(Interface) _fortvehicleinterface_3)
                          (:Is%20Valid
                            (if (Vehicle|ContainsPlayers _fortvehicleinterface_3)
                              (for _array_element_2 (Vehicle|GetAllPassengers _fortvehicleinterface_3)
                                (Utilities|IsValid _array_element_2
                                  (:Is%20Valid
                                    (CallFunction|UseRiftQuestUpdate (Pawn|GetController _array_element_2))
                                    (CallFunction|QuestUpdateComplexCustom _array_element_2)))))))))))))
            (:then_1
              (Utilities|FlowControl|DoOnce false
                (:Completed
                  (Utilities|Time|SetTimerbyEvent (EventDispatchers|CreateEvent :!event Custom|DestroyRift) (ScalableFloat|GetValueatLevel (Variables|RiftPortal|GetActiveDuration)))
                  (CallFunction|PortalOpened))))))
        (else
          (bind _asactor (Utilities|Casting|CastToActor _returnvalue))
          (CallFunction|OnRiftTeleportRejected _asactor))))))

(event Custom|OnRiftTeleportSucceed
  (params (Teleported%20Actor :type Actor%20Object%20Reference))
  ;[0.04688 0.04688 0.04688 0.5] Child classes can override to define success + failure case behavior
  (CallFunction|SendRiftTriggerEvent Teleported%20Actor))

(event Custom|OnRiftTeleportRejected
  (params (Rejected%20Actor :type Actor%20Object%20Reference)))

(fn BlockInteractionsWIthDisguises
  (params (Actor :type Actor%20Object%20Reference))
  (GameplayEffects|ApplyGameplayEffectToSelf (Ability|GetFortAbilitySystemComponent Actor) "/Game/Athena/Items/ForagedItems/Rift/GE_ProhibitDisguiseInteract.GE_ProhibitDisguiseInteract_C" 0.0 (GameplayEffects|MakeEffectContext (Ability|GetFortAbilitySystemComponent Actor))))

(fn CancelPlayerPawnAbilities
  (bind _playerpawn (Variables|RiftPortal|GetPlayerPawn)
    (:then
      (Utilities|IsValid (Ability|GetFortAbilitySystemComponent _playerpawn)
        (:Is%20Valid
          (Tags|CancelAbilitiesWithTags (Ability|GetFortAbilitySystemComponent _playerpawn) (Variables|Default|GetCancelAbilitiesTags)))))
    (:else)))

(fn CanPawnBeTeleported
  (params (InputPin :type Fort%20Player%20Pawn%20Object%20Reference))
  (return (Math|Boolean|ANDBoolean (not (Vehicle|IsinVehicle InputPin)) (not (Variables|/Script/FortniteGame.FortPawn|GetbIsDBNO InputPin)) (not (GameplayTags|HasAnyMatchingGameplayTags InputPin (Variables|Default|GetPawnBlockTeleportTags))))))

(fn CheckForRiftLocUpdate
  (params (Controller :type Controller%20Object%20Reference))
  (bind _objectivebackendname (Variables|QuestStuff|GetObjectiveBackendName))
  (bind _riftportallocationsquest (Variables|QuestStuff|GetRiftPortalLocationsQuest))
  (Utilities|IsValid Controller
    (:Is%20Valid
      (bind _asfort_player_controller (Utilities|Casting|CastToFortPlayerController Controller))
      (bind (_questisvalid _questcompleted) (Default|AthenaCheckQuestandObjectiveCompletion _asfort_player_controller _riftportallocationsquest _objectivebackendname))
      (if (and _questisvalid (not _questcompleted))
        (Class|AthenaQuestFunctionLibrary|AthenaGiveQuestUpdatetoPlayers :InQuestItemRef _riftportallocationsquest :ObjectiveBackendName _objectivebackendname :PlayerControllersForUpdate (Utilities|Array|MakeArray _asfort_player_controller))
        (Development|PrintString "SUCCESS!" true true "(R=0.416182,G=1.000000,B=0.000000,A=1.000000)" 1e+01)
        (else
          (Development|PrintString "Quest Not Valid" true true "(R=1.000000,G=0.066515,B=0.000000,A=1.000000)" 1e+01))))))

(fn GenerateTeleportLocation
  (bind _outxy (DataTable|EvaluateCurveTableRow "/Game/Athena/Balance/DataTables/AthenaGameData.AthenaGameData" "Default.Rift.TeleportHeight")
    (:RowFound
      (return (Math|Vector|vector+vector (* (Transformation|GetActorLocation) "1.000000,1.000000, 0") (Math|Vector|MakeVector 0.0 0.0 _outxy))))
    (:RowNotFound
      (return (Math|Vector|vector+vector (Transformation|GetActorLocation) (Math|Vector|MakeVector 0.0 0.0 (Variables|RiftPortal|GetFailsafefTeleportHeight)))))))

(fn IsFlying
  (bind _playerpawn (Variables|RiftPortal|GetPlayerPawn))
  (Utilities|IsValid _playerpawn
    (:Is%20Valid
      (bind _asfort_movement_comp_character (Utilities|Casting|CastToFortMovementComp_Character (Pawn|GetMovementComponent _playerpawn))
        (:then
          (return (Utilities|Enum|Equal(Enum) (Variables|/Script/Engine.CharacterMovementComponent|GetMovementMode _asfort_movement_comp_character) "MOVE_Flying")))
        (:CastFailed
          (return false))))))

(fn IsPlayerAttachedToAIPawn
  (params (AIPawn :type Fort%20AIPawn%20Object%20Reference))
  (for _array_element (Actor|GetAttachedActors AIPawn true false)
    (bind _asfort_player_pawn (Utilities|Casting|CastToFortPlayerPawn _array_element))
    (return true _asfort_player_pawn))
  (return false 0))

(fn PlayExpandOrCollapseAudio
  (params (Expand :type Boolean))
  (if (Variables|RiftPortal|GetPortalOpen)
    (Audio|PlaySoundatLocation (select Expand "/Game/Sounds/Fort_GamePlay_Sounds/Player/RiftPortal/RiftPortal_Expand_Cue.RiftPortal_Expand_Cue" "/Game/Sounds/Fort_GamePlay_Sounds/Player/RiftPortal/RiftPortal_Collapse_Cue.RiftPortal_Collapse_Cue") (Transformation|GetWorldLocation (Variables|Default|GetSM_Rift)))))

(fn QuestUpdateComplexCustom
  (params (Player :type Fort%20Player%20Pawn%20Object%20Reference))
  (if (Networking|IsServer)
    (Quest|SendComplexCustomStatEvent (Quests|GetQuestManager (Fortnite|GetFortPlayerControllerfromActor Player) "Invalid") 0 "(GameplayTags=((TagName=\"Athena.Quests.Rift\")))" "(GameplayTags=)" 1)))

(fn RiftVFXEnabled
  (return (or (Variables|/Script/Engine.SceneComponent|GetbVisible (Variables|Default|GetSM_Rift)) (Variables|/Script/Engine.SceneComponent|GetbVisible (Variables|Default|GetNS_Rift_Portal_Idle_Shimmer)))))

(fn SendRiftTriggerEvent
  (params (Actor :type Actor%20Object%20Reference))
  (bind (_outoverrideriftskydivege _returnvalue) (CallFunction|WantsOverrideRiftSkydiveGE))
  ;[1 1 1 0.5] If we want to override the skydive GE, pass ourselves into the gameplay event
  (Ability|SendGameplayEventtoActor Actor "(TagName=\"Abilities.Generic.Rift.Trigger\")" (Utilities|Struct|MakeGameplayEventData "(TagName=\"\")" 0 0 (select _returnvalue self 0))))

(fn ConstructionScript)

(fn UseRiftQuestUpdate
  (params (InController :type Controller%20Object%20Reference))
  (Utilities|IsValid InController
    (:Is%20Valid
      (bind _asfort_player_controller (Utilities|Casting|CastToFortPlayerController InController))
      (bind (_questisvalid _questcompleted) (Default|AthenaCheckQuestandObjectiveCompletion _asfort_player_controller (Variables|QuestStuff|GetQuestItem) (Variables|QuestStuff|GetObjectiveBackendName)))
      (if (and _questisvalid (not _questcompleted))
        (Class|AthenaQuestFunctionLibrary|AthenaGiveQuestUpdatetoPlayers :InQuestItemRef (Variables|QuestStuff|GetQuestItem) :ObjectiveBackendName (Variables|QuestStuff|GetObjectiveBackendName) :PlayerControllersForUpdate (Utilities|Array|MakeArray _asfort_player_controller))))))
