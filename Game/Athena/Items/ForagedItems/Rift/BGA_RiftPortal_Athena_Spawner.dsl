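; Server-side spawner actor that places a rift portal at a random height above its starting point in Athena.
; On BeginPlay it rolls a height multiplier capped by the MaxSpawnHeight scalable float.
; Rolls above one move it upward, snapped to the building grid, then trace from the origin to avoid spawning under the map.
; A trace hit reverts to the origin before spawning; low rolls spawn the Rift class variable there directly.
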(event EventBeginPlay
  (bind _maxspawnheightmultiplier (Variables|Default|GetMaxSpawnHeightMultiplier))
  (bind _returnvalue (* _maxspawnheightmultiplier 222.0))
  (bind _returnvalue_2 (/ _returnvalue 2.0))
  (bind _returnvalue_3 (- _returnvalue_2 65.0))
  (bind _returnvalue_4 (Transformation|GetActorLocation))
  (bind _output_vector (Utilities|SnapVectortoGrid _returnvalue_4 512.0 1.0))
  (bind _x (.x _output_vector))
  (bind _y (.y _output_vector))
  (bind _z (.z _output_vector))
  (bind _returnvalue_5 (+ _z _returnvalue_3))
  (bind _returnvalue_6 (Math|Vector|MakeVector _x _y _returnvalue_5))
  (bind _self self)
  (bind _actororiginallocation (Variables|Default|GetActorOriginalLocation))
  ;[1 1 1] Set Rift At Random Height and Center to building grid
  (Networking|SwitchHasAuthority
    (:Authority
      (Variables|Default|SetActorOriginalLocation (Transformation|GetActorLocation))
      (Variables|Default|SetMaxSpawnHeightMultiplier (Math|Random|RandomInteger (Math|Float|Truncate (ScalableFloat|GetValueatLevel (Variables|Default|GetMaxSpawnHeight)))))
      (if (> _maxspawnheightmultiplier 1)
        (Transformation|SetActorLocation _returnvalue_6 false true)
        ;[1 1 1] Make sure we didn't spawn under map.
        (bind (_outhit _returnvalue_7) (Collision|LineTraceForObjects :Start (Variables|Default|GetActorOriginalLocation) :End (Transformation|GetActorLocation) :ObjectTypes (Utilities|Array|MakeArray "ObjectTypeQuery1") :bTraceComplex true :DrawDebugType "None" :bIgnoreSelf true :TraceColor "(R=0.445000,G=0.384304,B=0.364900,A=1.000000)" :TraceHitColor "(R=1.000000,G=0.000000,B=0.000000,A=1.000000)" :DrawTime 5.0))
        (if _returnvalue_7
          (Transformation|SetActorLocation _actororiginallocation false true)
          (Game|SpawnActorfromClass :Class "/Game/Athena/Items/ForagedItems/Rift/BGA_RiftPortal_Athena.BGA_RiftPortal_Athena_C" :SpawnTransform (Math|Conversions|ToTransform(Vector) _actororiginallocation) :CollisionHandlingOverride "AlwaysSpawn" :TransformScaleMethod "OverrideRootScale" :Owner _self)
          (else
            (Game|SpawnActorfromClass :Class "/Game/Athena/Items/ForagedItems/Rift/BGA_RiftPortal_Athena.BGA_RiftPortal_Athena_C" :SpawnTransform (Math|Conversions|ToTransform(Vector) _returnvalue_6) :CollisionHandlingOverride "AlwaysSpawn" :TransformScaleMethod "OverrideRootScale" :Owner _self)))
        (else
          (Game|SpawnActorfromClass (Variables|Default|GetRift) (Math|Conversions|ToTransform(Vector) (Variables|Default|GetActorOriginalLocation)) "AlwaysSpawn" "SelectDefaultAtRuntime" self 0))))))

(event Collision|EventActorBeginOverlap
  (params (OtherActor :type Actor%20Object%20Reference)))

(event EventTick
  (params (DeltaSeconds :type Float%20%28single-precision%29)))

(fn ConstructionScript)
