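; Rift gate sky crack actor that places a crack particle effect at a set height and plays the looping sky tear sound.
; The construction script moves the effect to that world Z and passes it the height relative to the actor.
; After a delay on BeginPlay, it binds to the Asteria Rift Setup actor's event to drive the sound's CalendarControl parameter.
; A replicated debug value toggles timed logging of a material parameter collection value.
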(event EventBeginPlay
  (bind _returnvalue (Audio|SpawnSoundatLocation "/Game/Athena/Sounds/Props/RiftGate/Rift_Gate_Sky_Loop_Cue.Rift_Gate_Sky_Loop_Cue" (Transformation|GetWorldLocation (Variables|Default|GetAmbient%20Loop))))
  (Variables|Default|SetSkyTearAudioComponent _returnvalue)
  ;[1 1 1] Getting audio parameter from Rift Setup
  (Utilities|FlowControl|Delay 5.0)
  (bind _returnvalue_2 (Actor|GetActorOfClass "/Game/Athena/Asteria/Blueprints/BP_Asteria_Rift_Setup.BP_Asteria_Rift_Setup_C"))
  (Utilities|IsValid _returnvalue_2
    (:Is%20Valid
      (bind _output_get (Variables|Default|SetRiftSetupRef _returnvalue_2))
      (Default|BindEventtoSkyTearAudioParamED _output_get (delegate Custom|SetSkyTearAudioParameter)))))

(event Custom|SetSkyTearAudioParameter
  (params (AudioParam :type Float%20%28double-precision%29))
  (Audio|Parameter|SetFloatParameter (Variables|Default|GetSkyTearAudioComponent) "CalendarControl" AudioParam :!class /Script/Engine.AudioComponent))

(event EventEndPlay
  (params (EndPlayReason :type EEndPlayReason%20Enum :default "Destroyed"))
  (bind _riftsetupref (Variables|Default|GetRiftSetupRef)
    (:then
      (Default|UnbindallEventsfromSkyTearAudioParamED _riftsetupref))
    (:else)))

(event Custom|RiftGateCrack_DebugON
  ;[1 1 1] DEBUG
  (Variables|Default|SetDebugValue 1))

(event Custom|RiftGateCrack_DebugOff
  ;[1 1 1] DEBUG
  (Variables|Default|SetDebugValue 0))

(event Custom|TimerTick
  ;[1 1 1] DEBUG
  (bind _returnvalue (Rendering|Material|GetScalarParameterValue (Variables|Default|GetMPC%20Reference) (Variables|Default|GetMPC%20Name) :!class /Script/Engine.KismetMaterialLibrary))
  (Fortnite|FortShippingLog self (Utilities|String|Append "RiftGate_MPC Value = " (Utilities|String|ToString(Float) _returnvalue)) true))

(fn OnRep_DebugValue
  (switch int (Variables|Default|GetDebugValue)
    (:0
      (Utilities|Time|ClearandInvalidateTimerbyHandle (stale _returnvalue)))
    (:1
      (bind _returnvalue (Utilities|Time|SetTimerbyFunctionName 0 "TimerTick" 0.2 true)))))

(fn ConstructionScript
  (bind _self self)
  (bind (_returnvalue_x _returnvalue_y _returnvalue_z) (Math|Vector|BreakVector (Transformation|GetActorLocation _self)))
  (bind _height (Variables|Default|GetHeight))
  (Transformation|SetWorldLocation (Variables|Default|GetNS_RiftGate_Crack) (Math|Vector|MakeVector _returnvalue_x _returnvalue_y _height) false false)
  (Effects|Components|ParticleSystem|SetFloatParameter (Variables|Default|GetNS_RiftGate_Crack) "Height" (- _height _returnvalue_z)))
