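# Translucent unlit material for the cracked-glass shards of a Rift portal (gold variant), with refraction and parameter-driven opening.
# Panning smoke distorts crack textures, layered with fresnel, time-of-day crack colors, cloud tints and a perturbed cubemap reflection.
# A Corruption parameter shifts colors along a UV gradient.
# World position offset wobbles shards around Z and pushes them along directions stored in vertex color and UV2, with vertex-red-varied timing.
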Material /Game/Athena/Items/ForagedItems/Rift/FX/M_Shard_Portal_Gold
graph {
 properties {
  blendMode=BLEND_Translucent
  pixelDepthOffsetMode=PDOM_Legacy
  refractionMethod=RM_IndexOfRefraction
  shadingModel=MSM_Unlit
  usedShadingModels=Unlit
 }
 nodes [
  n1 = Constant3Vector() {constant=<rgba 0 0 1 0>}
  n2 = Time()
  n3 = VertexColor()
  n4 = LinearInterpolate(Alpha=@n3.R) {constA=-1, constB=2}
  n5 = ScalarParameter() {defaultValue=0.025, parameterName=TurnAmount}
  n6 = LinearInterpolate(Alpha=@n3.R) {constA=0.3}
  n7 = Multiply(A=Sine(None=Add(A=@n2, B=@n4)) {period=5.8}, B=Add(A=@n5, B=Multiply(A=@n5, B=@n6)) {constB=0.05}) {constB=0.1}
  n8 = ActorPositionWS()
  n9 = ReflectionVectorWS(CustomWorldNormal=MaterialFunctionCall(Rotation Axis=@n1, Rotation Angle=@n7, Normals Pivot=@n8) {materialFunction=&FixRotateAboutAxisNormals}.New World Space Vertex Normal) {bNormalizeCustomWorldNormal=true}
  n10 = Time()
  n11 = LinearInterpolate(Alpha=VertexColor().R) {constA=0.7, constB=1.2}
  n12 = TextureSample(UVs=Panner(Coordinate=TextureCoordinate() {coordinateIndex=1, uTiling=2, vTiling=1.2}, Time=Multiply(A=@n10, B=@n11)) {speedX=0.1, speedY=0.3}) {texture=&T_Inky_Smoke_Tile_Chromatic}
  n13 = ComponentMask(None=@n12.RGB) {g=true, r=true}
  n14 = Multiply(A=@n13, B=@n13)
  n15 = Constant2Vector() {g=0.05, r=0.05}
  n16 = Add(
   A=ShadingPathSwitch(Default=Multiply(A=@n14, B=@n15), Mobile=Constant()) {desc=Migrated from FeatureLevelSwitch}
   B=TextureCoordinate() {coordinateIndex=1}
  )
  n17 = Multiply(A=TextureSample(UVs=@n16) {texture=/Game/Effects/Fort_Effects/Textures/Tile/T_Glass_Cracks_M}.G) {constB=0.1}
  n18 = Multiply(
   A=TextureSampleParameterCube!(UVs=Add(A=@n9, B=@n17) {constB=0}, Apply View MipBias) {parameterName=Cubemap, samplerType=SAMPLERTYPE_LinearColor, texture=/Game/Packages/Fortress_Sky/TexturesHDR/T_AthenaSkylight}.RGB
   B=ComponentMask(None=MaterialFunctionCall() {materialFunction=/Engine/Functions/GetPostProcessFunctions/GetAmbientCubemapTint}.AmbientCubemapTint) {b=true, g=true, r=true}
  )
  n19 = VectorParameter() {parameterName=CubemapTint, defaultValue=<rgba 0.4914 0.5337 1 1>}
  n20 = ComponentMask(None=TextureCoordinate() {coordinateIndex=1}) {g=true}
  n21 = Add(
   A=Add(A=TextureSample() {texture=&T_Glass_Cracks_M}.G, B=@n17)
   B=Multiply(A=Power(Base=@n20) {constExponent=6.4}) {constB=0.15}
  )
  n22 = LinearInterpolate(Alpha=@n21) {constA=0.22}
  n23 = Add(A=ShadingPathSwitch(Default=@n12.RGB, Mobile=Constant()) {desc=Migrated from FeatureLevelSwitch}, B=@n22)
  n24 = ShadingPathSwitch(Default=Fresnel() {baseReflectFraction=0.2, exponent=4}, Mobile=Constant() {r=0.2}) {desc=Migrated from FeatureLevelSwitch}
  n25 = ComponentMask(None=TextureCoordinate()) {g=true}
  n26 = ScalarParameter() {parameterName=Corruption}
  n27 = LinearInterpolate(Alpha=@n26) {constA=10, constB=4}
  n28 = Clamp(None=Multiply(A=Power(Base=@n25, Exp=@n27), B=Multiply(A=@n26) {constB=5}) {constB=2})
  n29 = LinearInterpolate(
   A=VectorParameter() {parameterName=CrackColor, defaultValue=<rgba 2 1.4004 0.0519 0>}.RGB
   B=VectorParameter() {parameterName=CorruptCrackColor, defaultValue=<rgba 3 0.1193 0.0298 0>}.RGB
   Alpha=@n28
  )
  n30 = MaterialFunctionCall(
   Color_Morning=@n29
   Color_Day=Multiply(A=@n29) {constB=0.7}
   Color_Evening=@n29
   Color_Night=Multiply(A=@n29, B=VectorParameter() {parameterName=NightCrackColorMult, defaultValue=<rgba 0.7 0.4124 0.6978 0>}.RGB)
  ) {materialFunction=&Color_Time_of_Day}
  n31 = LinearInterpolate(
   A=VectorParameter() {parameterName=CloudColor, defaultValue=<rgba 0.1761 0.16 0.4792 0>}.RGB
   B=VectorParameter() {parameterName=CorruptCloudColor, defaultValue=<rgba 0.0144 0.0155 0.05 0>}.RGB
   Alpha=@n28
  )
  n32 = Add(A=Multiply(A=Multiply(A=@n23, B=@n24), B=Add(A=@n30.Result) {constA=0.2, constB=0.4}), B=Multiply(A=Multiply(A=@n31) {constB=8}, B=@n21))
  n33 = ParticleColor()
  Add(
   A=QualitySwitch(Default=Multiply(A=@n18, B=@n19.RGB), Low=Constant())
   B=Multiply(A=ScalarParameter() {defaultValue=1, parameterName=Emissive}, B=Multiply(A=@n32, B=@n33.RGB))
  )
  n34 = VertexColor()
  n35 = TextureCoordinate() {coordinateIndex=2}
  n36 = ComponentMask(None=@n35) {r=true}
  n37 = ComponentMask(None=@n35) {g=true}
  n38 = Subtract(A=AppendVector(A=OneMinus(None=@n34.G), B=AppendVector(A=@n36, B=@n37))) {constB=0.5}
  n39 = Multiply(
   A=Transform(None=Multiply(A=@n38) {constB=-2}) {transformSourceType=TRANSFORMSOURCE_Local}
   B=ScalarParameter() {defaultValue=38, parameterName=OpenDistance}
  )
  n40 = LinearInterpolate(Alpha=@n3.R) {constA=0.25, constB=0.05}
  n41 = Time()
  n42 = Multiply(A=@n41) {constB=0.25}
  n43 = MaterialFunctionCall() {materialFunction=&DebugTimeSine}
  n44 = Sine(None=@n41) {period=1.32}
  n45 = ScalarParameter() {parameterName=OpenAdd}
  n46 = Multiply(
   A=Multiply(
    A=MaterialFunctionCall!(Time=@n42) {materialFunction=/Game/Effects/Fort_Effects/Materials/Geode/_Functions/MF_SineVibration}.Result
    B=Add(A=@n43.Result, B=@n44)
   )
   B=Multiply(A=Add(A=@n45)) {constB=0.03}
  ) {constB=0.03}
  n47 = Add(
   A=Multiply(
    A=MaterialFunctionCall(Speed=@n40) {materialFunction=&DebugTimeSine}.Result
    B=Add(A=@n46) {constB=0.18}
   ) {constB=-0.17}
   B=Add(A=ScalarParameter() {parameterName=Open}, B=@n45)
  )
  Add(
   A=RotateAboutAxis(NormalizedRotationAxis=@n1, RotationAngle=@n7, PivotPoint=@n8, Position=WorldPosition().XYZ)
   B=Multiply(A=Multiply(A=@n39, B=@n47), B=ScalarParameter() {defaultValue=1, parameterName=CloseMult, sliderMax=1})
  )
  MaterialFunctionCall!(Use Alternate UVs, Alternate UVs) {materialFunction=/Game/Athena/Effects/Materials/Geode/Functions/MF_Geode_Crack_Colors}
  Multiply(A=ScalarParameter() {defaultValue=1, parameterName=Opacity}, B=Multiply(A=Add(A=@n24, B=@n22), B=@n33.A))
  n48 = OneMinus(None=@n24)
  QualitySwitch(
   Default=LinearInterpolate(B=ScalarParameter() {defaultValue=1.2, parameterName=Refraction}, Alpha=Multiply(A=@n33.A, B=@n48)) {constA=1}
   Low=Constant() {r=1}
  )
  Reroute(None=TextureSample(UVs=@n16) {samplerType=SAMPLERTYPE_Normal, texture=/Game/Effects/Fort_Effects/Textures/Tile/T_Glass_Cracks_N}.RGB)
 ]
}
