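# Niagara camera-lens effect for exiting a Rift crack: a speed build-up, cyan flash and shatter burst.
# GPU SpeedEmbers spawn on cylinders, growing faster and larger over the loop, stretched by velocity.
# A curve-driven Flare sweeps along local Z, spiking in size and brightness early in its life.
# Delayed Frags1/Frags2 bursts throw tumbling mesh shards down world Z with drag; Frags1 adds flickering sprites, while Bubble has no renderers.
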NiagaraSystem /Game/Athena/Items/ForagedItems/Rift/FX/NS_CameraLens_CrackExit
dependencies {
 usedDataInterfaces [
  &NiagaraDataInterfaceCurve
  &NiagaraDataInterfaceVectorCurve
  &NiagaraDataInterfaceVector2DCurve
 ]
 usedDynamicInputs []
 usedModules [&SystemState]
}
emitters {
 Bubble {rendererClasses [], renderers []}
 Flare {
  emitterUpdateScript [
   &EmitterState(Life Cycle Mode=::NewEnumerator1, Loop Behavior=::NewEnumerator1, Loop Duration=1, UseLoopDelay=true)
   &SpawnBurst_Instantaneous(Use Spawn Probability=false, UseLoopCountLimit=false)
  ]
  particleSpawnScript [
   &InitializeParticle(Lifetime=3.7, Sprite Size=<1200 1200>, Sprite Size Mode=::NewEnumerator4, UsePositionOffset=false, Write Material Random=true)
   &Color(Color=&MakeLinearColorFromVectorAndFloat("Vector (RGB)"=<0.6587 1.8605 2>, "Float (Alpha)"=1))
  ]
  particleUpdateScript [
   /SetVariables(Particles.Position=&TransformPosition(
     Position=&ConvertVectorToPosition(
      Input Position=&VectorFromCurve(VectorCurve=DataInterface() {propertyValues {
         XCurve {keys [<at 0.15 0 RCIM_Cubic>, {interpMode=RCIM_Cubic, tangentMode=RCTM_User, time=0.26, value=0}, <at 1 0 RCIM_Cubic>]}
         YCurve {keys [<at 0.15 0 RCIM_Cubic>, {interpMode=RCIM_Cubic, tangentMode=RCTM_User, time=0.26, value=0}, <at 1 0 RCIM_Cubic>]}
         ZCurve {keys [
           <at 0.15 -4500 RCIM_Cubic>
           {arriveTangent=13190.5518, interpMode=RCIM_Cubic, leaveTangent=13190.5518, tangentMode=RCTM_User, time=0.26, value=-3446.5576}
           <at 1 2000 RCIM_Cubic>
          ]}
        }})
      Conversion Operation:ENiagara_LWCConvertVecToPos=::NewEnumerator0
     )
     Source Space=::Local
     Destination Space=::Simulation
    )) {isSetParameters=true}
   &ParticleState()
   &ScaleSpriteSize(
    Scale Sprite Size Mode=::NewEnumerator0
    Uniform Scale Factor=&FloatFromCurve(FloatCurve=DataInterface() {propertyValues {Curve {keys [
         <at 0 0.3 RCIM_Cubic>
         {arriveTangent=18.0769, interpMode=RCIM_Cubic, leaveTangent=18.0769, time=0.25, value=1.2}
         {arriveTangent=-109.4233, interpMode=RCIM_Cubic, leaveTangent=-109.4233, tangentMode=RCTM_User, time=0.26, value=5}
         <at 0.35 0>
        ]}}})
   )
   &ScaleColor(
    Scale Alpha=&FloatFromCurve(FloatCurve=DataInterface() {propertyValues {Curve {keys [
         <at 0 0 RCIM_Cubic>
         {arriveTangent=3.8462, interpMode=RCIM_Cubic, leaveTangent=3.8462, time=0.24, value=0.25}
         {arriveTangent=-4.1667, interpMode=RCIM_Cubic, leaveTangent=-4.1667, time=0.26, value=1}
         <at 0.3 0>
        ]}}})
    Scale RGB=&VectorFromCurve(VectorCurve=DataInterface() {propertyValues {
       XCurve {keys [<at 0.1 5 RCIM_Cubic>,<at 0.3 1>]}
       YCurve {keys [<at 0.1 5 RCIM_Cubic>,<at 0.3 1>]}
       ZCurve {keys [<at 0.1 5 RCIM_Cubic>,<at 0.3 1>]}
      }})
    ScaleA=true
    ScaleRGB=true
   )
   &CameraOffset(Camera Offset Amount=20)
  ]
  rendererClasses [&NiagaraSpriteRendererProperties]
  renderers [{rendererClass=&NiagaraSpriteRendererProperties}]
 }
 Frags1 {
  emitterUpdateScript [
   &EmitterState(Life Cycle Mode=::NewEnumerator1, Loop Behavior=::NewEnumerator1, Loop Duration=1.5, UseLoopDelay=true)
   &SpawnBurst_Instantaneous(Spawn Count=14, Spawn Time=1, Use Spawn Probability=false, UseLoopCountLimit=false)
  ]
  particleSpawnScript [
   /SetVariables(Particles.ShapeLocationOffset:Vector3f=<0 0 -2000>) {isSetParameters=true}
   /SetVariables(Particles.ShapeLocationOffsetSpace:ENiagaraCoordinateSpace=::Local) {isSetParameters=true}
   /SetVariables(Particles.MeshRotationEuler:Vector3f=&RandomRangeVector(Minimum=<0 0 0>, Maximum=<1 1 1>, Evaluation Type=::NewEnumerator1)) {isSetParameters=true}
   /SetVariables(Particles.MeshOrientation=/CascadeToNiagaraConverter/NiagaraScripts/DynamicInputScripts/MakeQuatFromEuler(Euler:Vector3f=$Particles.MeshRotationEuler, Scale:NiagaraFloat=360)) {isSetParameters=true}
   /SetVariables(Particles.MeshRotationRateBase:Vector3f=&RandomRangeVector(Minimum=<0.25 2 -0.1>, Maximum=<1 3 0.1>, Evaluation Type=::NewEnumerator1)) {isSetParameters=true}
   &InitializeParticle(
    Lifetime=&RandomRangeFloat(Minimum=2, Maximum=2.7, Evaluation Type=::NewEnumerator1)
    Mesh Scale=&RandomRangeVector(Minimum=<10 10 10>, Maximum=<25 25 25>, Evaluation Type=::NewEnumerator1)
    Mesh Scale Mode=::NewEnumerator4
    Sprite Size Max=<2000 2000>
    Sprite Size Min=<1200 1200>
    Sprite Size Mode=::NewEnumerator2
    UsePositionOffset=false
    Write Material Random=true
   )
   &AddVelocity(
    Rotation Coordinate Space=::World
    Velocity=&RandomRangeVector(Minimum=<0 0 -15000>, Maximum=<0 0 -7000>, Evaluation Type=::NewEnumerator1)
   )
   &Color(Color=&MakeLinearColorFromVectorAndFloat("Vector (RGB)"=<3 3 3>, "Float (Alpha)"=1))
   &ShapeLocation(
    Hemisphere Distribution=<1 0.5>
    MaskSpawnGroup=false
    Non Uniform Scale=<1 1 -1>
    Offset=$Particles.ShapeLocationOffset
    Offset Coordinate Space=$Particles.ShapeLocationOffsetSpace
    Offset Mode=::NewEnumerator0
    Override Seed=false
    Sphere Radius=&RandomRangeFloat(Minimum=30, Maximum=40, Evaluation Type=::NewEnumerator1)
    Sphere Surface Distribution=1
    UseOwnerScale=true
   )
   /CascadeToNiagaraConverter/NiagaraScripts/ModuleScripts/FixShapeVector(
    Offset:Vector3f=$Particles.ShapeLocationOffset
    Offset Coordinate Space:ENiagaraCoordinateSpace=$Particles.ShapeLocationOffsetSpace
    Offset Mode:ENiagara_OffsetMode=::NewEnumerator0
   )
   &AddVelocity(Velocity=&Multiply_VectorByFloat(
     Vector=$Output.FixShapeVector.ShapeVector
     Float=&RandomRangeFloat(Minimum=120, Maximum=200, Evaluation Type=::NewEnumerator1)
    )) {module=AddVelocity001}
  ]
  particleUpdateScript [
   /SetVariables(Particles.MeshRotationRate:Vector3f=$Particles.MeshRotationRateBase) {isSetParameters=true}
   /SetVariables(
    Particles.Color2=<rgba 0.4795 2.1784 3 1>
    Particles.MeshRotationRate:Vector3f=&Multiply_Vector(
     A=$Particles.MeshRotationRate
     B=&VectorFromFloat(Value=&FloatFromCurve(FloatCurve=DataInterface() {propertyValues {Curve {keys [
           {arriveTangent=-1.4943, interpMode=RCIM_Cubic, leaveTangent=-1.4943, tangentMode=RCTM_User, time=0, value=1}
           {arriveTangent=-0.2218, interpMode=RCIM_Cubic, leaveTangent=-0.2218, tangentMode=RCTM_User, time=1, value=0}
          ]}}}))
    )
   ) {isSetParameters=true}
   &ParticleState()
   &ScaleMeshSize(Scale Factor=&VectorFromCurve(VectorCurve=DataInterface() {propertyValues {
       XCurve {keys [<at 0.2 1 RCIM_Cubic>, {arriveTangent=-1.25, interpMode=RCIM_Cubic, leaveTangent=-1.25, time=0.7932, value=0.7509}, <at 1 0>]}
       YCurve {keys [<at 0.2 1 RCIM_Cubic>, {arriveTangent=-0.25, interpMode=RCIM_Cubic, leaveTangent=-0.25, time=0.7932, value=0.85}, <at 1 0.8>]}
       ZCurve {keys [<at 0.2 1 RCIM_Cubic>,<at 0.7932 1 RCIM_Cubic>,<at 1 1>]}
      }}))
   &Drag(
    Drag=&RandomRangeFloat(Minimum=1.2, Maximum=2.2, Evaluation Type=::NewEnumerator1)
    Use Linear Drag=true
    Use Rotational Drag=true
   )
   &SolveForcesAndVelocity(Clamp Velocity=false, Limit Acceleration=false)
   &ScaleColor(
    Scale Alpha=&FloatFromCurve(FloatCurve=DataInterface() {propertyValues {Curve {keys [<at 0.7 1 RCIM_Cubic>,<at 1 0>]}}})
    Scale RGB=&VectorFromFloat(Value=&FloatFromCurve(FloatCurve=DataInterface() {propertyValues {Curve {keys [<at 0.1 25 RCIM_Cubic>, {arriveTangent=-26.6667, interpMode=RCIM_Cubic, leaveTangent=-26.6667, time=0.1961, value=5}, <at 1 1>]}}}))
    ScaleA=true
    ScaleRGB=true
   )
   &ScaleSpriteSize(
    Uniform Curve Index=&Add_Float(
     A=&Multiply_Float(A=$Particles.Age, B=&RandomRangeFloat(Minimum=0.8, Maximum=1.5))
     B=&RandomRangeFloat(Minimum=-0.25, Maximum=0.25)
    )
    Uniform Curve Sprite Scale=DataInterface() {propertyValues {Curve {keys [
        {tangentMode=RCTM_User, time=0.2, value=0}
        <at 0.22 1 RCIM_Cubic>
        {tangentMode=RCTM_User, time=0.5, value=0}
        <at 0.78 0>
        <at 0.8 0.8 RCIM_Cubic>
        <at 1 0>
       ]}}}
   )
   /CascadeToNiagaraConverter/NiagaraScripts/ModuleScripts/UpdateMeshOrientationByEuler(RotationRate:Vector3f=$Particles.MeshRotationRate)
  ]
  rendererClasses [&NiagaraMeshRendererProperties, &NiagaraSpriteRendererProperties]
  renderers [
   {rendererClass=&NiagaraMeshRendererProperties, rendererIndex=0}
   {rendererClass=&NiagaraSpriteRendererProperties, rendererIndex=1}
  ]
 }
 Frags2 {
  emitterUpdateScript [
   &EmitterState(Life Cycle Mode=::NewEnumerator1, Loop Behavior=::NewEnumerator1, Loop Duration=1.5, UseLoopDelay=true)
   &SpawnBurst_Instantaneous(Spawn Count=14, Spawn Time=1, Use Spawn Probability=false, UseLoopCountLimit=false)
  ]
  particleSpawnScript [
   /SetVariables(Particles.ShapeLocationOffset:Vector3f=<0 0 -2000>) {isSetParameters=true}
   /SetVariables(Particles.ShapeLocationOffsetSpace:ENiagaraCoordinateSpace=::Local) {isSetParameters=true}
   /SetVariables(Particles.MeshRotationEuler:Vector3f=&RandomRangeVector(Minimum=<0 0 0>, Maximum=<1 1 1>, Evaluation Type=::NewEnumerator1)) {isSetParameters=true}
   /SetVariables(Particles.MeshOrientation=&MakeQuatFromEuler(Euler:Vector3f=$Particles.MeshRotationEuler, Scale:NiagaraFloat=360)) {isSetParameters=true}
   /SetVariables(Particles.MeshRotationRateBase:Vector3f=&RandomRangeVector(Minimum=<0.25 2 -0.1>, Maximum=<1 3 0.1>, Evaluation Type=::NewEnumerator1)) {isSetParameters=true}
   &InitializeParticle(
    Lifetime=&RandomRangeFloat(Minimum=2, Maximum=2.7, Evaluation Type=::NewEnumerator1)
    Mesh Scale=&RandomRangeVector(Minimum=<10 10 10>, Maximum=<25 25 25>, Evaluation Type=::NewEnumerator1)
    Mesh Scale Mode=::NewEnumerator4
    UsePositionOffset=false
    Write Material Random=true
   )
   &AddVelocity(
    Rotation Coordinate Space=::World
    Velocity=&RandomRangeVector(Minimum=<0 0 -15000>, Maximum=<0 0 -7000>, Evaluation Type=::NewEnumerator1)
   )
   &Color(Color=&MakeLinearColorFromVectorAndFloat("Vector (RGB)"=<3 3 3>, "Float (Alpha)"=1))
   &ShapeLocation(
    Hemisphere Distribution=<1 0.5>
    MaskSpawnGroup=false
    Non Uniform Scale=<1 1 -1>
    Offset=$Particles.ShapeLocationOffset
    Offset Coordinate Space=$Particles.ShapeLocationOffsetSpace
    Offset Mode=::NewEnumerator0
    Override Seed=false
    Sphere Radius=&RandomRangeFloat(Minimum=30, Maximum=40, Evaluation Type=::NewEnumerator1)
    Sphere Surface Distribution=1
    UseOwnerScale=true
   )
   &FixShapeVector(
    Offset:Vector3f=$Particles.ShapeLocationOffset
    Offset Coordinate Space:ENiagaraCoordinateSpace=$Particles.ShapeLocationOffsetSpace
    Offset Mode:ENiagara_OffsetMode=::NewEnumerator0
   )
   &AddVelocity(Velocity=&Multiply_VectorByFloat(
     Vector=$Output.FixShapeVector.ShapeVector
     Float=&RandomRangeFloat(Minimum=120, Maximum=200, Evaluation Type=::NewEnumerator1)
    )) {module=AddVelocity001}
  ]
  particleUpdateScript [
   /SetVariables(Particles.MeshRotationRate:Vector3f=$Particles.MeshRotationRateBase) {isSetParameters=true}
   /SetVariables(Particles.MeshRotationRate:Vector3f=&Multiply_Vector(
     A=$Particles.MeshRotationRate
     B=&VectorFromFloat(Value=&FloatFromCurve(FloatCurve=DataInterface() {propertyValues {Curve {keys [
           {arriveTangent=-1.4943, interpMode=RCIM_Cubic, leaveTangent=-1.4943, tangentMode=RCTM_User, time=0, value=1}
           {arriveTangent=-0.2218, interpMode=RCIM_Cubic, leaveTangent=-0.2218, tangentMode=RCTM_User, time=1, value=0}
          ]}}}))
    )) {isSetParameters=true}
   &ParticleState()
   &ScaleMeshSize(Scale Factor=&VectorFromCurve(VectorCurve=DataInterface() {propertyValues {
       XCurve {keys [<at 0.2 1 RCIM_Cubic>, {arriveTangent=-1.25, interpMode=RCIM_Cubic, leaveTangent=-1.25, time=0.7932, value=0.7509}, <at 1 0>]}
       YCurve {keys [<at 0.2 1 RCIM_Cubic>, {arriveTangent=-0.25, interpMode=RCIM_Cubic, leaveTangent=-0.25, time=0.7932, value=0.85}, <at 1 0.8>]}
       ZCurve {keys [<at 0.2 1 RCIM_Cubic>,<at 0.7932 1 RCIM_Cubic>,<at 1 1>]}
      }}))
   &Drag(
    Drag=&RandomRangeFloat(Minimum=1.2, Maximum=2.2, Evaluation Type=::NewEnumerator1)
    Use Linear Drag=true
    Use Rotational Drag=true
   )
   &SolveForcesAndVelocity(Clamp Velocity=false, Limit Acceleration=false)
   &ScaleColor(
    Scale Alpha=&FloatFromCurve(FloatCurve=DataInterface() {propertyValues {Curve {keys [<at 0.7 1 RCIM_Cubic>,<at 1 0>]}}})
    Scale RGB=&VectorFromFloat(Value=&FloatFromCurve(FloatCurve=DataInterface() {propertyValues {Curve {keys [<at 0.1 25 RCIM_Cubic>, {arriveTangent=-26.6667, interpMode=RCIM_Cubic, leaveTangent=-26.6667, time=0.1961, value=5}, <at 1 1>]}}}))
    ScaleA=true
    ScaleRGB=true
   )
   &UpdateMeshOrientationByEuler(RotationRate:Vector3f=$Particles.MeshRotationRate)
  ]
  rendererClasses [&NiagaraMeshRendererProperties]
  renderers [{rendererClass=&NiagaraMeshRendererProperties, rendererIndex=0}]
 }
 SpeedEmbers {
  simTarget=GPUComputeSim
  emitterUpdateScript [
   &EmitterState(Life Cycle Mode=::NewEnumerator1, Loop Behavior=::NewEnumerator1, Loop Duration=1.1, UseLoopDelay=false)
   &SpawnBurst_Instantaneous(Spawn Count=100, Use Spawn Probability=false, UseLoopCountLimit=false)
   &SpawnRate(
    SpawnRate=&FloatFromCurve(FloatCurve=DataInterface() {propertyValues {Curve {keys [<at 0 200 RCIM_Cubic>,<at 1 500>]}}}, CurveIndex=$Emitter.NormalizedLoopAge)
    Use Spawn Probability=false
   )
  ]
  particleSpawnScript [
   /SetVariables(Particles.ShapeLocationOffset:Vector3f=&VectorFromCurve(
     VectorCurve=DataInterface() {propertyValues {
       XCurve {keys [<at 0 0 RCIM_Cubic>,<at 1 0>]}
       YCurve {keys [<at 0 0 RCIM_Cubic>,<at 1 0>]}
       ZCurve {keys [<at 0 -600 RCIM_Cubic>,<at 1 -1200>]}
      }}
     CurveIndex=$Emitter.LoopedAge
    )) {isSetParameters=true}
   /SetVariables(Particles.ShapeLocationOffsetSpace:ENiagaraCoordinateSpace=::Local) {isSetParameters=true}
   &InitializeParticle(
    Lifetime=&FloatFromCurve(FloatCurve=DataInterface() {propertyValues {Curve {keys [<at 0 1>,<at 1 0.25>]}}}, CurveIndex=$Emitter.LoopedAge)
    Sprite Size=&Vector2DFromCurve(
     Vector2Curve=DataInterface() {propertyValues {
       XCurve {keys [<at 0 1 RCIM_Cubic>, {arriveTangent=4.9588, interpMode=RCIM_Cubic, leaveTangent=4.9588, tangentMode=RCTM_User, time=1, value=3}]}
       YCurve {keys [<at 0 1 RCIM_Cubic>, {arriveTangent=27.8835, interpMode=RCIM_Cubic, leaveTangent=27.8835, tangentMode=RCTM_User, time=1, value=10}]}
      }}
     CurveIndex=$Emitter.LoopedAge
     ScaleCurve=<1 1>
    )
    Sprite Size Mode=::NewEnumerator4
    UsePositionOffset=false
    Write Material Random=true
   )
   &Color(Color=&MakeLinearColorFromVectorAndFloat("Vector (RGB)"=<0.7519 5.8989 8>, "Float (Alpha)"=1))
   &ShapeLocation(
    Cylinder Height=800
    Cylinder Radius=&RandomRangeFloat(Minimum=200, Maximum=600, Evaluation Type=::NewEnumerator1)
    MaskSpawnGroup=false
    Offset=$Particles.ShapeLocationOffset
    Offset Coordinate Space=$Particles.ShapeLocationOffsetSpace
    Offset Mode=::NewEnumerator0
    Override Seed=false
    Rotation Mode=::NewEnumerator2
    Shape Primitive=::NewEnumerator1
    Surface Only=true
    Use Endcaps In Surface Only Mode=false
    UseOwnerScale=true
   )
   &AddVelocity(
    Module.Use Falloff Distance=false
    Origin Offset=<0 0 -2400>
    Override Seed=false
    Velocity Mode=::NewEnumerator1
    Velocity Speed=&FloatFromCurve(
     FloatCurve=DataInterface() {propertyValues {Curve {
        defaultValue=1
        keys [<at 0 0 RCIM_Cubic>, {arriveTangent=1.9376, interpMode=RCIM_Cubic, leaveTangent=1.9376, tangentMode=RCTM_User, time=1, value=1}]
       }}}
     CurveIndex=$Emitter.LoopedAge
     Scale Curve=&RandomRangeFloat(Minimum=3800, Maximum=4500)
    )
   )
   &FixShapeVector(
    Offset:Vector3f=$Particles.ShapeLocationOffset
    Offset Coordinate Space:ENiagaraCoordinateSpace=$Particles.ShapeLocationOffsetSpace
    Offset Mode:ENiagara_OffsetMode=::NewEnumerator0
   ) {enabled=false}
   &AddVelocity(Velocity=&Multiply_VectorByFloat(
     Vector=&Multiply_Vector(A=$Output.FixShapeVector.ShapeVector, B=<1 1 0>)
     Float=&FloatFromCurve(FloatCurve=DataInterface() {propertyValues {Curve {keys [<at 0 0>,<at 1 2>]}}}, CurveIndex=$Emitter.LoopedAge)
    )) {enabled=false, module=AddVelocity001}
  ]
  particleUpdateScript [
   &ParticleState()
   &AccelerationForce(Acceleration=<0 0 2000>)
   &SolveForcesAndVelocity(Clamp Velocity=false, Limit Acceleration=false)
   &ScaleSpriteSizeBySpeed(
    Max Scale Factor=<3 50>
    Min Scale Factor=<0.01 0.01>
    Source Velocity:Vector3f=&Multiply_Vector(A=<0.01 0.01 1>, B=$Particles.Velocity)
    Velocity Threshold=4000
   )
   &ScaleColor(
    Scale Alpha=&FloatFromCurve(FloatCurve=DataInterface() {propertyValues {Curve {keys [<at 0 0 RCIM_Cubic>,<at 0.3 1 RCIM_Cubic>,<at 1 0>]}}})
    ScaleA=true
    ScaleRGB=true
   )
  ]
  rendererClasses [&NiagaraSpriteRendererProperties]
  renderers [{rendererClass=&NiagaraSpriteRendererProperties}]
 }
}
summary {
 systemName=NS_CameraLens_CrackExit
 emitters [
  {emitterName=Frags1, rendererClasses [&NiagaraMeshRendererProperties, &NiagaraSpriteRendererProperties]}
  {emitterName=Frags2, rendererClasses [&NiagaraMeshRendererProperties]}
  {emitterName=Flare, rendererClasses [&NiagaraSpriteRendererProperties]}
  {emitterName=SpeedEmbers, simTarget=GPUComputeSim, rendererClasses [&NiagaraSpriteRendererProperties]}
  {emitterName=Bubble, rendererClasses []}
 ]
 userVariables []
}
