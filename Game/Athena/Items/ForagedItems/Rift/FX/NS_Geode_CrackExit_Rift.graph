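# Niagara system for the Geode crack exit effect of the Rift foraged item, apparently the visual for a rift closing or being exited.
# Ten emitters cover a crack spline and portal, corona, end flare, light, two residual electric arc layers and three fragment sets.
# User color and mesh array parameters expose inputs, and the system uses SystemState plus a local scratch module.
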NiagaraSystem /Game/Athena/Items/ForagedItems/Rift/FX/NS_Geode_CrackExit_Rift
dependencies {
 usedDataInterfaces []
 usedDynamicInputs []
 usedModules [&self:ScratchModule, &SystemState]
}
emitters {
 Corona {rendererClasses [], renderers []}
 CrackPortal {rendererClasses [], renderers []}
 CrackSpline {rendererClasses [], renderers []}
 ElectricArcsResidual {rendererClasses [], renderers []}
 ElectricArcsResidual001 {rendererClasses [], renderers []}
 Endflare {rendererClasses [], renderers []}
 Frags1 {rendererClasses [], renderers []}
 Frags2 {rendererClasses [], renderers []}
 Frags3 {rendererClasses [], renderers []}
 Light {rendererClasses [], renderers []}
}
summary {
 systemName=NS_Geode_CrackExit_Rift
 emitters [
  {emitterName=CrackSpline, rendererClasses []}
  {emitterName=CrackPortal, rendererClasses []}
  {emitterName=ElectricArcsResidual, rendererClasses []}
  {emitterName=Light, rendererClasses []}
  {emitterName=Endflare, rendererClasses []}
  {emitterName=Corona, rendererClasses []}
  {emitterName=Frags1, rendererClasses []}
  {emitterName=Frags2, rendererClasses []}
  {emitterName=Frags3, rendererClasses []}
  {emitterName=ElectricArcsResidual001, rendererClasses []}
 ]
 userVariables [
  User.Color Array {
   type=NiagaraDataInterfaceArrayColor
   description=""
   defaultValue {
    struct=NiagaraExt_VariableValue_DataInterface
    value {
     dataInterface=&self:NiagaraDataInterfaceArrayColor_0
     dataInterfaceClass=&NiagaraDataInterfaceArrayColor
    }
   }
  }
  User.Mesh Array {
   type=NiagaraDataInterfaceArrayMesh
   description=""
   defaultValue {
    struct=NiagaraExt_VariableValue_DataInterface
    value {
     dataInterface=&self:NiagaraDataInterfaceArrayMesh_0
     dataInterfaceClass=/Script/Niagara.NiagaraDataInterfaceArrayMesh
    }
   }
  }
 ]
}
