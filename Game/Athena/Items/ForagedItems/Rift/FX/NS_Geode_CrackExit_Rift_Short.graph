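# Niagara effect for the short variant of a Rift geode crack exit: a brief sky crack with a downward burst.
# CrackPortal and CrackSpline meshes sit raised, using near-identical stepped curves on dynamic material parameters to open, hold and close over two seconds.
# Corona, flares, shockwave and light add the flash, alongside electric arcs, downward-thrown fragments and GPU curl-noise sparkles.
# Endflare follows after a delay; several emitters use distance culling.
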NiagaraSystem /Game/Athena/Items/ForagedItems/Rift/FX/NS_Geode_CrackExit_Rift_Short
dependencies {
 usedDataInterfaces [&NiagaraDataInterfaceVectorCurve, &NiagaraDataInterfaceCurve]
 usedDynamicInputs []
 usedModules [&SystemState]
}
emitters {
 Corona {
  emitterUpdateScript [
   &EmitterState(Life Cycle Mode=::NewEnumerator1, Loop Behavior=::NewEnumerator1, Loop Duration=1, UseLoopDelay=true)
   &SpawnBurst_Instantaneous(Use Spawn Probability=false, UseLoopCountLimit=false)
  ]
  particleSpawnScript [
   &InitializeParticle(
    Lifetime=0.6
    Position=&ConvertVectorToPosition(Input Position=<0 0 0>, Conversion Operation:ENiagara_LWCConvertVecToPos=::NewEnumerator0)
    Position Mode=::NewEnumerator1
    Sprite Size=<25000 25000>
    Sprite Size Mode=::NewEnumerator4
    UsePositionOffset=false
    Write Material Random=true
   )
   &Color(Color=&MakeLinearColorFromVectorAndFloat("Vector (RGB)"=<0.1344 0.4283 1>, "Float (Alpha)"=&ClampFloat(Float=1)))
  ]
  particleUpdateScript [
   &ParticleState()
   &ScaleSpriteSize(
    Scale Sprite Size Mode=::NewEnumerator0
    Uniform Scale Factor=&FloatFromCurve(FloatCurve=DataInterface() {propertyValues {Curve {keys [{arriveTangent=1.4067, interpMode=RCIM_Cubic, time=0, value=0.1}, {interpMode=RCIM_Cubic, leaveTangent=0.2191, time=1, value=1}]}}})
   )
   &ScaleColor(
    Scale Alpha=&FloatFromCurve(FloatCurve=DataInterface() {propertyValues {Curve {keys [<at 0 0 RCIM_Cubic>,<at 0.25 1 RCIM_Cubic>,<at 1 0>]}}})
    Scale RGB=<2 2 2>
    ScaleA=true
    ScaleRGB=true
   )
   &DynamicMaterialParameters(
    Index 0 Param 1=&FloatFromCurve(FloatCurve=DataInterface() {propertyValues {Curve {keys [{arriveTangent=-2.4947, interpMode=RCIM_Cubic, time=0, value=1.5}, <at 1 0.2 RCIM_Cubic>]}}})
    Index 0 Param 2=0.4
    Index 0 Param 3=0.6
    Index 0 Param 4=1
    Module.Index0Param0Write=true
    Module.Index0Param1Write=true
    Module.Index0Param2Write=true
    Module.Index0Param3Write=true
   )
   &CameraOffset(Camera Offset Amount=500)
  ]
  rendererClasses [&NiagaraSpriteRendererProperties]
  renderers [{rendererClass=&NiagaraSpriteRendererProperties}]
 }
 CrackPortal {
  emitterUpdateScript [
   &EmitterState(Life Cycle Mode=::NewEnumerator1, Loop Behavior=::NewEnumerator1, Loop Duration=1, UseLoopDelay=true)
   &SpawnBurst_Instantaneous(Use Spawn Probability=false, UseLoopCountLimit=false)
  ]
  particleSpawnScript [
   &InitializeParticle(
    Lifetime=2
    Mesh Scale=<67 67 90>
    Mesh Scale Mode=::NewEnumerator4
    Position Offset=<0 0 2100>
    UsePositionOffset=true
    Write Material Random=true
   )
   &InitialMeshOrientation(Mesh Orientation Mode=::NewEnumerator2, Rotation=<0.5 0 0>, Use Rotation Vector=true)
   &Color(Color=&MakeLinearColorFromVectorAndFloat("Vector (RGB)"=<2 2 2>, "Float (Alpha)"=&ClampFloat(Float=1)))
  ]
  particleUpdateScript [
   &ParticleState()
   &ScaleMeshSize(Scale Factor=&VectorFromCurve(VectorCurve=DataInterface() {propertyValues {
       XCurve {keys [
         {arriveTangent=13.7302, interpMode=RCIM_Cubic, time=0, value=0.9}
         {arriveTangent=0.2, interpMode=RCIM_Cubic, leaveTangent=0.2, time=0.015, value=1}
         <at 0.5 1 RCIM_Cubic>
        ]}
       YCurve {keys [
         {arriveTangent=13.2017, interpMode=RCIM_Cubic, time=0, value=0.9}
         {arriveTangent=0.2, interpMode=RCIM_Cubic, leaveTangent=0.2, time=0.015, value=1}
         <at 0.5 1 RCIM_Cubic>
        ]}
       ZCurve {keys [
         {arriveTangent=60.1727, interpMode=RCIM_Cubic, time=0, value=0.5}
         {arriveTangent=1.2, interpMode=RCIM_Cubic, leaveTangent=1.2, time=0.015, value=1}
         <at 0.5 1.1 RCIM_Cubic>
        ]}
      }}))
   &ScaleColor(
    Scale Alpha=&FloatFromCurve(FloatCurve=DataInterface() {propertyValues {Curve {keys [
         <at 0.1 1 RCIM_Cubic>
         {arriveTangent=-1.1088, interpMode=RCIM_Cubic, leaveTangent=-1.1088, time=0.2, value=0.2833}
         {arriveTangent=-0.3541, interpMode=RCIM_Cubic, leaveTangent=-0.3541, time=0.9, value=0.1129}
         <at 1 0 RCIM_Cubic>
        ]}}})
    Scale RGB=&VectorFromFloat(Value=&FloatFromCurve(FloatCurve=DataInterface() {propertyValues {Curve {keys [{arriveTangent=-52.7099, interpMode=RCIM_Cubic, leaveTangent=-52.7099, tangentMode=RCTM_User, time=0, value=5}, <at 0.15 1>]}}}))
    ScaleA=true
    ScaleRGB=true
   )
   &DynamicMaterialParameters(
    Index 0 Param 1=&FloatFromCurve(FloatCurve=DataInterface() {propertyValues {Curve {keys [
         {arriveTangent=13.0398, interpMode=RCIM_Cubic, leaveTangent=13.0398, tangentMode=RCTM_User, time=0, value=0}
         {arriveTangent=0.1152, leaveTangent=0.1152, time=0.0197, value=0.2592}
         {arriveTangent=0.0123, leaveTangent=0.0123, time=0.0261, value=0.3744}
         {arriveTangent=0.1324, leaveTangent=0.1324, time=0.1008, value=0.3867}
         {arriveTangent=0.0442, leaveTangent=0.0442, time=0.101, value=0.519}
         {arriveTangent=0.1123, leaveTangent=0.1123, time=0.1941, value=0.5633}
         {arriveTangent=2.9135, interpMode=RCIM_Cubic, leaveTangent=2.9135, tangentMode=RCTM_User, time=0.1951, value=0.6755}
         <at 0.3 1>
         {arriveTangent=-0.3505, leaveTangent=-0.3505, time=0.7395, value=1}
         {arriveTangent=-0.0069, leaveTangent=-0.0069, time=0.7605, value=0.6495}
         {arriveTangent=-0.2804, leaveTangent=-0.2804, time=0.8367, value=0.6427}
         {arriveTangent=-0.0148, leaveTangent=-0.0148, time=0.8402, value=0.3623}
         {arriveTangent=-0.1832, leaveTangent=-0.1832, time=0.9216, value=0.3475}
         {arriveTangent=0.0012, leaveTangent=0.0012, time=0.9314, value=0.1644}
         {arriveTangent=-0.1656, leaveTangent=-0.1656, time=0.972, value=0.1656}
         <at 1 0 RCIM_Cubic>
        ]}}})
    Index 0 Param 2=&FloatFromCurve(FloatCurve=DataInterface() {propertyValues {Curve {keys [<at 0.4 2.5 RCIM_Cubic>, {arriveTangent=-10.9864, interpMode=RCIM_Cubic, leaveTangent=-10.9864, tangentMode=RCTM_User, time=1, value=0.05}]}}})
    Index 0 Param 3=&FloatFromCurve(FloatCurve=DataInterface() {propertyValues {Curve {keys [
         {arriveTangent=5.4303, interpMode=RCIM_Cubic, leaveTangent=5.4303, tangentMode=RCTM_User, time=0, value=-0.1}
         {arriveTangent=0.0133, interpMode=RCIM_Cubic, leaveTangent=0.0133, tangentMode=RCTM_User, time=0.02, value=0.0059}
         {arriveTangent=0.0056, interpMode=RCIM_Cubic, leaveTangent=0.0056, time=0.3, value=0.0104}
         {arriveTangent=-0.0288, interpMode=RCIM_Cubic, leaveTangent=-0.0288, time=0.7539, value=0.01}
         {arriveTangent=-0.2604, interpMode=RCIM_Cubic, leaveTangent=-0.2604, time=0.7654, value=-0.003}
         {arriveTangent=-0.2775, interpMode=RCIM_Cubic, leaveTangent=-0.2775, time=0.8371, value=-0.0117}
         {arriveTangent=-0.2326, interpMode=RCIM_Cubic, leaveTangent=-0.2326, time=0.8389, value=-0.0234}
         {arriveTangent=-0.2151, interpMode=RCIM_Cubic, leaveTangent=-0.2151, time=0.9168, value=-0.0302}
         {arriveTangent=-0.5984, interpMode=RCIM_Cubic, leaveTangent=-0.5984, time=0.9199, value=-0.0408}
         {interpMode=RCIM_Cubic, leaveTangent=-0.5311, time=1, value=-0.08}
        ]}}})
    Index 0 Param 4=&FloatFromCurve(FloatCurve=DataInterface() {propertyValues {Curve {keys [
         {arriveTangent=91.4522, interpMode=RCIM_Cubic, leaveTangent=91.4522, tangentMode=RCTM_User, time=0, value=0}
         {interpMode=RCIM_Cubic, tangentMode=RCTM_User, time=0.02, value=0.64}
         {interpMode=RCIM_Cubic, tangentMode=RCTM_User, time=0.8, value=0.64}
         {arriveTangent=-6.1885, interpMode=RCIM_Cubic, leaveTangent=-6.1885, tangentMode=RCTM_User, time=1, value=-0.25}
        ]}}})
    Module.Index0Param0Write=true
    Module.Index0Param1Write=true
    Module.Index0Param2Write=true
    Module.Index0Param3Write=true
   )
  ]
  rendererClasses [&NiagaraMeshRendererProperties]
  renderers [{rendererClass=&NiagaraMeshRendererProperties, rendererIndex=0}]
 }
 CrackSpline {
  emitterUpdateScript [
   &EmitterState(Life Cycle Mode=::NewEnumerator1, Loop Behavior=::NewEnumerator1, Loop Duration=1, UseLoopDelay=true)
   &SpawnBurst_Instantaneous(Use Spawn Probability=false, UseLoopCountLimit=false)
  ]
  particleSpawnScript [
   &InitializeParticle(
    Lifetime=2
    Mesh Scale=<90 90 90>
    Mesh Scale Mode=::NewEnumerator4
    Position Offset=<0 0 2100>
    UsePositionOffset=true
    Write Material Random=true
   )
   &InitialMeshOrientation(Mesh Orientation Mode=::NewEnumerator2, Rotation=<0.5 0 0>, Use Rotation Vector=true)
  ]
  particleUpdateScript [
   &ParticleState()
   &ScaleMeshSize(Scale Factor=&VectorFromCurve(VectorCurve=DataInterface() {propertyValues {
       XCurve {keys [
         {arriveTangent=13.7302, interpMode=RCIM_Cubic, time=0, value=0.9}
         {arriveTangent=0.2, interpMode=RCIM_Cubic, leaveTangent=0.2, time=0.015, value=1}
         <at 0.5 1 RCIM_Cubic>
        ]}
       YCurve {keys [
         {arriveTangent=13.2017, interpMode=RCIM_Cubic, time=0, value=0.9}
         {arriveTangent=0.2, interpMode=RCIM_Cubic, leaveTangent=0.2, time=0.015, value=1}
         <at 0.5 1 RCIM_Cubic>
        ]}
       ZCurve {keys [
         {arriveTangent=60.1727, interpMode=RCIM_Cubic, time=0, value=0.5}
         {arriveTangent=1.2, interpMode=RCIM_Cubic, leaveTangent=1.2, time=0.015, value=1}
         <at 0.5 1.1 RCIM_Cubic>
        ]}
      }}))
   &DynamicMaterialParameters(
    Index 0 Param 1=&FloatFromCurve(FloatCurve=DataInterface() {propertyValues {Curve {keys [
         {arriveTangent=13.0398, interpMode=RCIM_Cubic, leaveTangent=13.0398, tangentMode=RCTM_User, time=0, value=0}
         {arriveTangent=0.1152, leaveTangent=0.1152, time=0.0197, value=0.2592}
         {arriveTangent=0.0123, leaveTangent=0.0123, time=0.0261, value=0.3744}
         {arriveTangent=0.1324, leaveTangent=0.1324, time=0.1008, value=0.3867}
         {arriveTangent=0.0442, leaveTangent=0.0442, time=0.101, value=0.519}
         {arriveTangent=0.1123, leaveTangent=0.1123, time=0.1941, value=0.5632}
         {arriveTangent=2.9135, interpMode=RCIM_Cubic, leaveTangent=2.9135, tangentMode=RCTM_User, time=0.1951, value=0.6755}
         <at 0.3 1>
         {arriveTangent=-0.3505, leaveTangent=-0.3505, time=0.7395, value=1}
         {arriveTangent=-0.0069, leaveTangent=-0.0069, time=0.7605, value=0.6495}
         {arriveTangent=-0.2804, leaveTangent=-0.2804, time=0.8367, value=0.6427}
         {arriveTangent=-0.0148, leaveTangent=-0.0148, time=0.8402, value=0.3623}
         {arriveTangent=-0.1832, leaveTangent=-0.1832, time=0.9216, value=0.3475}
         {arriveTangent=0.0012, leaveTangent=0.0012, time=0.9314, value=0.1644}
         {arriveTangent=-0.1656, leaveTangent=-0.1656, time=0.972, value=0.1656}
         <at 1 0 RCIM_Cubic>
        ]}}})
    Index 0 Param 2=&FloatFromCurve(FloatCurve=DataInterface() {propertyValues {Curve {keys [<at 0.4 2.5 RCIM_Cubic>, {arriveTangent=-10.9864, interpMode=RCIM_Cubic, leaveTangent=-10.9864, tangentMode=RCTM_User, time=1, value=0.05}]}}})
    Index 0 Param 3=&FloatFromCurve(FloatCurve=DataInterface() {propertyValues {Curve {keys [
         {arriveTangent=5.4303, interpMode=RCIM_Cubic, leaveTangent=5.4303, tangentMode=RCTM_User, time=0, value=-0.1}
         {arriveTangent=0.0133, interpMode=RCIM_Cubic, leaveTangent=0.0133, tangentMode=RCTM_User, time=0.02, value=0.0059}
         {arriveTangent=0.0056, interpMode=RCIM_Cubic, leaveTangent=0.0056, time=0.3, value=0.0104}
         {arriveTangent=-0.0288, interpMode=RCIM_Cubic, leaveTangent=-0.0288, time=0.7539, value=0.01}
         {arriveTangent=-0.2604, interpMode=RCIM_Cubic, leaveTangent=-0.2604, time=0.7654, value=-0.003}
         {arriveTangent=-0.2775, interpMode=RCIM_Cubic, leaveTangent=-0.2775, time=0.8371, value=-0.0117}
         {arriveTangent=-0.2326, interpMode=RCIM_Cubic, leaveTangent=-0.2326, time=0.8389, value=-0.0234}
         {arriveTangent=-0.2151, interpMode=RCIM_Cubic, leaveTangent=-0.2151, time=0.9168, value=-0.0302}
         {arriveTangent=-0.5984, interpMode=RCIM_Cubic, leaveTangent=-0.5984, time=0.9199, value=-0.0408}
         {interpMode=RCIM_Cubic, leaveTangent=-0.5311, time=1, value=-0.08}
        ]}}})
    Index 0 Param 4=&FloatFromCurve(FloatCurve=DataInterface() {propertyValues {Curve {keys [
         {arriveTangent=91.4522, interpMode=RCIM_Cubic, leaveTangent=91.4522, tangentMode=RCTM_User, time=0, value=0}
         {interpMode=RCIM_Cubic, tangentMode=RCTM_User, time=0.02, value=0.64}
         {interpMode=RCIM_Cubic, tangentMode=RCTM_User, time=0.8, value=0.64}
         {arriveTangent=-6.1885, interpMode=RCIM_Cubic, leaveTangent=-6.1885, tangentMode=RCTM_User, time=1, value=-0.25}
        ]}}})
    Module.Index0Param0Write=true
    Module.Index0Param1Write=true
    Module.Index0Param2Write=true
    Module.Index0Param3Write=true
   )
   &ScaleColor(
    Scale Alpha=&FloatFromCurve(FloatCurve=DataInterface() {propertyValues {Curve {keys [<at 0.95 1 RCIM_Cubic>,<at 1 0>]}}})
    Scale RGB=&VectorFromFloat(Value=&FloatFromCurve(FloatCurve=DataInterface() {propertyValues {Curve {keys [{arriveTangent=-81.4887, interpMode=RCIM_Cubic, leaveTangent=-81.4887, tangentMode=RCTM_User, time=0, value=5}, <at 0.15 1>]}}}))
    ScaleA=true
    ScaleRGB=true
   )
  ]
  rendererClasses [&NiagaraMeshRendererProperties]
  renderers [{rendererClass=&NiagaraMeshRendererProperties, rendererIndex=0}]
 }
 ElectricArcsResidual {
  emitterUpdateScript [
   &EmitterState(Life Cycle Mode=::NewEnumerator1, Loop Behavior=::NewEnumerator1, Loop Duration=1.25, UseLoopDelay=true)
   &SpawnBurst_Instantaneous(Spawn Count=6, Spawn Time=0.1, Use Spawn Probability=false, UseLoopCountLimit=false)
   &SpawnRate(
    SpawnRate=&Multiply_Float(
     A=30
     B=&FloatFromCurve(FloatCurve=DataInterface() {propertyValues {Curve {keys [<at 0 1 RCIM_Cubic>,<at 2 0>]}}}, CurveIndex=$Emitter.NormalizedLoopAge)
    )
    Use Spawn Probability=false
   )
  ]
  particleSpawnScript [
   &InitializeParticle(
    Lifetime=&RandomRangeFloat(Minimum=0.1, Maximum=0.25)
    Mesh Scale=&RandomRangeVector(Minimum=<50 50 50>, Maximum=<120 90 90>)
    Mesh Scale Mode=::NewEnumerator4
    UsePositionOffset=false
    Write Material Random=true
   )
   &InitialMeshOrientation(
    Mesh Orientation Mode=::NewEnumerator7
    Rotation=&RandomRangeVector(Minimum=<0 0 0>, Maximum=<1 0 0>)
    Use Rotation Vector=true
   )
   &ShapeLocation(
    Cylinder Height=20
    Cylinder Radius=250
    MaskSpawnGroup=false
    Override Seed=false
    Rotation Axis=<0 0 1>
    Rotation Mode=::NewEnumerator1
    Shape Primitive=::NewEnumerator1
    Surface Only=false
    UseOwnerScale=true
   )
   &AddVelocity(Velocity=&Multiply_VectorByFloat(Vector=&Multiply_Vector(A=$Particles.ShapeLocation.ShapeVector, B=<1 1 0>), Float=5))
   &Color(Color=&MakeLinearColorFromVectorAndFloat(
     "Vector (RGB)"=&RandomRangeVector(Minimum=<3.9751 0.4547 4>, Maximum=<0 6 4.9028>)
     "Float (Alpha)"=&ClampFloat(Float=1)
    ))
  ]
  particleUpdateScript [
   &ParticleState()
   &ScaleMeshSize(Scale Factor=&VectorFromCurve(VectorCurve=DataInterface() {propertyValues {
       XCurve {keys [<at 0.0033 0.5 RCIM_Constant>,<at 0.2054 0.7619 RCIM_Constant>,<at 0.4101 0.8575 RCIM_Constant>,<at 0.6967 1>]}
       YCurve {keys [<at 0.0033 0.6004 RCIM_Constant>,<at 0.2054 0.6716 RCIM_Constant>,<at 0.4101 0.3646 RCIM_Constant>,<at 0.6967 0.2468>]}
       ZCurve {keys [<at 0.0033 0.7767 RCIM_Constant>,<at 0.2054 0.5215 RCIM_Constant>,<at 0.4101 0.4457 RCIM_Constant>,<at 0.6967 0.6282>]}
      }}))
   &DynamicMaterialParameters(
    Index 0 Param 1=&RandomRangeFloat(Minimum=2, Maximum=3)
    Index 0 Param 2=&FloatFromCurve(FloatCurve=DataInterface() {propertyValues {Curve {keys [{arriveTangent=-1.8894, interpMode=RCIM_Cubic, time=0, value=1}, <at 1 -0.25>]}}})
    Index 0 Param 3=&RandomRangeFloat(Minimum=0.2, Maximum=2)
    Index 0 Param 4=&RandomRangeFloat(Minimum=0, Maximum=1)
    Module.Index0Param0Write=true
    Module.Index0Param1Write=true
    Module.Index0Param2Write=true
    Module.Index0Param3Write=true
   ) {module=DynamicMaterialParameters001}
   &ScaleColor(
    Scale RGB=&VectorFromFloat(Value=&FloatFromCurve(FloatCurve=DataInterface() {propertyValues {Curve {keys [<at 0 50 RCIM_Cubic>, {arriveTangent=-45, interpMode=RCIM_Cubic, leaveTangent=-45, time=0.5, value=30}, <at 1 5>]}}}))
    ScaleA=true
    ScaleRGB=true
   )
   &SolveForcesAndVelocity(Clamp Velocity=false, Limit Acceleration=false)
  ]
  rendererClasses [&NiagaraMeshRendererProperties]
  renderers [{rendererClass=&NiagaraMeshRendererProperties, rendererIndex=0}]
 }
 Endflare {
  emitterUpdateScript [
   &EmitterState(Life Cycle Mode=::NewEnumerator1, Loop Behavior=::NewEnumerator1, Loop Delay=2, Loop Duration=1, UseLoopDelay=true)
   &SpawnBurst_Instantaneous(Use Spawn Probability=false, UseLoopCountLimit=false)
  ]
  particleSpawnScript [
   &InitializeParticle(
    Lifetime=1.1
    Position=&ConvertVectorToPosition(Input Position=<0 0 350>, Conversion Operation:ENiagara_LWCConvertVecToPos=::NewEnumerator0)
    Position Mode=::NewEnumerator1
    Sprite Size=<12000 1500>
    Sprite Size Mode=::NewEnumerator4
    UsePositionOffset=false
    Write Material Random=true
   )
   &Color(Color=&MakeLinearColorFromVectorAndFloat("Vector (RGB)"=<0.4634 1.6791 2>, "Float (Alpha)"=&ClampFloat(Float=1)))
  ]
  particleUpdateScript [
   &ParticleState()
   &ScaleSpriteSize(
    Scale Factor=&Vector2DFromFloat(Value=&FloatFromCurve(FloatCurve=DataInterface() {propertyValues {Curve {keys [<at 0 1 RCIM_Cubic>, {interpMode=RCIM_Cubic, leaveTangent=-1.2932, time=1, value=0.2}]}}}))
    Scale Sprite Size Mode=::NewEnumerator2
   )
   &ScaleColor(
    Scale Alpha=&FloatFromCurve(FloatCurve=DataInterface() {propertyValues {Curve {keys [<at 0 0 RCIM_Cubic>,<at 0.6 1 RCIM_Cubic>,<at 1 0>]}}})
    Scale RGB=<2 2 2>
    ScaleA=true
    ScaleRGB=true
   )
   &DynamicMaterialParameters(
    Index 0 Param 1=&FloatFromCurve(FloatCurve=DataInterface() {propertyValues {Curve {keys [<at 0 0.15 RCIM_Cubic>,<at 1 0>]}}})
    Index 0 Param 2=5
    Index 0 Param 3=1
    Index 0 Param 4=1
    Module.Index0Param0Write=true
    Module.Index0Param1Write=true
    Module.Index0Param2Write=true
    Module.Index0Param3Write=true
   )
   &CameraOffset(Camera Offset Amount=400)
  ]
  rendererClasses [&NiagaraSpriteRendererProperties]
  renderers [{rendererClass=&NiagaraSpriteRendererProperties}]
 }
 Frags1 {
  emitterUpdateScript [
   &EmitterState(Life Cycle Mode=::NewEnumerator1, Loop Behavior=::NewEnumerator1, Loop Duration=1, UseLoopDelay=true)
   &SpawnBurst_Instantaneous(Spawn Count=12, Use Spawn Probability=false, UseLoopCountLimit=false)
   &SpawnBurst_Instantaneous(Spawn Count=12, Spawn Group=1, Use Spawn Probability=false, UseLoopCountLimit=false) {module=SpawnBurst_Instantaneous001}
   &SpawnBurst_Instantaneous(Spawn Count=12, Spawn Group=2, Use Spawn Probability=false, UseLoopCountLimit=false) {module=SpawnBurst_Instantaneous002}
  ]
  particleSpawnScript [
   /SetVariables(Particles.MeshIndex=$Emitter.SpawnGroup) {isSetParameters=true}
   &InitializeParticle(
    Color=<rgba 5 5 5 1>
    Lifetime=&RandomRangeFloat(Minimum=1.25, Maximum=2)
    Mesh Scale=&RandomRangeVector(Minimum=<5 5 5>, Maximum=<25 12 12>)
    Mesh Scale Mode=::NewEnumerator4
    UsePositionOffset=false
    Write Material Random=true
   )
   &ShapeLocation(
    Hemisphere Distribution=<1 0.5>
    MaskSpawnGroup=false
    Non Uniform Scale=<1 1 -1>
    Override Seed=false
    Sphere Radius=&RandomRangeFloat(Minimum=30, Maximum=40)
    UseOwnerScale=true
   )
   /SetVariables(Particles.MeshOrientation=&MakeQuatFromAxisAngle(Axis=&RandomRangeVector(Minimum=<0 0 0>, Maximum=<1 1 1>), AngleInDegrees=90)) {isSetParameters=true}
   &AddVelocity(Velocity=&Multiply_VectorByFloat(Vector=$Particles.ShapeLocation.ShapeVector, Float=&RandomRangeFloat(Minimum=70, Maximum=120))) {module=AddVelocity001}
   &AddVelocity(Velocity=&RandomRangeVector(Minimum=<0 0 -8000>, Maximum=<0 0 -1000>))
   /SetVariables(Particles.SpawnRotationVector=&RandomRangeVector(Minimum=<0.1 1.5 -0.1>, Maximum=<0.4 2 0.1>)) {isSetParameters=true}
   &AddRotationalVelocity(Rotation Rate=$StackContext.SpawnRotationVector)
  ]
  particleUpdateScript [
   &ParticleState()
   &Drag(
    Drag=&RandomRangeFloat(Minimum=1.2, Maximum=2.2, Evaluation Type=::NewEnumerator1)
    Use Linear Drag=true
    Use Rotational Drag=true
   )
   &UpdateMeshOrientation(Rotation Vector=$Particles.SpawnRotationVector) {enabled=false}
   &SolveForcesAndVelocity(Clamp Velocity=false, Limit Acceleration=false)
   &ScaleMeshSize(Scale Factor=&VectorFromCurve(VectorCurve=DataInterface() {propertyValues {
       XCurve {keys [<at 0.2 1 RCIM_Cubic>, {arriveTangent=-1.25, interpMode=RCIM_Cubic, leaveTangent=-1.25, time=0.7932, value=0.7509}, <at 1 0>]}
       YCurve {keys [<at 0.2 1 RCIM_Cubic>, {arriveTangent=-0.25, interpMode=RCIM_Cubic, leaveTangent=-0.25, time=0.7932, value=0.85}, <at 1 0.8>]}
       ZCurve {keys [<at 0.2 1 RCIM_Cubic>,<at 0.7932 1 RCIM_Cubic>,<at 1 1>]}
      }}))
   &ScaleColor(
    Scale Alpha=&FloatFromCurve(FloatCurve=DataInterface() {propertyValues {Curve {keys [<at 0.7 1 RCIM_Cubic>,<at 1 0>]}}})
    Scale RGB=&VectorFromFloat(Value=&FloatFromCurve(FloatCurve=DataInterface() {propertyValues {Curve {keys [<at 0.1 25 RCIM_Cubic>, {arriveTangent=-26.6667, interpMode=RCIM_Cubic, leaveTangent=-26.6667, time=0.1961, value=7.5192}, <at 1 1>]}}}))
    ScaleA=true
    ScaleRGB=true
   )
   &CameraOffset(Camera Offset Amount=-50)
  ]
  rendererClasses [&NiagaraMeshRendererProperties]
  renderers [{rendererClass=&NiagaraMeshRendererProperties, rendererIndex=0}]
 }
 Light {
  emitterUpdateScript [
   &EmitterState(
    Enable Distance Culling=true
    Life Cycle Mode=::NewEnumerator1
    Loop Behavior=::NewEnumerator1
    Loop Duration=1
    Module.bUseMaxDistance=true
    Module.bUseMinDistance=false
    Scalability Mode=::NewEnumerator1
    UseLoopDelay=true
   )
   &SpawnBurst_Instantaneous(Use Spawn Probability=false, UseLoopCountLimit=false)
  ]
  particleSpawnScript [&InitializeParticle(
    Color=<rgba 0.4634 2 1 1>
    Lifetime=2
    Sprite Size=<2000 2000>
    Sprite Size Mode=::NewEnumerator4
    UsePositionOffset=false
    Write Material Random=true
   )]
  particleUpdateScript [
   &ParticleState()
   /SetVariables(Particles.LightExponent=8) {isSetParameters=true}
   &ScaleColor(
    Scale Alpha=&FloatFromCurve(FloatCurve=DataInterface() {propertyValues {Curve {keys [<at 0 0.7 RCIM_Cubic>, {interpMode=RCIM_Cubic, leaveTangent=-1.3318, time=1, value=0}]}}})
    Scale RGB=&VectorFromFloat(Value=&FloatFromCurve(FloatCurve=DataInterface() {propertyValues {Curve {keys [<at 0 10 RCIM_Cubic>,<at 0.05 1>]}}}))
    ScaleA=true
    ScaleRGB=true
   )
   &CascadeConversion_LightProperties(LightBrightnessScale=32, LightColorScale=<1 1 1>, LightRadiusScale=1)
  ]
  rendererClasses [&NiagaraLightRendererProperties]
  renderers [{rendererClass=&NiagaraLightRendererProperties, rendererIndex=0}]
 }
 RoundFlare {
  emitterUpdateScript [
   &EmitterState(
    Enable Distance Culling=true
    Life Cycle Mode=::NewEnumerator1
    Loop Behavior=::NewEnumerator1
    Loop Duration=1
    MaxDistance=1000000
    Min Distance Response=::KillImmediately
    MinDistance=2000
    Module.bUseMaxDistance=false
    Module.bUseMinDistance=true
    Scalability Mode=::NewEnumerator1
    UseLoopDelay=true
   )
   &SpawnBurst_Instantaneous(Use Spawn Probability=false, UseLoopCountLimit=false)
  ]
  particleSpawnScript [&InitializeParticle(
    Color=<rgba 0.4634 1.6791 2 1>
    Lifetime=2
    Sprite Size=<14000 14000>
    Sprite Size Mode=::NewEnumerator4
    UsePositionOffset=false
    Write Material Random=true
   )]
  particleUpdateScript [
   &ParticleState()
   &ScaleSpriteSize(
    Scale Sprite Size Mode=::NewEnumerator0
    Uniform Scale Factor=&FloatFromCurve(FloatCurve=DataInterface() {propertyValues {Curve {keys [<at 0 1 RCIM_Cubic>, {arriveTangent=-1.8205, interpMode=RCIM_Cubic, leaveTangent=-1.8205, tangentMode=RCTM_User, time=1, value=0.2}]}}})
   )
   &ScaleColor(
    Scale Alpha=&FloatFromCurve(FloatCurve=DataInterface() {propertyValues {Curve {keys [<at 0 0.7 RCIM_Cubic>, {interpMode=RCIM_Cubic, leaveTangent=-1.3318, time=1, value=0}]}}})
    Scale RGB=&VectorFromFloat(Value=&FloatFromCurve(FloatCurve=DataInterface() {propertyValues {Curve {keys [<at 0 10 RCIM_Cubic>,<at 0.05 1>]}}}))
    ScaleA=true
    ScaleRGB=true
   )
   &DynamicMaterialParameters(
    Index 0 Param 1=&FloatFromCurve(FloatCurve=DataInterface() {propertyValues {Curve {keys [<at 0 1 RCIM_Cubic>,<at 0.8 0>]}}})
    Index 0 Param 2=2
    Index 0 Param 3=1
    Index 0 Param 4=1
    Module.Index0Param0Write=true
    Module.Index0Param1Write=true
    Module.Index0Param2Write=true
    Module.Index0Param3Write=true
   )
  ]
  rendererClasses [&NiagaraSpriteRendererProperties]
  renderers [{rendererClass=&NiagaraSpriteRendererProperties}]
 }
 ShockWave {
  emitterUpdateScript [
   &EmitterState(
    Enable Distance Culling=true
    Life Cycle Mode=::NewEnumerator1
    Loop Behavior=::NewEnumerator1
    Loop Duration=1
    MaxDistance=1000000
    Min Distance Response=::KillImmediately
    MinDistance=2000
    Module.bUseMaxDistance=false
    Module.bUseMinDistance=true
    Scalability Mode=::NewEnumerator1
    UseLoopDelay=true
   )
   &SpawnBurst_Instantaneous(Use Spawn Probability=false, UseLoopCountLimit=false)
  ]
  particleSpawnScript [
   &InitializeParticle(
    Color=<rgba 1 1 1 0.1>
    Lifetime=0.6
    Mesh Scale=<30 30 30>
    Mesh Scale Mode=::NewEnumerator4
    UsePositionOffset=false
    Write Material Random=true
   )
   &InitialMeshOrientation(Mesh Orientation Mode=::NewEnumerator2, Orientation Vector=<0 0 -1>, Use Rotation Vector=false)
  ]
  particleUpdateScript [
   &ParticleState()
   &ScaleMeshSize(Scale Factor=&VectorFromCurve(VectorCurve=DataInterface() {propertyValues {XCurve {keys [<at 0 0.2>,<at 1 8>]}, YCurve {keys [<at 0 0.2>,<at 1 8>]}, ZCurve {keys [<at 0 0.2>,<at 1 8>]}}}))
   &ScaleColor(
    Scale Alpha=&FloatFromCurve(FloatCurve=DataInterface() {propertyValues {Curve {keys [{arriveTangent=-2.1694, interpMode=RCIM_Cubic, time=0, value=1}, <at 1 0 RCIM_Cubic>]}}})
    ScaleA=true
    ScaleRGB=true
   )
   &CameraOffset(Camera Offset Amount=500)
  ]
  rendererClasses [&NiagaraMeshRendererProperties]
  renderers [{rendererClass=&NiagaraMeshRendererProperties, rendererIndex=0}]
 }
 Sparklies {
  simTarget=GPUComputeSim
  emitterUpdateScript [
   &EmitterState(
    Enable Distance Culling=true
    Life Cycle Mode=::NewEnumerator1
    Loop Behavior=::NewEnumerator1
    Loop Duration=0.1
    MaxDistance=1000000
    Min Distance Response=::KillImmediately
    MinDistance=2000
    Module.bUseMaxDistance=false
    Module.bUseMinDistance=true
    Scalability Mode=::NewEnumerator1
    UseLoopDelay=true
   )
   &SpawnBurst_Instantaneous(Spawn Count=50, Spawn Time=0.04, Use Spawn Probability=false, UseLoopCountLimit=false) {module=SpawnBurst_Instantaneous001}
   &SpawnBurst_Instantaneous(Spawn Count=50, Use Spawn Probability=false, UseLoopCountLimit=false)
   &SpawnRate(
    SpawnRate=&Multiply_Float(
     A=600
     B=&FloatFromCurve(
      FloatCurve=DataInterface() {propertyValues {Curve {keys [
          <at 0 1 RCIM_Cubic>
          {arriveTangent=-0.2543, interpMode=RCIM_Cubic, leaveTangent=-0.2543, time=1.2828, value=1.6362}
          {arriveTangent=-0.3703, interpMode=RCIM_Cubic, leaveTangent=-0.3703, time=2.4472, value=0.3776}
          {arriveTangent=-0.11, interpMode=RCIM_Cubic, leaveTangent=-0.11, time=3.6302, value=0.7671}
          {arriveTangent=0.2907, interpMode=RCIM_Cubic, leaveTangent=0.2907, time=5.034, value=0.0931}
          {arriveTangent=0.3058, interpMode=RCIM_Cubic, leaveTangent=0.3058, time=6.8459, value=1.702}
          <at 8 1>
         ]}}}
      CurveIndex=$Emitter.NormalizedLoopAge
     )
    )
    Use Spawn Probability=false
   )
  ]
  particleSpawnScript [
   &InitializeParticle(
    Lifetime=&RandomRangeFloat(Minimum=0.6, Maximum=3)
    Sprite Rotation Angle=&RandomRangeFloat(Minimum=0, Maximum=1)
    Sprite Rotation Mode=::NewEnumerator1
    Sprite Size=&Vector2DFromFloat(Value=&RandomRangeFloat(Minimum=100, Maximum=300))
    Sprite Size Mode=::NewEnumerator4
    UsePositionOffset=false
    Write Material Random=true
   )
   &ShapeLocation(
    Hemisphere Distribution=<1 0.5>
    MaskSpawnGroup=false
    Non Uniform Scale=<1 1 -1>
    Override Seed=false
    Sphere Radius=&RandomRangeFloat(Minimum=20, Maximum=80)
    UseOwnerScale=true
   )
   &AddVelocity(Velocity=&Multiply_VectorByFloat(Vector=$Particles.ShapeLocation.ShapeVector, Float=&RandomRangeFloat(Minimum=0, Maximum=50))) {module=AddVelocity001}
   &AddVelocity(Velocity=&RandomRangeVector(Minimum=<0 0 -12000>, Maximum=<0 0 -3000>))
   &Color(Color=&MakeLinearColorFromVectorAndFloat("Vector (RGB)"=<2.3177 4.968 5>, "Float (Alpha)"=&ClampFloat(Float=1)))
  ]
  particleUpdateScript [
   &ParticleState()
   &Drag(
    Drag=&RandomRangeFloat(Minimum=0.5, Maximum=2, Evaluation Type=::NewEnumerator1)
    Use Linear Drag=true
    Use Rotational Drag=true
   )
   &CurlNoiseForce(
    Noise Frequency=100
    Noise Strength=&FloatFromCurve(
     FloatCurve=DataInterface() {propertyValues {Curve {keys [
         {arriveTangent=4.1785, interpMode=RCIM_Cubic, leaveTangent=4.1785, tangentMode=RCTM_User, time=0, value=0}
         {arriveTangent=1.1179, interpMode=RCIM_Cubic, leaveTangent=1.1179, tangentMode=RCTM_User, time=0.2042, value=0.6228}
         {arriveTangent=0.0224, interpMode=RCIM_Cubic, leaveTangent=0.0224, tangentMode=RCTM_User, time=1, value=1}
        ]}}}
     Scale Curve=4000
    )
    Pan Noise=false
   )
   &SolveForcesAndVelocity(Clamp Velocity=false, Limit Acceleration=false)
   &SpriteRotationRate(Rotation Rate=&AngleConversion(
     Angle:NiagaraFloat=&RandomRangeFloat(Minimum=-1, Maximum=1)
     Angle Input:ENiagara_AngleInput=::NewEnumerator1
     Angle Output:ENiagara_AngleInput=::NewEnumerator0
    ))
   &ScaleSpriteSize(
    Scale Sprite Size Mode=::NewEnumerator0
    Uniform Scale Factor=&FloatFromCurve(FloatCurve=DataInterface() {propertyValues {Curve {keys [
         {arriveTangent=-0.4739, time=-0.0044, value=1}
         {arriveTangent=0.2957, interpMode=RCIM_Cubic, leaveTangent=0.2957, time=0.157, value=0.5261}
         {arriveTangent=1.0007, interpMode=RCIM_Cubic, leaveTangent=1.0007, time=0.1686, value=1.0512}
         {arriveTangent=-3.1557, interpMode=RCIM_Cubic, leaveTangent=-3.1557, time=0.1876, value=0.5567}
         {arriveTangent=1.2911, interpMode=RCIM_Cubic, leaveTangent=1.2911, time=0.3841, value=0.3711}
         {arriveTangent=0.7645, interpMode=RCIM_Cubic, leaveTangent=0.7645, time=0.4102, value=0.8441}
         {arriveTangent=-2.1497, interpMode=RCIM_Cubic, leaveTangent=-2.1497, time=0.4322, value=0.4079}
         {arriveTangent=0.7294, interpMode=RCIM_Cubic, leaveTangent=0.7294, time=0.689, value=0.2446}
         {arriveTangent=-0.8294, interpMode=RCIM_Cubic, leaveTangent=-0.8294, time=0.7007, value=0.6037}
         {arriveTangent=-2.017, interpMode=RCIM_Cubic, leaveTangent=-2.017, time=0.7161, value=0.2222}
         <at 1 0>
        ]}}})
   )
   &ScaleColor(
    Scale RGB=&VectorFromFloat(Value=&FloatFromCurve(FloatCurve=DataInterface() {propertyValues {Curve {keys [<at 0 2 RCIM_Cubic>,<at 0.4 1>]}}}))
    ScaleA=true
    ScaleRGB=true
   )
  ]
  rendererClasses [&NiagaraSpriteRendererProperties]
  renderers [{rendererClass=&NiagaraSpriteRendererProperties}]
 }
 WideFlare {
  emitterUpdateScript [
   &EmitterState(
    Enable Distance Culling=true
    Life Cycle Mode=::NewEnumerator1
    Loop Behavior=::NewEnumerator1
    Loop Duration=1
    MaxDistance=1000000
    Min Distance Response=::KillImmediately
    MinDistance=2000
    Module.bUseMaxDistance=false
    Module.bUseMinDistance=true
    Scalability Mode=::NewEnumerator1
    UseLoopDelay=true
   )
   &SpawnBurst_Instantaneous(Use Spawn Probability=false, UseLoopCountLimit=false)
  ]
  particleSpawnScript [&InitializeParticle(
    Color=<rgba 0.4634 1.6791 2 1>
    Lifetime=0.8
    Sprite Size=<26000 500>
    Sprite Size Mode=::NewEnumerator4
    UsePositionOffset=false
    Write Material Random=true
   )]
  particleUpdateScript [
   &ParticleState()
   &ScaleColor(
    Scale Alpha=&FloatFromCurve(FloatCurve=DataInterface() {propertyValues {Curve {keys [<at 0.0503 1 RCIM_Cubic>,<at 1 0>]}}})
    Scale RGB=<5 5 5>
    ScaleA=true
    ScaleRGB=true
   )
   &DynamicMaterialParameters(
    Index 0 Param 1=&FloatFromCurve(FloatCurve=DataInterface() {propertyValues {Curve {keys [<at 0 1 RCIM_Cubic>,<at 0.8 0>]}}})
    Index 0 Param 2=2
    Index 0 Param 3=1
    Index 0 Param 4=1
    Module.Index0Param0Write=true
    Module.Index0Param1Write=true
    Module.Index0Param2Write=true
    Module.Index0Param3Write=true
   )
  ]
  rendererClasses [&NiagaraSpriteRendererProperties]
  renderers [{rendererClass=&NiagaraSpriteRendererProperties}]
 }
}
summary {
 systemName=NS_Geode_CrackExit_Rift_Short
 emitters [
  {emitterName=CrackSpline, rendererClasses [&NiagaraMeshRendererProperties]}
  {emitterName=CrackPortal, rendererClasses [&NiagaraMeshRendererProperties]}
  {emitterName=ElectricArcsResidual, rendererClasses [&NiagaraMeshRendererProperties]}
  {emitterName=WideFlare, rendererClasses [&NiagaraSpriteRendererProperties]}
  {emitterName=RoundFlare, rendererClasses [&NiagaraSpriteRendererProperties]}
  {emitterName=Light, rendererClasses [&NiagaraLightRendererProperties]}
  {emitterName=Endflare, rendererClasses [&NiagaraSpriteRendererProperties]}
  {emitterName=ShockWave, rendererClasses [&NiagaraMeshRendererProperties]}
  {emitterName=Corona, rendererClasses [&NiagaraSpriteRendererProperties]}
  {emitterName=Sparklies, simTarget=GPUComputeSim, rendererClasses [&NiagaraSpriteRendererProperties]}
  {emitterName=Frags1, rendererClasses [&NiagaraMeshRendererProperties]}
 ]
 userVariables []
}
