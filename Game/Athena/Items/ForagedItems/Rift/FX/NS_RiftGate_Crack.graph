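# Niagara effect for the Rift Gate crack: crack and glow meshes, drifting fragments, rays, flares, glints, red lightning and a downward pointer ribbon.
# System.ControlCurve drives growth across emitters, scaling counts, spread, size, opacity and color, while SyncWaveform pulses colors.
# An expanding region mask tints Fragments and culls Fragment_Lighting particles spawned from them; Glints spawn from Tiny_Fragments.
# User parameters set scale and pointer height; time of day dims some layers.
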NiagaraSystem /Game/Athena/Items/ForagedItems/Rift/FX/NS_RiftGate_Crack
dependencies {
 usedDataInterfaces [
  &NiagaraDataInterfaceParticleRead
  &NiagaraDataInterfaceCurve
  &NiagaraDataInterfaceColorCurve
  &NiagaraDataInterfaceMeshRendererInfo
  &NiagaraDataInterfaceCamera
  /Script/Niagara.NiagaraDataInterfaceOcclusion
 ]
 usedDynamicInputs [&ClampFloat, &Waveform]
 usedModules [
  &self:SystemSpawnScript.SystemScriptSource.SystemScriptGraph.NiagaraNodeAssignment_3.SetVariables_#1
  &SystemState
  &self:SystemSpawnScript.SystemScriptSource.SystemScriptGraph.NiagaraNodeAssignment_0.SetVariables_#2
  &self:SystemSpawnScript.SystemScriptSource.SystemScriptGraph.NiagaraNodeAssignment_2.SetVariables_#3
  &self:SystemSpawnScript.SystemScriptSource.SystemScriptGraph.NiagaraNodeAssignment_1.SetVariables_#4
 ]
}
emitters {
 CenterFlare {
  bEnabled=false
  simTarget=GPUComputeSim
  emitterUpdateScript [
   &EmitterState(Life Cycle Mode=::NewEnumerator1, Loop Behavior=::NewEnumerator1, Loop Duration Mode=::NewEnumerator1, UseLoopDelay=false)
   &SpawnBurst_Instantaneous(Use Spawn Probability=false, UseLoopCountLimit=false)
  ]
  particleSpawnScript [&InitializeParticle(
    Sprite Size Mode=::NewEnumerator3
    Uniform Sprite Size=&Multiply_Float(A=15000, B=$User.GlobalScale)
    UsePositionOffset=false
    Write Material Random=true
   )]
  particleUpdateScript [
   &ParticleState(Kill Particles When Lifetime Has Elapsed=false, Let Infinitely Lived Particles Die When Emitter Deactivates=true)
   &ScaleSpriteSize(
    Uniform Curve Index=$System.ControlCurve
    Uniform Curve Sprite Scale=DataInterface() {propertyValues {Curve {keys [{tangentMode=RCTM_User, time=0, value=1}, {tangentMode=RCTM_User, time=1, value=2}]}}}
   )
   &self:MaskInScreenSpace() {module=MaskInScreenSpace}
   &self:ScratchModule(NumberOfRings=3, OcclusionSize=5000, SamplesPerRing=8) {module=ScratchModule}
   &ScaleColor(
    Curve Index=$Transient.DistanceFromScreenEdge
    Linear Color Curve=DataInterface() {propertyValues {
      AlphaCurve {keys [<at 0 1>]}
      BlueCurve {keys [<at 0 0.5>,<at 0.8973 1.4987>]}
      GreenCurve {keys [<at 0 0.3689>,<at 0.8973 1.3582>]}
      RedCurve {keys [<at 0 0.4494>,<at 0.8973 2>]}
     }}
    Scale Mode=::NewEnumerator2
   ) {module=ScaleColor002}
   /SetVariables(Particles.Position=&AddVectorToPosition(Position=$Engine.Emitter.SimulationPosition, Vector=<0 0 -18000>)) {isSetParameters=true}
   &ScaleColor(
    Scale Alpha=$Particles.Visibility
    Scale RGB=&MakeVectorFromLinearColorRGB(Color=&MultiplyLinearColor(
      Linear Color=&LerpLinearColors(StartColor=<rgba 0.7344 0.0365 0 1>, EndColor=<rgba 0.1052 0.3566 0.974 1>, LerpFactor=$System.ControlCurve)
      Scale Factor=&MakeFloat4FromVectorAndFloat("Vector (RGB)"=&VectorFromFloat(Value=10), "Float (Alpha)"=1)
     ))
    ScaleA=true
    ScaleRGB=true
   )
   &ScaleColor(
    Color Value To Scale=$Particles.Color
    Scale Alpha=&Multiply_Float(A=$User.GlobalOpacity, B=1)
    Scale RGB=&Multiply_Vector(A=&MakeVectorFromLinearColorRGB(Color=$Particles.AtmosphereColor), B=<1 1 1>)
    ScaleA=true
    ScaleRGB=true
   ) {enabled=false, module=ScaleColor001}
   /SetVariables(Particles.CrossFlareSize:Vector2f=&Multiply_Vector2D_ByFloat(
     Float=&Lerp_Float(A=1, B=2, Alpha=$System.ControlCurve)
     Vector2D=&Multiply_Vector2D(
      A=$Particles.Initial.SpriteSize
      B=&MakeVector2D(
       X=&FloatFromCurve(FloatCurve=DataInterface() {propertyValues {Curve {keys [<at 0 2>,<at 1 1>]}}}, CurveIndex=$Transient.DistanceFromScreenEdge)
       Y=1
      )
     )
    )) {isSetParameters=true}
  ]
  rendererClasses [&NiagaraSpriteRendererProperties]
  renderers [
   {rendererClass=&NiagaraSpriteRendererProperties}
   {rendererClass=&NiagaraSpriteRendererProperties, rendererIndex=1}
  ]
 }
 Crack {
  emitterUpdateScript [
   &EmitterState(Life Cycle Mode=::NewEnumerator1, Loop Behavior=::NewEnumerator1, Loop Duration Mode=::NewEnumerator1, UseLoopDelay=false)
   &SpawnBurst_Instantaneous(Spawn Count=1, Use Spawn Probability=false)
   &SpawnBurst_Instantaneous(Spawn Group=1, Use Spawn Probability=false, UseLoopCountLimit=false) {module=SpawnBurst_Instantaneous001}
   /SetVariables(
    Emitter.CorruptionControl:NiagaraFloat=&Lerp_Float(A=0, B=1.1379, Alpha=$System.ControlCurve)
    Emitter.CorruptionControl_B:NiagaraFloat=&Lerp_Float(A=0, B=1.4174, Alpha=$System.ControlCurve)
   ) {isSetParameters=true}
  ]
  particleSpawnScript [
   /SetVariables(Particles.VisibilityTag=&SpawnGroup()) {isSetParameters=true}
   &InitializeParticle(
    Lifetime=2
    Mesh Scale=&Multiply_VectorByFloat(
     Vector=&MaskVectorBySpawnGroup(Masked Vector=<1 750 1600>, Passthrough Vector=<1 450 900>)
     Float=$User.GlobalScale
    )
    Mesh Scale Mode=::NewEnumerator4
    UsePositionOffset=false
    Write Material Random=true
   )
   &InitialMeshOrientation(
    Mesh Axis To Orient=<0 0 1>
    Mesh Orientation Mode=::NewEnumerator2
    Rotation=&MaskVectorBySpawnGroup(Masked Vector=<-0.8042 0 0>, Passthrough Vector=<0.3162 0 0>)
    Use Rotation Vector=true
   )
  ]
  particleUpdateScript [
   &ParticleState(Kill Particles When Lifetime Has Elapsed=false)
   &DynamicMaterialParameters(
    Index 0 Param 1=&MaskFloatBySpawnGroup(
     Masked Float=&Lerp_Float(A=0.075, B=1, Alpha=$System.ControlCurve)
     Passthrough Float=&Lerp_Float(A=0.05, B=0.5, Alpha=$System.ControlCurve)
    )
    Index 0 Param 2=&Multiply_Float(A=60, B=$User.GlobalScale)
    Index 0 Param 3=&Multiply_Float(A=$Engine.Time, B=0.0125)
    Index 0 Param 4=1
    Index 1 Param 1=1
    Module.Param0WriteEnabled=true
    Module.Param0WriteEnabled1=true
    Module.Param1WriteEnabled=true
    Module.Param1WriteEnabled1=false
    Module.Param2WriteEnabled=true
    Module.Param2WriteEnabled1=false
    Module.Param3WriteEnabled=true
    Module.Param3WriteEnabled1=false
    Write Parameter Index 1=true
   )
  ]
  rendererClasses [&NiagaraMeshRendererProperties]
  renderers [
   {rendererClass=&NiagaraMeshRendererProperties, rendererIndex=0}
   {rendererClass=&NiagaraMeshRendererProperties, rendererIndex=1}
  ]
 }
 Energy_Glints {
  simTarget=GPUComputeSim
  emitterUpdateScript [
   &EmitterState()
   &SpawnRate(
    SpawnRate=&MakeCustomFloatFromBool(
     Bool=&SetBoolByFloatComparison(
      A=&FloatFromCurve(FloatCurve=DataInterface() {propertyValues {Curve {keys [<at 0 1>,<at 1 0>]}}}, CurveIndex=$System.ControlCurve)
      B=0.6
      Comparison Type=::NewEnumerator2
     )
     True Float=&Multiply_Float(
      A=50
      B=&Lerp_Float(
       A=0
       B=1
       Alpha=&FloatFromCurve(FloatCurve=DataInterface() {propertyValues {bOptimizeLUT=false, Curve {keys [<at 0 1>,<at 1 0>]}}}, CurveIndex=$System.ControlCurve)
      )
     )
     False Float=0
    )
    Use Spawn Probability=false
   )
   /SetVariables(
    Emitter.EndColor:LinearColor=&MultiplyLinearColor(
     Linear Color=&LerpLinearColors(
      StartColor=<rgba 0.3073 0.0156 0 1>
      EndColor=<rgba 0.2212 0.1709 0.974 1>
      LerpFactor=&FloatFromCurve(FloatCurve=DataInterface() {propertyValues {bOptimizeLUT=false, Curve {keys [<at 0 1>,<at 1 0>]}}}, CurveIndex=$System.ControlCurve)
     )
     Scale Factor=&MakeFloat4FromVectorAndFloat("Vector (RGB)"=&VectorFromFloat(Value=15), "Float (Alpha)"=1)
    )
    Emitter.StartColor:LinearColor=&MultiplyLinearColor(
     Linear Color=&LerpLinearColors(
      StartColor=<rgba 0.7344 0.0365 0 1>
      EndColor=<rgba 0.1052 0.3566 0.974 1>
      LerpFactor=&FloatFromCurve(FloatCurve=DataInterface() {propertyValues {bOptimizeLUT=false, Curve {keys [<at 0 1>,<at 1 0>]}}}, CurveIndex=$System.ControlCurve)
     )
     Scale Factor=&MakeFloat4FromVectorAndFloat("Vector (RGB)"=&VectorFromFloat(Value=18), "Float (Alpha)"=1)
    )
   ) {isSetParameters=true}
  ]
  particleSpawnScript [
   &InitializeParticle(
    Lifetime Max=8
    Lifetime Min=3
    Lifetime Mode=::NewEnumerator1
    Sprite Rotation Mode=::NewEnumerator2
    Sprite Size Mode=::NewEnumerator1
    Uniform Sprite Size Max=1600
    Uniform Sprite Size Min=1000
    UsePositionOffset=false
    Write Material Random=true
   )
   &ShapeLocation(
    Hemisphere Distribution=<1 0.5>
    MaskSpawnGroup=false
    Offset=<0 0 -15000>
    Offset Mode=::NewEnumerator0
    Override Seed=false
    Rotation Axis=<0 0 -1>
    Rotation Mode=::NewEnumerator1
    Sphere Radius=5000
    UseOwnerScale=true
   )
   &AddVelocity(Velocity=&Multiply_VectorByFloat(Vector=$StackContext.ShapeLocation.ShapeNormal, Float=&RandomRangeFloat(Minimum=1500, Maximum=4500)))
  ]
  particleUpdateScript [
   &ParticleState()
   &ScaleSpriteSize(Uniform Curve Sprite Scale=DataInterface() {propertyValues {Curve {keys [
        {arriveTangent=-3.967, interpMode=RCIM_Cubic, leaveTangent=-3.967, tangentMode=RCTM_User, time=0, value=1}
        {arriveTangent=-0.1447, interpMode=RCIM_Cubic, leaveTangent=-0.1447, tangentMode=RCTM_User, time=0.157, value=0.5261}
        {arriveTangentWeight=0.0055, interpMode=RCIM_Cubic, leaveTangentWeight=0.0089, tangentMode=RCTM_User, time=0.1686, value=1.0512}
        {arriveTangent=-2.2925, interpMode=RCIM_Cubic, leaveTangent=-2.2925, tangentMode=RCTM_User, time=0.1876, value=0.5567}
        {arriveTangent=-0.1397, interpMode=RCIM_Cubic, leaveTangent=-0.1397, tangentMode=RCTM_User, time=0.3841, value=0.3711}
        {arriveTangentWeight=0.0109, interpMode=RCIM_Cubic, leaveTangentWeight=0.0093, tangentMode=RCTM_User, time=0.4102, value=0.8441}
        {arriveTangent=-1.868, interpMode=RCIM_Cubic, leaveTangent=-1.868, tangentMode=RCTM_User, time=0.4322, value=0.408}
        {arriveTangent=0.6076, interpMode=RCIM_Cubic, leaveTangent=0.6076, tangentMode=RCTM_User, time=0.689, value=0.2446}
        {arriveTangentWeight=0.005, interpMode=RCIM_Cubic, leaveTangentWeight=0.0067, tangentMode=RCTM_User, time=0.7007, value=0.6037}
        {arriveTangent=-2.017, interpMode=RCIM_Cubic, leaveTangent=-2.017, time=0.7161, value=0.2222}
        <at 1 0 RCIM_Cubic>
       ]}}})
   &GravityForce(Gravity=<0 0 -1960>)
   &VectorNoiseForce(Force Amount=2000, Module.Use Random Noise Time Interval=false)
   &SolveForcesAndVelocity(Clamp Velocity=false, Limit Acceleration=false)
   &ScaleColor(
    Scale RGB=&MakeVectorFromLinearColorRGB(Color=&LerpLinearColors(StartColor=$Emitter.StartColor, EndColor=$Emitter.EndColor, LerpFactor=$System.SyncWaveform))
    ScaleA=true
    ScaleRGB=true
   ) {module=ScaleColor001}
   &ScaleColor(
    Scale RGB=&VectorFromFloat(Value=&FloatFromCurve(
      FloatCurve=DataInterface() {propertyValues {Curve {keys [
          {arriveTangent=-3.967, interpMode=RCIM_Cubic, leaveTangent=-3.967, tangentMode=RCTM_User, time=0, value=1}
          {arriveTangent=-0.1447, interpMode=RCIM_Cubic, leaveTangent=-0.1447, tangentMode=RCTM_User, time=0.157, value=0.5261}
          {arriveTangentWeight=0.0055, interpMode=RCIM_Cubic, leaveTangentWeight=0.0089, tangentMode=RCTM_User, time=0.1686, value=1.0512}
          {arriveTangent=-2.2925, interpMode=RCIM_Cubic, leaveTangent=-2.2925, tangentMode=RCTM_User, time=0.1876, value=0.5567}
          {arriveTangent=-0.1397, interpMode=RCIM_Cubic, leaveTangent=-0.1397, tangentMode=RCTM_User, time=0.3841, value=0.3711}
          {arriveTangentWeight=0.0109, interpMode=RCIM_Cubic, leaveTangentWeight=0.0093, tangentMode=RCTM_User, time=0.4102, value=0.8441}
          {arriveTangent=-1.868, interpMode=RCIM_Cubic, leaveTangent=-1.868, tangentMode=RCTM_User, time=0.4322, value=0.408}
          {arriveTangent=0.6076, interpMode=RCIM_Cubic, leaveTangent=0.6076, tangentMode=RCTM_User, time=0.689, value=0.2446}
          {arriveTangentWeight=0.005, interpMode=RCIM_Cubic, leaveTangentWeight=0.0067, tangentMode=RCTM_User, time=0.7007, value=0.6037}
          {arriveTangent=-2.017, interpMode=RCIM_Cubic, leaveTangent=-2.017, time=0.7161, value=0.2222}
          <at 1 0 RCIM_Cubic>
         ]}}}
      Scale Curve=1500
     ))
    ScaleA=true
    ScaleRGB=true
   )
   &DynamicMaterialParameters(
    Index 0 Param 1=0.25
    Index 0 Param 2=1
    Module.Param0WriteEnabled=true
    Module.Param1WriteEnabled=true
    Module.Param2WriteEnabled=true
    Module.Param3WriteEnabled=true
   )
   /SetVariables(Particles.SharpCrossSize:Vector2f=&Multiply_Vector2D_ByFloat(
     Float=&RandomRangeFloat(Minimum=1, Maximum=3, Evaluation Type=::NewEnumerator1)
     Vector2D=$Particles.SpriteSize
    )) {isSetParameters=true}
  ]
  rendererClasses [&NiagaraSpriteRendererProperties]
  renderers [
   {rendererClass=&NiagaraSpriteRendererProperties}
   {rendererClass=&NiagaraSpriteRendererProperties, rendererIndex=1}
  ]
 }
 Fragment_Lighting {
  emitterUpdateScript [
   &EmitterState()
   &SpawnParticlesFromOtherEmitter(
    Attribute Reader=DataInterface() {propertyValues {EmitterBinding {bindingMode=Other, emitterName=Fragments}}}
    Calculate Spawn Rate Per Particle=false
    Maximum Age Spawn Threshold=0.333
    Spawn Rate=&Multiply_Float(
     A=&RandomRangeFloat(Minimum=3, Maximum=8)
     B=&Lerp_Float(A=1, B=0, Alpha=$System.ControlCurve)
    )
    Use Maximum Age=true
    Use Minimum Age=false
   )
  ]
  particleSpawnScript [
   &InitializeParticle(
    Color=<rgba 1 0.2267 0.2 1>
    Lifetime Max=0.15
    Lifetime Min=0.12
    Lifetime Mode=::NewEnumerator1
    Mesh Scale Mode=::NewEnumerator3
    Mesh Uniform Scale=750
    UsePositionOffset=false
    Write Material Random=true
   )
   &SampleParticlesFromOtherEmitter(Mesh Orientation Sampling:ENiagara_DirectReadSamplingMode=::NewEnumerator0)
   &RegionMask(
    Offset:Vector3f=<0 0 -15000>
    Region Falloff=0
    Region Radius:NiagaraFloat=&Lerp_Float(A=0, B=75000, Alpha=$System.ControlCurve)
    UseOffset=true
    UseRotation=false
   )
   &self:Kill_Outside_Region(Region:NiagaraFloat=$Output.RegionMask.RegionMask) {module=Kill_Outside_Region}
  ]
  particleUpdateScript [
   &ParticleState()
   &ScaleMeshSize(Scale Factor=&VectorFromFloat(Value=&FloatFromCurve(FloatCurve=DataInterface() {propertyValues {Curve {keys [{arriveTangent=3.8738, interpMode=RCIM_Cubic, leaveTangent=3.8738, tangentMode=RCTM_User, time=0, value=0}, <at 1 1 RCIM_Cubic>]}}})))
   &ScaleColor(
    Scale Alpha=&Multiply_Float(
     A=&Lerp_Float(A=0.5, B=0.8, Alpha=&RandomRangeFloat(Minimum=0, Maximum=1))
     B=&MakeFloatFromVector(
      Vector=&FN_Select_Vector_from_Time_Of_Day(
       Morning=&VectorFromFloat(Value=1)
       Day=&VectorFromFloat(Value=0.85)
       Evening=&VectorFromFloat(Value=1)
       Night=&VectorFromFloat(Value=0.25)
      )
      Channel=::NewEnumerator0
     )
    )
    Scale RGB=&VectorFromFloat(Value=&FloatFromCurve(
      FloatCurve=DataInterface() {propertyValues {Curve {keys [
          {arriveTangent=51.9074, interpMode=RCIM_Cubic, leaveTangent=51.9074, tangentMode=RCTM_User, time=0, value=0}
          {arriveTangentWeight=0.1274, interpMode=RCIM_Cubic, leaveTangentWeight=0.7778, tangentMode=RCTM_User, time=0.1408, value=5}
          {arriveTangent=-0.0405, interpMode=RCIM_Cubic, leaveTangent=-0.0405, tangentMode=RCTM_User, time=1, value=0}
         ]}}}
      Scale Curve=50
     ))
    ScaleA=true
    ScaleRGB=true
   )
   &SolveForcesAndVelocity(Clamp Velocity=false, Limit Acceleration=false)
   &DynamicMaterialParameters(
    Index 0 Param 1=&FloatFromCurve(FloatCurve=DataInterface() {propertyValues {Curve {keys [<at 0 1>,<at 1 0>]}}})
    Index 0 Param 2=&FloatFromCurve(
     FloatCurve=DataInterface() {propertyValues {Curve {keys [
         <at 0 1 RCIM_Cubic>
         {arriveTangent=0.987, interpMode=RCIM_Cubic, leaveTangent=0.987, tangentMode=RCTM_User, time=0.4992, value=1}
         <at 0.571 1 RCIM_Cubic>
         <at 1 1>
        ]}}}
     Scale Curve=5000
    )
    Index 0 Param 3=&FloatFromCurve(
     FloatCurve=DataInterface() {propertyValues {Curve {keys [
         {interpMode=RCIM_Cubic, tangentMode=RCTM_User, time=0, value=1}
         {interpMode=RCIM_Cubic, tangentMode=RCTM_User, time=0.2, value=0.25}
         {interpMode=RCIM_Cubic, tangentMode=RCTM_User, time=0.5, value=1.375}
         {interpMode=RCIM_Cubic, tangentMode=RCTM_User, time=1, value=2}
        ]}}}
     Scale Curve=&RandomRangeFloat(Minimum=0.4, Maximum=0.6)
    )
    Module.Param0WriteEnabled=true
    Module.Param1WriteEnabled=true
    Module.Param2WriteEnabled=true
    Module.Param3WriteEnabled=true
   )
  ]
  rendererClasses [&NiagaraMeshRendererProperties]
  renderers [{rendererClass=&NiagaraMeshRendererProperties, rendererIndex=0}]
 }
 Fragments {
  emitterUpdateScript [
   &EmitterState(Life Cycle Mode=::NewEnumerator1, Loop Behavior=::NewEnumerator1, Loop Duration Mode=::NewEnumerator1, UseLoopDelay=false)
   &SpawnBurst_Instantaneous(
    Spawn Count=&MakeIntFromFloat(Float=&Lerp_Float(A=5, B=17, Alpha=$System.ControlCurve))
    Use Spawn Probability=false
    UseLoopCountLimit=false
   )
  ]
  particleSpawnScript [
   /SetVariables(Particles.RandCurve:NiagaraFloat=&FloatFromCurve(
     FloatCurve=DataInterface() {propertyValues {Curve {keys [
         {interpMode=RCIM_Cubic, tangentMode=RCTM_User, time=0, value=0}
         {arriveTangent=0.6613, interpMode=RCIM_Cubic, leaveTangent=0.6613, tangentMode=RCTM_User, time=0.7, value=0.2}
         {arriveTangentWeight=0.4215, interpMode=RCIM_Cubic, tangentMode=RCTM_User, time=1, value=1}
        ]}}}
     CurveIndex=&RandomRangeFloat(Minimum=0, Maximum=1)
    )) {isSetParameters=true, module=SetVariables_#5}
   &InitializeParticle(
    Color Mode=::NewEnumerator0
    Lifetime=5
    Mesh Renderer Array Visibility Mode=::NewEnumerator3
    Mesh Renderer Info=DataInterface() {propertyValues {MeshRenderer=&self:Fragments_5.NiagaraMeshRendererProperties_0}}
    Mesh Renderer Random Seed=80678
    Mesh Scale Mode=::NewEnumerator3
    Mesh Uniform Scale=&Multiply_Float(A=&Lerp_Float(A=200, B=300, Alpha=$Particles.RandCurve), B=$User.GlobalScale)
    Sprite Size Mode=::NewEnumerator3
    Uniform Sprite Size=15000
    UsePositionOffset=false
    Write Material Random=true
   )
   &ShapeLocation(
    Cone Angle:NiagaraFloat=&Lerp_Float(A=62, B=125, Alpha=$System.ControlCurve)
    Cone Inner Angle=10
    Cone Length=-90000
    Fixed Random Seed=true
    MaskSpawnGroup=false
    Offset=<0 0 -3334.5359>
    Offset Mode=::NewEnumerator0
    Override Seed=true
    Random Seed:NiagaraInt32=&Add_Integer(A=81155, B=&ReturnExecIndex())
    Shape Primitive=::NewEnumerator5
    UseOwnerScale=true
   )
   &InitialMeshOrientation(
    Mesh Orientation Mode=::NewEnumerator2
    Orientation Vector=&RandomRangeVector(Minimum=<-0.5 -0.5 -1>, Maximum=<0.5 0.5 -1>)
    Rotation=&RandomRangeVector(
     Minimum=<0 0 0>
     Maximum=<0 0 1>
     Randomness Mode=::NewEnumerator1
     Random Seed=&Add_Integer(A=80678, B=&ReturnExecIndex())
     Fixed Random Seed=true
    )
    Use Rotation Vector=true
   )
  ]
  particleUpdateScript [
   &ParticleState(Kill Particles When Lifetime Has Elapsed=false)
   &UpdateMeshOrientation(
    Rotation Rate=&RandomRangeFloat(
     Minimum=0.01
     Maximum=0.02
     Randomness Mode=::NewEnumerator1
     Random Seed=&Add_Integer(A=80678, B=&ReturnExecIndex())
    )
    Rotation Vector=&RandomRangeVector(
     Minimum=<0 0 0>
     Maximum=<0 0 1>
     Randomness Mode=::NewEnumerator1
     Random Seed=&Add_Integer(A=80678, B=&ReturnExecIndex())
     Fixed Random Seed=true
    )
   )
   &RegionMask(
    Offset:Vector3f=<0 0 -15000>
    Region Falloff=0
    Region Radius:NiagaraFloat=&Lerp_Float(A=0, B=75000, Alpha=$System.ControlCurve)
    UseOffset=true
    UseRotation=false
   )
   &ScaleMeshSize(Scale Factor=&VectorFromFloat(Value=&Lerp_Float(A=1, B=1, Alpha=$Output.RegionMask.RegionMask)))
   &DynamicMaterialParameters(
    Index 0 Param 4=&Lerp_Float(A=&Lerp_Float(A=2, B=10, Alpha=$Output.RegionMask.RegionMask), B=3, Alpha=$System.ControlCurve)
    Module.Param0WriteEnabled=false
    Module.Param1WriteEnabled=false
    Module.Param2WriteEnabled=false
    Module.Param3WriteEnabled=true
   )
   &ScaleColor(
    Scale Alpha=&Lerp_Float(A=1, B=2, Alpha=$Output.RegionMask.RegionMask)
    Scale RGB=&MakeVectorFromLinearColorRGB(Color=&LerpLinearColors(StartColor=<rgba 1 1 1 1>, EndColor=<rgba 0.3073 0.0156 0 1>, LerpFactor=$Output.RegionMask.RegionMask))
    ScaleA=true
    ScaleRGB=true
   )
   &SolveForcesAndVelocity(Clamp Velocity=false, Limit Acceleration=false)
  ]
  rendererClasses [&NiagaraMeshRendererProperties]
  renderers [{rendererClass=&NiagaraMeshRendererProperties, rendererIndex=0}]
 }
 Glints {
  emitterUpdateScript [
   &EmitterState()
   &SpawnParticlesFromOtherEmitter(
    Attribute Reader=DataInterface() {propertyValues {EmitterBinding {bindingMode=Other, emitterName=Tiny_Fragments}}}
    Calculate Spawn Rate Per Particle=false
    Maximum Age Spawn Threshold=0.333
    Spawn Rate=&Lerp_Float(A=1, B=5, Alpha=$System.ControlCurve)
    Use Maximum Age=true
    Use Minimum Age=false
   )
  ]
  particleSpawnScript [
   /SetVariables(Particles.RandCurve:NiagaraFloat=&FloatFromCurve(
     FloatCurve=DataInterface() {propertyValues {Curve {keys [
         {interpMode=RCIM_Cubic, tangentMode=RCTM_User, time=0, value=0}
         {arriveTangent=0.6613, interpMode=RCIM_Cubic, leaveTangent=0.6613, tangentMode=RCTM_User, time=0.7, value=0.2}
         {arriveTangentWeight=0.4215, interpMode=RCIM_Cubic, tangentMode=RCTM_User, time=1, value=1}
        ]}}}
     CurveIndex=&RandomRangeFloat(Minimum=0, Maximum=1)
    )) {isSetParameters=true}
   &InitializeParticle(
    Sprite Rotation Mode=::NewEnumerator2
    Sprite Size=&Multiply_Vector2D_ByFloat(Float=$User.GlobalScale, Vector2D=&RandomRangeVector2D(Minimum=<1750 800>, Maximum=<3000 1400>))
    Sprite Size Mode=::NewEnumerator4
    UsePositionOffset=false
    Write Material Random=true
   )
   &SampleParticlesFromOtherEmitter(Color Sampling=::NewEnumerator0, Reject Particle on Sampled Particle's First Frame=true)
  ]
  particleUpdateScript [
   &ParticleState()
   &ScaleSpriteSize(
    Uniform Curve Scale=&RandomRangeFloat(Minimum=2, Maximum=5)
    Uniform Curve Sprite Scale=DataInterface() {propertyValues {Curve {keys [
        {arriveTangent=9.0725, interpMode=RCIM_Cubic, leaveTangent=9.0725, tangentMode=RCTM_User, time=0, value=0}
        {arriveTangent=-2.5249, interpMode=RCIM_Cubic, leaveTangent=-2.5249, tangentMode=RCTM_User, time=0.1408, value=1}
        {arriveTangent=-0.0405, interpMode=RCIM_Cubic, leaveTangent=-0.0405, tangentMode=RCTM_User, time=1, value=0}
       ]}}}
   )
   &SolveForcesAndVelocity(Clamp Velocity=false, Limit Acceleration=false)
   &ScaleColor(Scale RGB=&VectorFromFloat(Value=5), ScaleA=true, ScaleRGB=true)
  ]
  rendererClasses [&NiagaraSpriteRendererProperties]
  renderers [{rendererClass=&NiagaraSpriteRendererProperties}]
 }
 Opening_Glow {
  emitterUpdateScript [
   &EmitterState(Life Cycle Mode=::NewEnumerator1, Loop Behavior=::NewEnumerator1, Loop Duration Mode=::NewEnumerator1, UseLoopDelay=false)
   &SpawnBurst_Instantaneous(Spawn Count=1, Use Spawn Probability=false)
   &SpawnBurst_Instantaneous(Spawn Group=1, Use Spawn Probability=false, UseLoopCountLimit=false) {module=SpawnBurst_Instantaneous001}
   &SpawnBurst_Instantaneous(Spawn Group=2, Use Spawn Probability=false, UseLoopCountLimit=false) {module=SpawnBurst_Instantaneous002}
   /SetVariables(Emitter.Tiling Scale:NiagaraFloat=&Lerp_Float(A=4, B=3, Alpha=$System.ControlCurve)) {isSetParameters=true}
   /SetVariables(Emitter.CorruptionControl:NiagaraFloat=&FloatFromCurve(
     FloatCurve=DataInterface() {propertyValues {
       bOptimizeLUT=false
       bUseLUT=false
       Curve {keys [{arriveTangent=3.7121, interpMode=RCIM_Cubic, leaveTangent=3.7121, tangentMode=RCTM_User, time=0, value=0}, <at 1 1.5 RCIM_Cubic>]}
      }}
     CurveIndex=$System.ControlCurve
    )) {isSetParameters=true}
  ]
  particleSpawnScript [
   /SetVariables(Particles.VisibilityTag=&SpawnGroup()) {isSetParameters=true}
   &InitializeParticle(
    Lifetime=2
    Mesh Scale Mode=::NewEnumerator3
    Mesh Uniform Scale=&Multiply_Float(A=&MaskFloatBySpawnGroup(Masked Float=175, Passthrough Float=225, Spawn Group Mask=2), B=$User.GlobalScale)
    Position Offset=&Multiply_VectorByFloat(Vector=<0 0 0>, Float=$User.GlobalScale)
    UsePositionOffset=true
    Write Material Random=true
   )
   &InitialMeshOrientation(Mesh Axis To Orient=<0 0 -1>, Mesh Orientation Mode=::NewEnumerator2, Orientation Vector=<0 0 1>, Use Rotation Vector=true)
  ]
  particleUpdateScript [
   &ParticleState(Kill Particles When Lifetime Has Elapsed=false)
   &ScaleMeshSize(Scale Factor=&VectorFromFloat(Value=&MaskFloatBySpawnGroup(
      Masked Float=&Lerp_Float(A=1, B=2, Alpha=$System.ControlCurve)
      Passthrough Float=&Lerp_Float(A=1, B=1.5, Alpha=$System.ControlCurve)
     )))
   &DynamicMaterialParameters(
    Index 0 Param 1=&MaskFloatBySpawnGroup(
     Masked Float=&Lerp_Float(A=0, B=0.2, Alpha=$System.ControlCurve)
     Passthrough Float=&Lerp_Float(A=0, B=0, Alpha=$System.ControlCurve)
    )
    Index 0 Param 2=1
    Module.Param0WriteEnabled=true
    Module.Param1WriteEnabled=true
    Module.Param2WriteEnabled=false
    Module.Param3WriteEnabled=false
   )
   &ScaleColor(
    Scale Alpha=&MaskFloatBySpawnGroup(
     Masked Float=&Lerp_Float(A=0.5, B=0.333, Alpha=$System.ControlCurve)
     Passthrough Float=&MaskFloatBySpawnGroup(
      Masked Float=&Lerp_Float(A=0.05, B=0.5, Alpha=$System.ControlCurve)
      Passthrough Float=&Lerp_Float(A=0.25, B=1, Alpha=$System.ControlCurve)
      Spawn Group Mask=1
     )
    )
    Scale RGB=&VectorFromFloat(Value=&MaskFloatBySpawnGroup(Masked Float=1, Passthrough Float=3))
    ScaleA=true
    ScaleRGB=true
   )
   /SetVariables(Particles.Shatter_DynParameters:Vector4f=&MakeVector4(
     X=&Lerp_Float(A=10, B=20, Alpha=$System.ControlCurve)
     Y=0
     Z=0
     W=&Lerp_Float(A=0.5, B=1, Alpha=$System.ControlCurve)
    )) {isSetParameters=true}
  ]
  rendererClasses [&NiagaraMeshRendererProperties]
  renderers [
   {rendererClass=&NiagaraMeshRendererProperties, rendererIndex=0}
   {rendererClass=&NiagaraMeshRendererProperties, rendererIndex=1}
   {rendererClass=&NiagaraMeshRendererProperties, rendererIndex=2}
  ]
 }
 Pointer {
  emitterUpdateScript [
   &EmitterState(Life Cycle Mode=::NewEnumerator1, Loop Behavior=::NewEnumerator1, Loop Duration Mode=::NewEnumerator1, UseLoopDelay=false)
   &SpawnBurst_Instantaneous(Spawn Count=2, Use Spawn Probability=false, UseLoopCountLimit=false)
   /SetVariables(Emitter.CorruptionControl=1) {isSetParameters=true}
  ]
  particleSpawnScript [&InitializeParticle(Color=<rgba 0.25 0.25 0.25 0.35>, Ribbon Width=2500, Ribbon Width Mode=::NewEnumerator1, UsePositionOffset=false, Write Material Random=true)]
  particleUpdateScript [
   &ParticleState(Kill Particles When Lifetime Has Elapsed=false, Let Infinitely Lived Particles Die When Emitter Deactivates=true)
   /SetVariables(Particles.Position=&Lerp_Vector(
     A=Linked() {linkedVariable {name=Engine.Emitter.SimulationPosition, type=NiagaraPosition}}
     B=&MakeVector(X=0, Y=0, Z=&Multiply_Float(A=$User.Height, B=-1))
     Alpha=&ReturnNormalizedExecIndex()
    )) {isSetParameters=true}
   /SetVariables(Particles.RibbonLinkOrder:NiagaraFloat=&ReturnNormalizedExecIndex()) {isSetParameters=true}
   &DynamicMaterialParameters(
    Index 0 Param 1=1
    Index 0 Param 3=&Multiply_Float(A=$Engine.Time, B=0.0088)
    Index 1 Param 1=1
    Module.Param0WriteEnabled=true
    Module.Param0WriteEnabled1=true
    Module.Param1WriteEnabled=true
    Module.Param1WriteEnabled1=true
    Module.Param2WriteEnabled=true
    Module.Param2WriteEnabled1=true
    Module.Param3WriteEnabled=true
    Module.Param3WriteEnabled1=true
    Write Parameter Index 1=true
   )
   &ScaleRibbonWidth(Ribbon Width Scale=&FloatFromCurve(
     FloatCurve=DataInterface() {propertyValues {Curve {keys [{interpMode=RCIM_Cubic, tangentMode=RCTM_User, time=0.5, value=1}, {tangentMode=RCTM_User, time=1, value=0.9}]}}}
     CurveIndex=&ReturnNormalizedExecIndex()
    ))
   &ScaleColor(
    Scale Alpha=&FloatFromCurve(
     FloatCurve=DataInterface() {propertyValues {bOptimizeLUT=false, bUseLUT=false, Curve {keys [<at 0 0.5>,<at 1 1>]}}}
     CurveIndex=$System.ControlCurve
    )
    Scale RGB=&VectorFromFloat(Value=&FNTimeOfDayScalar(Night Scale=0.3, Morning Scale=0.5, Evening Scale=0.5))
    ScaleA=true
    ScaleRGB=true
   )
  ]
  rendererClasses [&NiagaraRibbonRendererProperties]
  renderers [{rendererClass=&NiagaraRibbonRendererProperties}]
 }
 Rays {
  emitterUpdateScript [
   &EmitterState(Life Cycle Mode=::NewEnumerator1, Loop Behavior=::NewEnumerator1, Loop Duration Mode=::NewEnumerator1, UseLoopDelay=false)
   &SpawnBurst_Instantaneous(Use Spawn Probability=false, UseLoopCountLimit=false)
   /SetVariables(Emitter.PanSpeed=0.2) {isSetParameters=true}
   /SetVariables(
    Emitter.EndColor:LinearColor=&MultiplyLinearColor(
     Linear Color=&LerpLinearColors(
      StartColor=<rgba 0.3073 0.0156 0 1>
      EndColor=<rgba 0.2212 0.1709 0.974 1>
      LerpFactor=&FloatFromCurve(
       FloatCurve=DataInterface() {propertyValues {bOptimizeLUT=false, bUseLUT=false, Curve {keys [<at 0 1>,<at 1 0>]}}}
       CurveIndex=$System.ControlCurve
      )
     )
     Scale Factor=&MakeFloat4FromVectorAndFloat("Vector (RGB)"=&VectorFromFloat(Value=20), "Float (Alpha)"=1)
    )
    Emitter.StartColor:LinearColor=&MultiplyLinearColor(
     Linear Color=&LerpLinearColors(
      StartColor=<rgba 0.7344 0.0365 0 1>
      EndColor=<rgba 0.1052 0.3566 0.974 1>
      LerpFactor=&FloatFromCurve(
       FloatCurve=DataInterface() {propertyValues {bOptimizeLUT=false, bUseLUT=false, Curve {keys [<at 0 1>,<at 1 0>]}}}
       CurveIndex=$System.ControlCurve
      )
     )
     Scale Factor=&MakeFloat4FromVectorAndFloat("Vector (RGB)"=&VectorFromFloat(Value=25), "Float (Alpha)"=1)
    )
   ) {isSetParameters=true, module=SetVariables_#6}
  ]
  particleSpawnScript [
   &InitializeParticle(
    Lifetime=0.5
    Mesh Scale=<150 150 1>
    Mesh Scale Mode=::NewEnumerator4
    Position Offset=<0 0 -15000>
    UsePositionOffset=true
    Write Material Random=true
   )
   &InitialMeshOrientation(Mesh Axis To Orient=<0 0 1>, Mesh Orientation Mode=::NewEnumerator2, Orientation Vector=<0 0 1>, Use Rotation Vector=true)
  ]
  particleUpdateScript [
   &ParticleState(Kill Particles When Lifetime Has Elapsed=false, Let Infinitely Lived Particles Die When Emitter Deactivates=true)
   &UpdateMeshOrientation(
    Rotation Rate=&RandomRangeFloat(
     Minimum=0.005
     Maximum=0.01
     Randomness Mode=::NewEnumerator1
     Random Seed=&Add_Integer(A=80678, B=&ReturnExecIndex())
    )
    Rotation Vector=<0 0 1>
   )
   &DynamicMaterialParameters(
    Index 0 Param 1=&Lerp_Float(A=50, B=100, Alpha=$System.ControlCurve)
    Index 0 Param 2=144.9005
    Index 0 Param 3=&Lerp_Float(A=25, B=45, Alpha=$System.ControlCurve)
    Index 0 Param 4=5000
    Index 1 Param 1=1
    Module.Param0WriteEnabled=true
    Module.Param0WriteEnabled1=true
    Module.Param1WriteEnabled=true
    Module.Param1WriteEnabled1=false
    Module.Param2WriteEnabled=true
    Module.Param2WriteEnabled1=false
    Module.Param3WriteEnabled=true
    Module.Param3WriteEnabled1=true
    Write Parameter Index 1=true
   )
   &ScaleColor(
    Scale Alpha=&Lerp_Float(A=0.25, B=1, Alpha=$System.ControlCurve)
    Scale RGB=&MakeVectorFromLinearColorRGB(Color=&LerpLinearColors(StartColor=$Emitter.StartColor, EndColor=$Emitter.EndColor, LerpFactor=$System.SyncWaveform))
    ScaleA=true
    ScaleRGB=true
   )
  ]
  rendererClasses [&NiagaraMeshRendererProperties]
  renderers [{rendererClass=&NiagaraMeshRendererProperties, rendererIndex=0}]
 }
 Red_Lightning {
  emitterUpdateScript [
   &EmitterState()
   &SpawnBurst_Instantaneous(
    Spawn Count=&RandomRangeInt(Minimum=0, Maximum=3)
    Use Spawn Probability=false
    UseLoopCountLimit=false
   )
  ]
  particleSpawnScript [
   &InitializeParticle(
    Lifetime Max=0.35
    Lifetime Min=0.15
    Lifetime Mode=::NewEnumerator1
    Mesh Scale Max=<75 50 50>
    Mesh Scale Min=<25 15 15>
    Mesh Scale Mode=::NewEnumerator2
    Position Offset=<0 0 -18144.9141>
    UsePositionOffset=true
    Write Material Random=true
   )
   &InitialMeshOrientation(
    Mesh Orientation Mode=::NewEnumerator2
    Orientation Vector=&RandomRangeVector(Minimum=<0 0 -1>, Maximum=<1 1 -1>)
    Rotation=&RandomRangeVector(Minimum=<0 0 0>, Maximum=<1 0 0>)
    Use Rotation Vector=true
   )
  ]
  particleUpdateScript [
   &ParticleState()
   &ScaleMeshSize(Scale Factor=&VectorFromFloat(Value=&Lerp_Float(A=0, B=1, Alpha=$System.ControlCurve)))
   &ScaleColor(
    Scale Alpha=&Multiply_Float(
     A=&Lerp_Float(A=0.5, B=0.8, Alpha=&RandomRangeFloat(Minimum=0, Maximum=1))
     B=&MakeFloatFromVector(
      Vector=&FN_Select_Vector_from_Time_Of_Day(
       Morning=&VectorFromFloat(Value=1)
       Day=&VectorFromFloat(Value=0.85)
       Evening=&VectorFromFloat(Value=1)
       Night=&VectorFromFloat(Value=0.25)
      )
      Channel=::NewEnumerator0
     )
    )
    Scale RGB=&VectorFromFloat(Value=&FloatFromCurve(
      FloatCurve=DataInterface() {propertyValues {Curve {keys [
          {arriveTangent=51.9074, interpMode=RCIM_Cubic, leaveTangent=51.9074, tangentMode=RCTM_User, time=0, value=0}
          {arriveTangentWeight=0.1274, interpMode=RCIM_Cubic, leaveTangentWeight=0.7778, tangentMode=RCTM_User, time=0.1408, value=5}
          {arriveTangent=-0.0405, interpMode=RCIM_Cubic, leaveTangent=-0.0405, tangentMode=RCTM_User, time=1, value=0}
         ]}}}
      Scale Curve=50
     ))
    ScaleA=true
    ScaleRGB=true
   )
   &DynamicMaterialParameters(
    Index 0 Param 1=&FloatFromCurve(
     FloatCurve=DataInterface() {propertyValues {Curve {keys [<at 0 0>,<at 1 1>]}}}
     Scale Curve=&Lerp_Float(A=1, B=1.5, Alpha=$System.SyncWaveform)
    )
    Index 0 Param 2=&FloatFromCurve(
     FloatCurve=DataInterface() {propertyValues {Curve {keys [
         {arriveTangent=9.0725, interpMode=RCIM_Cubic, leaveTangent=9.0725, tangentMode=RCTM_User, time=0, value=0}
         {arriveTangent=-2.5249, interpMode=RCIM_Cubic, leaveTangent=-2.5249, tangentMode=RCTM_User, time=0.1408, value=1}
         {arriveTangent=-0.0405, interpMode=RCIM_Cubic, leaveTangent=-0.0405, tangentMode=RCTM_User, time=1, value=0}
        ]}}}
     Scale Curve=10
    )
    Index 0 Param 3=&FloatFromCurve(
     FloatCurve=DataInterface() {propertyValues {Curve {keys [
         {interpMode=RCIM_Cubic, tangentMode=RCTM_User, time=0, value=1}
         {interpMode=RCIM_Cubic, tangentMode=RCTM_User, time=0.2, value=0.25}
         {interpMode=RCIM_Cubic, tangentMode=RCTM_User, time=0.5, value=1.375}
         {interpMode=RCIM_Cubic, tangentMode=RCTM_User, time=1, value=2}
        ]}}}
     Scale Curve=&RandomRangeFloat(Minimum=0.4, Maximum=0.6)
    )
    Index 0 Param 4=&RandomRangeFloat(Minimum=0, Maximum=1)
    Module.Param0WriteEnabled=true
    Module.Param1WriteEnabled=true
    Module.Param2WriteEnabled=true
    Module.Param3WriteEnabled=true
   )
  ]
  rendererClasses [&NiagaraMeshRendererProperties]
  renderers [{rendererClass=&NiagaraMeshRendererProperties, rendererIndex=0}]
 }
 SharpFlare {
  emitterUpdateScript [
   &EmitterState(Life Cycle Mode=::NewEnumerator1, Loop Behavior=::NewEnumerator1, Loop Duration Mode=::NewEnumerator1, UseLoopDelay=false)
   &SpawnBurst_Instantaneous(Spawn Count=1, Use Spawn Probability=false)
   /SetVariables(
    Emitter.EndColor:LinearColor=&MultiplyLinearColor(
     Linear Color=&LerpLinearColors(
      StartColor=<rgba 0.3073 0.0156 0 1>
      EndColor=<rgba 0.2212 0.1709 0.974 1>
      LerpFactor=&FloatFromCurve(
       FloatCurve=DataInterface() {propertyValues {bOptimizeLUT=false, bUseLUT=false, Curve {keys [<at 0 1>,<at 1 0>]}}}
       CurveIndex=$System.ControlCurve
      )
     )
     Scale Factor=&MakeFloat4FromVectorAndFloat("Vector (RGB)"=&VectorFromFloat(Value=20), "Float (Alpha)"=1)
    )
    Emitter.StartColor:LinearColor=&MultiplyLinearColor(
     Linear Color=&LerpLinearColors(
      StartColor=<rgba 0.7344 0.0365 0 1>
      EndColor=<rgba 0.1052 0.3566 0.974 1>
      LerpFactor=&FloatFromCurve(
       FloatCurve=DataInterface() {propertyValues {bOptimizeLUT=false, bUseLUT=false, Curve {keys [<at 0 1>,<at 1 0>]}}}
       CurveIndex=$System.ControlCurve
      )
     )
     Scale Factor=&MakeFloat4FromVectorAndFloat("Vector (RGB)"=&VectorFromFloat(Value=25), "Float (Alpha)"=1)
    )
   ) {isSetParameters=true, module=SetVariables_#6}
  ]
  particleSpawnScript [
   &InitializeParticle(
    Position Offset=<0 0 -18144.9141>
    Sprite Size=&Multiply_Vector2D_ByFloat(Float=25, Vector2D=<5000 5000>)
    Sprite Size Mode=::NewEnumerator4
    UsePositionOffset=true
    Write Material Random=true
   )
   &CameraOffset(Camera Offset Amount=50000)
  ]
  particleUpdateScript [
   &ParticleState(Kill Particles When Lifetime Has Elapsed=false)
   &ScaleColor(
    Scale Alpha=&Lerp_Float(A=0.25, B=0.5, Alpha=$System.ControlCurve)
    Scale RGB=&MakeVectorFromLinearColorRGB(Color=&LerpLinearColors(StartColor=$Emitter.StartColor, EndColor=$Emitter.EndColor, LerpFactor=$System.SyncWaveform))
    ScaleA=true
    ScaleRGB=true
   )
   &DynamicMaterialParameters(
    Index 0 Param 1=0.75
    Index 0 Param 2=0.5
    Module.Param0WriteEnabled=true
    Module.Param1WriteEnabled=true
    Module.Param2WriteEnabled=true
    Module.Param3WriteEnabled=true
   )
   /SetVariables(
    Particles.Corona Size:Vector2f=&Multiply_Vector2D_ByFloat(Float=&Lerp_Float(A=0.75, B=1.8, Alpha=$System.ControlCurve), Vector2D=$Particles.Initial.SpriteSize)
    Particles.CoronaColor:LinearColor=&LerpLinearColors(StartColor=<rgba 1 0.5 0.5 1>, EndColor=<rgba 0.5 0.5 1 1>, LerpFactor=$System.ControlCurve)
    Particles.Dyn Paramter Corona:Vector4f=&MakeVector4(X=0.75, Y=0.1, Z=0.75, W=0)
   ) {isSetParameters=true}
   &ScaleSpriteSize(Scale Sprite Size Mode=::NewEnumerator0, Uniform Scale Factor=&Lerp_Float(A=0.5, B=1, Alpha=$System.ControlCurve))
  ]
  rendererClasses [&NiagaraSpriteRendererProperties]
  renderers [
   {rendererClass=&NiagaraSpriteRendererProperties}
   {rendererClass=&NiagaraSpriteRendererProperties, rendererIndex=1}
   {rendererClass=&NiagaraSpriteRendererProperties, rendererIndex=2}
  ]
 }
 Tiny_Fragments {
  emitterUpdateScript [
   &EmitterState(Life Cycle Mode=::NewEnumerator1, Loop Behavior=::NewEnumerator1, Loop Duration Mode=::NewEnumerator1, UseLoopDelay=false)
   &SpawnBurst_Instantaneous(
    Spawn Count=&MakeIntFromFloat(Float=&Lerp_Float(A=50, B=150, Alpha=$System.ControlCurve))
    Use Spawn Probability=false
    UseLoopCountLimit=false
   )
  ]
  particleSpawnScript [
   /SetVariables(Particles.RandCurve:NiagaraFloat=&FloatFromCurve(
     FloatCurve=DataInterface() {propertyValues {Curve {keys [
         {interpMode=RCIM_Cubic, tangentMode=RCTM_User, time=0, value=0}
         {arriveTangent=0.6613, interpMode=RCIM_Cubic, leaveTangent=0.6613, tangentMode=RCTM_User, time=0.7, value=0.2}
         {arriveTangentWeight=0.4215, interpMode=RCIM_Cubic, tangentMode=RCTM_User, time=1, value=1}
        ]}}}
     CurveIndex=&RandomRangeFloat(Minimum=0, Maximum=1)
    )) {isSetParameters=true, module=SetVariables_#5}
   &InitializeParticle(
    Color Mode=::NewEnumerator0
    Lifetime=5
    Mesh Renderer Array Visibility Mode=::NewEnumerator3
    Mesh Renderer Info=DataInterface() {propertyValues {MeshRenderer=&self:Tiny_Fragments_5.NiagaraMeshRendererProperties_0}}
    Mesh Renderer Random Seed=80678
    Mesh Scale Mode=::NewEnumerator3
    Mesh Uniform Scale=&Multiply_Float(A=&Lerp_Float(A=50, B=200, Alpha=$Particles.RandCurve), B=$User.GlobalScale)
    Sprite Size Mode=::NewEnumerator3
    Uniform Sprite Size=5000
    UsePositionOffset=false
    Write Material Random=true
   )
   &ShapeLocation(
    Cone Angle:NiagaraFloat=&Lerp_Float(A=80, B=160, Alpha=$System.ControlCurve)
    Cone Inner Angle=10
    Cone Length=-90000
    Fixed Random Seed=true
    MaskSpawnGroup=false
    Offset=<0 0 -3334.5359>
    Offset Mode=::NewEnumerator0
    Override Seed=true
    Random Seed:NiagaraInt32=&Add_Integer(A=81155, B=&ReturnExecIndex())
    Shape Primitive=::NewEnumerator5
    UseOwnerScale=true
   )
   &InitialMeshOrientation(
    Mesh Orientation Mode=::NewEnumerator2
    Orientation Vector=<0 0 1>
    Rotation=&RandomRangeVector(
     Minimum=<0 0 0>
     Maximum=<0 0.25 1>
     Randomness Mode=::NewEnumerator1
     Random Seed=&Add_Integer(A=80678, B=&ReturnExecIndex())
     Fixed Random Seed=true
    )
    Use Rotation Vector=true
   )
  ]
  particleUpdateScript [
   &ParticleState(Kill Particles When Lifetime Has Elapsed=false)
   &UpdateMeshOrientation(
    Rotation Rate=&RandomRangeFloat(
     Minimum=0.01
     Maximum=0.02
     Randomness Mode=::NewEnumerator1
     Random Seed=&Add_Integer(A=80678, B=&ReturnExecIndex())
     Fixed Random Seed=true
    )
    Rotation Vector=&RandomRangeVector(
     Minimum=<0 0 0>
     Maximum=<1 1 1>
     Randomness Mode=::NewEnumerator1
     Random Seed=&Add_Integer(A=80678, B=&ReturnExecIndex())
     Fixed Random Seed=true
    )
   )
   &RegionMask(
    Offset:Vector3f=<0 0 -15000>
    Region Falloff=0
    Region Radius:NiagaraFloat=&Lerp_Float(A=0, B=75000, Alpha=$System.ControlCurve)
    UseOffset=true
    UseRotation=false
   )
   &ScaleMeshSize(Scale Factor=&VectorFromFloat(Value=&Lerp_Float(A=1, B=1, Alpha=$Output.RegionMask.RegionMask)))
   &DynamicMaterialParameters(
    Index 0 Param 4=&Lerp_Float(A=&Lerp_Float(A=2, B=10, Alpha=$Output.RegionMask.RegionMask), B=3, Alpha=$System.ControlCurve)
    Module.Param0WriteEnabled=false
    Module.Param1WriteEnabled=false
    Module.Param2WriteEnabled=false
    Module.Param3WriteEnabled=true
   )
   &ScaleColor(
    Scale Alpha=&Lerp_Float(A=1, B=2, Alpha=$Output.RegionMask.RegionMask)
    Scale RGB=&MakeVectorFromLinearColorRGB(Color=&LerpLinearColors(StartColor=<rgba 1 1 1 1>, EndColor=<rgba 0.3073 0.0156 0 1>, LerpFactor=$Output.RegionMask.RegionMask))
    ScaleA=true
    ScaleRGB=true
   )
   &SolveForcesAndVelocity(Clamp Velocity=false, Limit Acceleration=false)
  ]
  rendererClasses [&NiagaraMeshRendererProperties]
  renderers [{rendererClass=&NiagaraMeshRendererProperties, rendererIndex=0}]
 }
}
summary {
 systemName=NS_RiftGate_Crack
 emitters [
  {emitterName=Glints, rendererClasses [&NiagaraSpriteRendererProperties]}
  {emitterName=SharpFlare, rendererClasses [&NiagaraSpriteRendererProperties]}
  {emitterName=Rays, rendererClasses [&NiagaraMeshRendererProperties]}
  {emitterName=Fragments, rendererClasses [&NiagaraMeshRendererProperties]}
  {emitterName=Crack, rendererClasses [&NiagaraMeshRendererProperties]}
  {emitterName=Opening_Glow, rendererClasses [&NiagaraMeshRendererProperties]}
  {emitterName=Tiny_Fragments, rendererClasses [&NiagaraMeshRendererProperties]}
  {emitterName=Red_Lightning, rendererClasses [&NiagaraMeshRendererProperties]}
  {emitterName=Fragment_Lighting, rendererClasses [&NiagaraMeshRendererProperties]}
  {emitterName=Energy_Glints, simTarget=GPUComputeSim, rendererClasses [&NiagaraSpriteRendererProperties]}
  {emitterName=Pointer, rendererClasses [&NiagaraRibbonRendererProperties]}
  {bEnabled=false, emitterName=CenterFlare, simTarget=GPUComputeSim, rendererClasses [&NiagaraSpriteRendererProperties]}
 ]
 userVariables [
  User.CalendarControl {type=NiagaraFloat, defaultValue=0, description=""}
  User.GlobalOpacity {type=NiagaraFloat, defaultValue=0, description=""}
  User.GlobalScale {type=NiagaraFloat, defaultValue=4.2, description=""}
  User.Height {type=NiagaraFloat, defaultValue=371920, description=""}
 ]
}
