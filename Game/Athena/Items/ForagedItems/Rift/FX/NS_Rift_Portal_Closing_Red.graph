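# Red-tinted Niagara effect for a rift portal closing, building over about ten seconds into a final collapse.
# Merged flares, mesh fragments, a light driven by User.Open, and GPU and CPU embers pulled inward by rising attraction form the buildup.
# Delayed electric arcs precede end flares and a shrinking shockwave mesh at ten seconds.
# Merged emitters share particles across renderers via visibility tags, while LOD-distance curves adjust size and ember spawn rate.
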NiagaraSystem /Game/Athena/Items/ForagedItems/Rift/FX/NS_Rift_Portal_Closing_Red
dependencies=null
emitters {
 ElectricArcsResidual {
  emitterUpdateScript [
   &EmitterState(Life Cycle Mode=::NewEnumerator1, Loop Behavior=::NewEnumerator1, Loop Delay=7, Loop Duration=3.1, UseLoopDelay=true)
   &SpawnRate(SpawnRate=5, Use Spawn Probability=false)
   &SpawnBurst_Instantaneous(Spawn Count=5, Spawn Time=2.2, Use Spawn Probability=false)
   &SpawnBurst_Instantaneous(Spawn Count=6, Spawn Time=3, Use Spawn Probability=false) {module=SpawnBurst_Instantaneous001}
   &SpawnBurst_Instantaneous(Spawn Count=4, Spawn Time=3.05, Use Spawn Probability=false) {module=SpawnBurst_Instantaneous002}
  ]
  particleSpawnScript [
   /SetVariables(Particles.MeshOrientation=&MakeQuatFromAxisAngle(
     Axis=&RandomRangeVector(Minimum=<0 0 0>, Maximum=<1 0 0>, Evaluation Type=::NewEnumerator1)
     AngleInDegrees=&RandomRangeFloat(Minimum=0, Maximum=360)
    )) {isSetParameters=true}
   /Niagara/Modules/Spawn/Initialization/InitializeParticle(
    Lifetime=&RandomRangeFloat(Minimum=0.08, Maximum=0.2, Evaluation Type=::NewEnumerator1)
    Mesh Scale=&RandomRangeVector(Minimum=<8 6 6>, Maximum=<15 20 20>, Evaluation Type=::NewEnumerator1)
    Module.Write Color=true
    Module.Write Lifetime=true
    Module.Write Mass=true
    Module.Write Material Random=true
    Module.Write Position=false
    Module.Write Scale=true
    Module.Write SpriteRotation=false
    Module.Write SpriteSize=false
   )
   &DynamicMaterialParameters(
    Index 0 Param 1=&RandomRangeFloat(Minimum=2, Maximum=3, Evaluation Type=::NewEnumerator1)
    Index 0 Param 3=&RandomRangeFloat(Minimum=0.2, Maximum=2, Evaluation Type=::NewEnumerator1)
    Index 0 Param 4=&RandomRangeFloat(Minimum=0, Maximum=1, Evaluation Type=::NewEnumerator1)
    Module.Param0WriteEnabled=true
    Module.Param1WriteEnabled=true
    Module.Param2WriteEnabled=true
    Module.Param3WriteEnabled=true
   )
   &Color(Color=&MakeLinearColorFromVectorAndFloat("Vector (RGB)"=<1 0.2 0.1>, "Float (Alpha)"=&ClampFloat(Float=2)))
   /Niagara/Modules/Spawn/Location/SphereLocation(MaskSpawnGroup=false, Module.Override Randomness=false, Sphere Radius=10, Surface Only=true)
   &AddVelocityFromPoint(Module.Use Falloff Distance=false)
  ]
  particleUpdateScript [
   &ParticleState()
   &DynamicMaterialParameters(
    Index 0 Param 2=&FloatFromCurve(FloatCurve=DataInterface() {propertyValues {Curve {keys [{arriveTangent=-1.8894, interpMode=RCIM_Cubic, time=0, value=1}, <at 1 -0.25>]}}})
    Module.Param0WriteEnabled=true
    Module.Param1WriteEnabled=true
    Module.Param2WriteEnabled=true
    Module.Param3WriteEnabled=true
   ) {module=DynamicMaterialParameters001}
   &ScaleColor(
    Scale RGB=&VectorFromCurve(VectorCurve=DataInterface() {propertyValues {
       XCurve {keys [<at 0 50 RCIM_Cubic>, {arriveTangent=-45, interpMode=RCIM_Cubic, leaveTangent=-45, time=0.5, value=30}, <at 1 5>]}
       YCurve {keys [<at 0 19.1123 RCIM_Cubic>, {arriveTangent=-18.2463, interpMode=RCIM_Cubic, leaveTangent=-18.2463, time=0.5, value=13.6826}, <at 1 0.8661>]}
       ZCurve {keys [<at 0 22.1016 RCIM_Cubic>, {arriveTangent=-21.4843, interpMode=RCIM_Cubic, leaveTangent=-21.4843, time=0.5, value=12.787}, <at 1 0.6173>]}
      }})
    ScaleA=true
    ScaleRGB=true
   )
   &ScaleMeshSize(Scale Factor=&VectorFromCurve(VectorCurve=DataInterface() {propertyValues {
       XCurve {keys [<at 0.0033 0.5 RCIM_Constant>,<at 0.2054 0.7619 RCIM_Constant>,<at 0.4101 0.8575 RCIM_Constant>,<at 0.6967 1>]}
       YCurve {keys [<at 0.0033 0.6004 RCIM_Constant>,<at 0.2054 0.6716 RCIM_Constant>,<at 0.4101 0.3646 RCIM_Constant>,<at 0.6967 0.2468>]}
       ZCurve {keys [<at 0.0033 0.7767 RCIM_Constant>,<at 0.2054 0.5215 RCIM_Constant>,<at 0.4101 0.4457 RCIM_Constant>,<at 0.6967 0.6282>]}
      }}))
   &ScaleMeshSize(Scale Factor=&VectorFromFloat(Value=&FloatFromCurve(FloatCurve=DataInterface() {propertyValues {Curve {keys [<at 0 1>,<at 15000 1.785>]}}}, CurveIndex=$Engine.Owner.LODDistance))) {module=ScaleMeshSize001}
   &CameraOffset(Camera Offset Amount=100)
   &SolveForcesAndVelocity(Clamp Velocity=false, Limit Acceleration=false)
  ]
  rendererClasses [&NiagaraMeshRendererProperties]
  renderers [{rendererClass=&NiagaraMeshRendererProperties, rendererIndex=0}]
 }
 Embers {
  simTarget=GPUComputeSim
  emitterUpdateScript [
   &EmitterState(Life Cycle Mode=::NewEnumerator1, Loop Behavior=::NewEnumerator1, Loop Duration=10, UseLoopDelay=true)
   &SpawnRate(
    SpawnRate=&FloatFromCurve(
     FloatCurve=DataInterface() {propertyValues {Curve {keys [<at 0 0 RCIM_Cubic>,<at 4 350>]}}}
     CurveIndex=$Emitter.Age
     Scale Curve=&FloatFromCurve(
      FloatCurve=DataInterface() {propertyValues {Curve {keys [{interpMode=RCIM_Cubic, tangentMode=RCTM_Break, time=0, value=1}, <at 15000 0.005>]}}}
      CurveIndex=$Engine.Owner.LODDistance
     )
    )
    Use Spawn Probability=false
   )
  ]
  particleSpawnScript [
   &InitializeParticle(
    Lifetime=&RandomRangeFloat(Minimum=0.3, Maximum=0.5, Evaluation Type=::NewEnumerator1)
    Module.Write Color=true
    Module.Write Lifetime=true
    Module.Write Mass=true
    Module.Write Material Random=true
    Module.Write Position=false
    Module.Write Scale=false
    Module.Write SpriteRotation=false
    Module.Write SpriteSize=true
    Sprite Size=&RandomRangeVector2D(Minimum=<4 4>, Maximum=<8 8>, Evaluation Type=::NewEnumerator1)
   )
   &Color(Color=&MakeLinearColorFromVectorAndFloat("Vector (RGB)"=<25 0.1 2>, "Float (Alpha)"=&ClampFloat(Float=2)))
   &SphereLocation(
    MaskSpawnGroup=false
    Module.Override Randomness=false
    Sphere Radius=&RandomRangeFloat(Minimum=250, Maximum=350, Evaluation Type=::NewEnumerator1)
    Surface Only=true
   )
   &AddVelocity(Velocity=&Multiply_VectorByFloat(
     Vector=$Output.SphereLocation.SphereVector
     Float=&RandomRangeFloat(Minimum=0, Maximum=0.1, Evaluation Type=::NewEnumerator1)
    ))
  ]
  particleUpdateScript [
   &ParticleState()
   &Drag(Use Linear Drag=true, Use Rotational Drag=false)
   &PointAttractionForce(
    Attraction Radius=2000
    Attraction Strength=&FloatFromCurve(
     FloatCurve=DataInterface() {propertyValues {Curve {keys [
         {interpMode=RCIM_Cubic, tangentMode=RCTM_Break, time=0, value=0}
         <at 9.8 500>
         {interpMode=RCIM_Cubic, tangentMode=RCTM_Break, time=10, value=0}
        ]}}}
     CurveIndex=$Emitter.Age
    )
    Kill Within Radius=false
    Module.Use Falloff=true
    UseAttractorOffset=false
   )
   &SolveForcesAndVelocity(Clamp Velocity=false, Limit Acceleration=false)
   &ScaleSpriteSize(Scale Factor=&Vector2DFromFloat(Value=&FloatFromCurve(FloatCurve=DataInterface() {propertyValues {Curve {keys [
          <at -0.0044 0.2654>
          {arriveTangent=0.7558, interpMode=RCIM_Cubic, leaveTangent=0.7558, time=0.157, value=0.5261}
          {arriveTangent=-0.6241, interpMode=RCIM_Cubic, leaveTangent=-0.6241, time=0.7034, value=0.8003}
          <at 1 0>
         ]}}})))
   &ScaleSpriteSizeBySpeed(Max Scale Factor=<1 12>, Source Velocity:Vector3f=&Multiply_Vector(A=<0 0.01 1>, B=$Particles.Velocity))
   &ScaleSpriteSize(Scale Factor=&MakeVector2D(
     X=&FloatFromCurve(FloatCurve=DataInterface() {propertyValues {Curve {keys [<at 0 1>,<at 15000 12>]}}}, CurveIndex=$Engine.Owner.LODDistance)
     Y=1
    )) {module=ScaleSpriteSize001}
  ]
  rendererClasses [&NiagaraSpriteRendererProperties]
  renderers [{rendererClass=&NiagaraSpriteRendererProperties}]
 }
 EmbersCPU {
  emitterUpdateScript [
   &EmitterState(Life Cycle Mode=::NewEnumerator1, Loop Behavior=::NewEnumerator1, Loop Duration=10, UseLoopDelay=true)
   &SpawnRate(
    SpawnRate=&FloatFromCurve(
     FloatCurve=DataInterface() {propertyValues {Curve {keys [<at 0 0 RCIM_Cubic>,<at 4 40>]}}}
     CurveIndex=$Emitter.Age
     Scale Curve=&FloatFromCurve(
      FloatCurve=DataInterface() {propertyValues {Curve {keys [{interpMode=RCIM_Cubic, tangentMode=RCTM_Break, time=0, value=1}, <at 15000 0.005>]}}}
      CurveIndex=$Engine.Owner.LODDistance
     )
    )
    Use Spawn Probability=false
   )
  ]
  particleSpawnScript [
   &InitializeParticle(
    Lifetime=&RandomRangeFloat(Minimum=0.3, Maximum=0.5, Evaluation Type=::NewEnumerator1)
    Module.Write Color=true
    Module.Write Lifetime=true
    Module.Write Mass=true
    Module.Write Material Random=true
    Module.Write Position=false
    Module.Write Scale=false
    Module.Write SpriteRotation=false
    Module.Write SpriteSize=true
    Sprite Size=&RandomRangeVector2D(Minimum=<4 4>, Maximum=<8 8>, Evaluation Type=::NewEnumerator1)
   )
   &Color(Color=&MakeLinearColorFromVectorAndFloat("Vector (RGB)"=<25 2 1>, "Float (Alpha)"=&ClampFloat(Float=2)))
   &SphereLocation(
    MaskSpawnGroup=false
    Module.Override Randomness=false
    Sphere Radius=&RandomRangeFloat(Minimum=250, Maximum=350, Evaluation Type=::NewEnumerator1)
    Surface Only=true
   )
   &AddVelocity(Velocity=&Multiply_VectorByFloat(
     Vector=$Output.SphereLocation.SphereVector
     Float=&RandomRangeFloat(Minimum=0, Maximum=0.1, Evaluation Type=::NewEnumerator1)
    ))
  ]
  particleUpdateScript [
   &ParticleState()
   &Drag(Use Linear Drag=true, Use Rotational Drag=false)
   &PointAttractionForce(
    Attraction Radius=2000
    Attraction Strength=&FloatFromCurve(
     FloatCurve=DataInterface() {propertyValues {Curve {keys [
         {interpMode=RCIM_Cubic, tangentMode=RCTM_Break, time=0, value=0}
         <at 9.8 500>
         {interpMode=RCIM_Cubic, tangentMode=RCTM_Break, time=10, value=0}
        ]}}}
     CurveIndex=$Emitter.Age
    )
    Kill Within Radius=false
    Module.Use Falloff=true
    UseAttractorOffset=false
   )
   &SolveForcesAndVelocity(Clamp Velocity=false, Limit Acceleration=false)
   &ScaleSpriteSize(Scale Factor=&Vector2DFromFloat(Value=&FloatFromCurve(FloatCurve=DataInterface() {propertyValues {Curve {keys [
          <at -0.0044 0.2654>
          {arriveTangent=0.7558, interpMode=RCIM_Cubic, leaveTangent=0.7558, time=0.157, value=0.5261}
          {arriveTangent=-0.6241, interpMode=RCIM_Cubic, leaveTangent=-0.6241, time=0.7034, value=0.8003}
          <at 1 0>
         ]}}})))
   &ScaleSpriteSizeBySpeed(Max Scale Factor=<1 12>, Source Velocity:Vector3f=&Multiply_Vector(A=<0 0.01 1>, B=$Particles.Velocity))
   &ScaleSpriteSize(Scale Factor=&MakeVector2D(
     X=&FloatFromCurve(FloatCurve=DataInterface() {propertyValues {Curve {keys [<at 0 1>,<at 15000 12>]}}}, CurveIndex=$Engine.Owner.LODDistance)
     Y=1
    )) {module=ScaleSpriteSize001}
  ]
  rendererClasses [&NiagaraSpriteRendererProperties]
  renderers [{rendererClass=&NiagaraSpriteRendererProperties}]
 }
 End_Flares_Merged {
  emitterUpdateScript [
   &EmitterState(
    Enable Distance Culling=true
    Inactive Response=::NewEnumerator1
    Life Cycle Mode=::NewEnumerator1
    Loop Behavior=::NewEnumerator1
    Loop Delay=10
    Loop Duration=1
    Max Distance Response=::KillAfterParticlesFinish
    MaxDistance=25000
    Module.bUseMaxDistance=true
    Module.bUseMinDistance=false
    Scalability Mode=::NewEnumerator1
    UseLoopDelay=true
   )
   &SpawnBurst_Instantaneous(Spawn Count=1, Use Spawn Probability=false)
   &SpawnBurst_Instantaneous(Spawn Count=1, Spawn Group=1, Use Spawn Probability=false) {module=SpawnBurst_Instantaneous001}
   &SpawnBurst_Instantaneous(Spawn Count=8, Spawn Group=2, Use Spawn Probability=false) {module=SpawnBurst_Instantaneous002}
  ]
  particleSpawnScript [
   &PartitionParticles()
   /SetVariables(Particles.VisibilityTag=$Emitter.SpawnGroup) {isSetParameters=true}
   &InitializeParticle(
    Lifetime=&MaskFloatBySpawnGroup(Masked Float=&RandomRangeFloat(Minimum=1.2, Maximum=1.7), Passthrough Float=0.6, Spawn Group Mask=2)
    Module.Write Color=true
    Module.Write Lifetime=true
    Module.Write Mass=true
    Module.Write Material Random=true
    Module.Write Position=false
    Module.Write Scale=false
    Module.Write SpriteRotation=false
    Module.Write SpriteSize=true
    Sprite Size=<600 600>
   )
   &Color(Color=&MakeLinearColorFromVectorAndFloat("Vector (RGB)"=<3 0.2 0.1>, "Float (Alpha)"=&ClampFloat(Float=1)))
   &TorusLocation(
    Align Torus With Axis=true
    Coordinate Space:ENiagaraCoordinateSpace=::World
    Large Radius:NiagaraFloat=&MaskFloatBySpawnGroup(Masked Float=&RandomRangeFloat(Minimum=300, Maximum=350), Passthrough Float=0, Spawn Group Mask=2)
    MaskSpawnGroup=false
    Torus Distribution Mode:ENiagaraTorusDistributionMode=::NewEnumerator1
    Torus Mode=::NewEnumerator2
    U Position:NiagaraFloat=&ReturnNormalizedExecIndex(Normalized Index Includes One=false)
   )
   &AddVelocityFromPoint(
    Module.Use Falloff Distance=false
    Velocity Strength=&MaskFloatBySpawnGroup(Masked Float=-550, Passthrough Float=0, Spawn Group Mask=2)
   )
   &CameraOffset(Camera Offset Amount=20)
  ]
  particleUpdateScript [
   &SolveForcesAndVelocity(Clamp Velocity=false, Limit Acceleration=false)
   &ParticleState()
   &ScaleColor(
    Scale Alpha=&FloatFromCurve(FloatCurve=DataInterface() {propertyValues {Curve {keys [<at 0 1 RCIM_Cubic>, {interpMode=RCIM_Cubic, leaveTangent=-3.0157, time=1, value=0}]}}})
    Scale RGB=&VectorFromFloat(Value=&FloatFromCurve(FloatCurve=DataInterface() {propertyValues {Curve {keys [{arriveTangent=-6.3837, interpMode=RCIM_Cubic, time=0, value=10}, <at 1 1 RCIM_Cubic>]}}}))
    ScaleA=true
    ScaleRGB=true
   )
   &ScaleSpriteSize(Scale Factor=&Vector2DFromFloat(Value=&FloatFromCurve(FloatCurve=DataInterface() {propertyValues {Curve {keys [
          {arriveTangent=-2.6785, interpMode=RCIM_Cubic, time=0, value=1}
          {arriveTangent=-1.0722, interpMode=RCIM_Cubic, leaveTangent=-1.0722, time=0.0887, value=0.2526}
          {arriveTangent=-0.2772, interpMode=RCIM_Cubic, leaveTangent=-0.2772, time=0.914, value=0.02}
          <at 1 0 RCIM_Cubic>
         ]}}})))
   /SetVariables(
    Particles.Color2=&MakeLinearColorFromVectorAndFloat(
     "Vector (RGB)"=&VectorFromCurve(
      VectorCurve=DataInterface() {propertyValues {
        XCurve {keys [{interpMode=RCIM_Cubic, leaveTangent=1.512, tangentMode=RCTM_Break, time=0, value=0}, <at 1 0.8>]}
        YCurve {keys [{interpMode=RCIM_Cubic, leaveTangent=0.263, tangentMode=RCTM_Break, time=-0.0035, value=0.819}, <at 1 0.9198>]}
        ZCurve {keys [{interpMode=RCIM_Cubic, tangentMode=RCTM_Break, time=0, value=1}, <at 1 1>]}
       }}
      Scale Curve=<0.2 0.1843 0.1686>
     )
     "Float (Alpha)"=&FloatFromCurve(
      FloatCurve=DataInterface() {propertyValues {Curve {keys [
          {interpMode=RCIM_Cubic, leaveTangent=1.7492, tangentMode=RCTM_Break, time=0, value=0}
          {interpMode=RCIM_Cubic, leaveTangent=-3.1414, tangentMode=RCTM_Break, time=0.35, value=1}
          {interpMode=RCIM_Cubic, tangentMode=RCTM_Break, time=1, value=0}
         ]}}}
      Scale Curve=0.5
     )
    )
    Particles.DyMat00:Vector4f=&MakeVector4(
     X=&FloatFromCurve(FloatCurve=DataInterface() {propertyValues {Curve {keys [{arriveTangent=-3.8495, interpMode=RCIM_Cubic, time=0, value=0.9}, <at 0.5 0.05>]}}})
     Y=2
     Z=1
     W=1
    )
    Particles.SpriteSize1:Vector2f=&Multiply_Vector2D(
     A=<1000 40>
     B=&Vector2DFromFloat(Value=&FloatFromCurve(FloatCurve=DataInterface() {propertyValues {Curve {keys [<at 0 1>,<at 10000 3>]}}}, CurveIndex=$Engine.Owner.LODDistance))
    )
    Particles.SpriteSize2=&Multiply_Vector2D(
     A=<350 350>
     B=&Vector2DFromFloat(Value=&FloatFromCurve(FloatCurve=DataInterface() {propertyValues {Curve {keys [
           {arriveTangent=2.0613, interpMode=RCIM_Cubic, leaveTangent=1.6105, tangentMode=RCTM_User, time=0, value=0}
           {arriveTangent=0.1638, interpMode=RCIM_Cubic, leaveTangent=0.1638, tangentMode=RCTM_User, time=1.0463, value=1}
          ]}}}))
    )
   ) {isSetParameters=true}
   &DynamicMaterialParameters(
    Index 0 Param 1=&Add_Float(
     A=$Particles.MaterialRandom
     B=&FloatFromCurve(FloatCurve=DataInterface() {propertyValues {Curve {keys [
          {arriveTangent=2.7533, interpMode=RCIM_Cubic, leaveTangent=2.847, tangentMode=RCTM_Break, time=0, value=0}
          {arriveTangent=1.3581, interpMode=RCIM_Cubic, leaveTangent=1.3581, tangentMode=RCTM_User, time=0.2558, value=0.5125}
          {arriveTangent=0.0529, interpMode=RCIM_Cubic, leaveTangent=0.2021, tangentMode=RCTM_User, time=1, value=1}
         ]}}})
    )
    Index 0 Param 3=&FloatFromCurve(FloatCurve=DataInterface() {propertyValues {Curve {keys [{arriveTangent=-53.9199, time=0, value=0}, {leaveTangent=0.2723, time=0.25, value=0}]}}})
    Index 0 Param 4=&FloatFromCurve(FloatCurve=DataInterface() {propertyValues {Curve {keys [
         {interpMode=RCIM_Cubic, tangentMode=RCTM_Break, time=0, value=1}
         {arriveTangent=-1.7062, interpMode=RCIM_Cubic, leaveTangent=-1.7062, tangentMode=RCTM_User, time=1, value=0}
        ]}}})
    Index 1 Param 1=&FloatFromCurve(FloatCurve=DataInterface() {propertyValues {Curve {keys [<at 0 0>, {arriveTangent=4.2257, interpMode=RCIM_Cubic, leaveTangent=4.2257, tangentMode=RCTM_User, time=0.25, value=0}]}}})
    Index 1 Param 2=&FloatFromCurve(FloatCurve=DataInterface() {propertyValues {Curve {keys [<at 0 0>]}}})
    Index 1 Param 3=&FloatFromCurve(FloatCurve=DataInterface() {propertyValues {Curve {keys [{interpMode=RCIM_Cubic, leaveTangent=6.3917, tangentMode=RCTM_Break, time=0, value=0}]}}})
    Index 1 Param 4=&FloatFromCurve(FloatCurve=DataInterface() {propertyValues {Curve {keys [{arriveTangent=-0.136, leaveTangent=-0.136, time=0, value=1}]}}})
    Index 2 Param 1=&FloatFromCurve(FloatCurve=DataInterface() {propertyValues {Curve {keys [
         {arriveTangent=2.646, interpMode=RCIM_Cubic, leaveTangent=1.2647, tangentMode=RCTM_Break, time=0, value=0}
         {leaveTangent=-0.0719, time=1, value=1}
        ]}}})
    Index 2 Param 2=&FloatFromCurve(
     FloatCurve=DataInterface() {propertyValues {Curve {keys [
         {arriveTangent=1.9056, time=0, value=0.2}
         {arriveTangent=0.0259, interpMode=RCIM_Cubic, leaveTangent=0.0259, tangentMode=RCTM_User, time=1, value=1.4}
        ]}}}
     Scale Curve=&RandomRangeFloat(Minimum=0.5, Maximum=1)
    )
    Index 2 Param 3=&FloatFromCurve(FloatCurve=DataInterface() {propertyValues {Curve {keys [
         {interpMode=RCIM_Cubic, leaveTangent=2.3064, tangentMode=RCTM_Break, time=0, value=1}
         {interpMode=RCIM_Cubic, leaveTangent=11.1233, tangentMode=RCTM_Break, time=1, value=2}
        ]}}})
    Index 2 Param 4=&FloatFromCurve(
     FloatCurve=DataInterface() {propertyValues {Curve {keys [
         {arriveTangent=-0.6867, interpMode=RCIM_Cubic, leaveTangent=-0.6867, tangentMode=RCTM_User, time=0, value=1}
         {interpMode=RCIM_Cubic, leaveTangent=-1.2956, tangentMode=RCTM_Break, time=1, value=0.75}
        ]}}}
     Scale Curve=1.6
    )
    Module.Param0WriteEnabled=true
    Module.Param0WriteEnabled1=true
    Module.Param0WriteEnabled2=true
    Module.Param1WriteEnabled=false
    Module.Param1WriteEnabled1=true
    Module.Param1WriteEnabled2=true
    Module.Param2WriteEnabled=true
    Module.Param2WriteEnabled1=true
    Module.Param2WriteEnabled2=true
    Module.Param3WriteEnabled=true
    Module.Param3WriteEnabled1=true
    Module.Param3WriteEnabled2=true
    Write Parameter Index 1=true
    Write Parameter Index 2=true
   )
  ]
  rendererClasses [&NiagaraSpriteRendererProperties]
  renderers [
   {rendererClass=&NiagaraSpriteRendererProperties}
   {rendererClass=&NiagaraSpriteRendererProperties, rendererIndex=1}
   {rendererClass=&NiagaraSpriteRendererProperties, rendererIndex=2}
   {rendererClass=&NiagaraSpriteRendererProperties, rendererIndex=3}
  ]
 }
 Flares_Merged {
  emitterUpdateScript [
   &EmitterState(
    Enable Distance Culling=true
    Max Distance Response=::KillAfterParticlesFinish
    MaxDistance=25000
    Module.bUseMaxDistance=true
    Module.bUseMinDistance=false
    Scalability Mode=::NewEnumerator1
   )
   &SpawnBurst_Instantaneous(Spawn Count=4, Use Spawn Probability=false)
  ]
  particleSpawnScript [
   &PartitionParticles(Number Of Partitions=4)
   /SetVariables(Particles.VisibilityTag=$Particles.PartitionParticles.Partition) {isSetParameters=true}
   &InitializeParticle(
    Lifetime=10
    Module.Write Color=true
    Module.Write Lifetime=true
    Module.Write Mass=true
    Module.Write Material Random=true
    Module.Write Position=false
    Module.Write Scale=false
    Module.Write SpriteRotation=false
    Module.Write SpriteSize=true
    Sprite Size=<1000 50>
   )
   &Color(Color=&MakeLinearColorFromVectorAndFloat("Vector (RGB)"=<2 0.1 0.2>, "Float (Alpha)"=&ClampFloat(Float=1)))
   &CameraOffset(Camera Offset Amount=&SelectFloatFromArray(Direct Array Index=$Particles.PartitionParticles.Partition, Float Selection Array=DataInterface() {propertyValues {FloatData [100,100,300]}}))
  ]
  particleUpdateScript [
   &ParticleState()
   &ScaleColor(
    Scale RGB=&VectorFromFloat(Value=&FloatFromCurve(FloatCurve=DataInterface() {propertyValues {Curve {keys [<at 0 1 RCIM_Cubic>, {interpMode=RCIM_Cubic, leaveTangent=12.6773, time=1, value=5}]}}}))
    ScaleA=true
    ScaleRGB=true
   )
   &DynamicMaterialParameters(
    Index 0 Param 1=&FloatFromCurve(FloatCurve=DataInterface() {propertyValues {Curve {keys [<at 0.4 0.15>,<at 0.8 0.7>]}}})
    Index 0 Param 2=1
    Index 0 Param 3=1
    Index 0 Param 4=1
    Module.Param0WriteEnabled=true
    Module.Param1WriteEnabled=true
    Module.Param2WriteEnabled=true
    Module.Param3WriteEnabled=true
   )
   &ScaleSpriteSize(Scale Factor=&Vector2DFromCurve(
     Vector2Curve=DataInterface() {propertyValues {
       XCurve {keys [<at 0 0 RCIM_Cubic>,<at 0.6 1 RCIM_Cubic>]}
       YCurve {keys [{arriveTangent=5.9554, interpMode=RCIM_Cubic, time=0, value=0}, <at 0.6 1 RCIM_Cubic>]}
      }}
     ScaleCurve=<1 1>
    ))
   &ScaleSpriteSize(Scale Factor=&Vector2DFromFloat(Value=&FloatFromCurve(FloatCurve=DataInterface() {propertyValues {Curve {keys [<at 0 1>,<at 15000 1.4>]}}}, CurveIndex=$Engine.Owner.LODDistance))) {module=ScaleSpriteSize001}
   /SetVariables(
    Particles.Color2=&MakeLinearColorFromVectorAndFloat(
     "Vector (RGB)"=<1.1458 2.1149 4>
     "Float (Alpha)"=&FloatFromCurve(FloatCurve=DataInterface() {propertyValues {Curve {keys [<at 0 0 RCIM_Cubic>, {arriveTangent=1, interpMode=RCIM_Cubic, leaveTangent=1, time=0.2355, value=0.501}, <at 1 1>]}}})
    )
    Particles.Color3:LinearColor=&MultiplyLinearColor(
     Linear Color=&MakeLinearColorFromVectorAndFloat(
      "Vector (RGB)"=<0.5 0.5 0.5>
      "Float (Alpha)"=&FloatFromCurve(
       FloatCurve=DataInterface() {propertyValues {Curve {keys [
           {arriveTangent=-11.2994, interpMode=RCIM_Cubic, time=0.2063, value=0}
           {arriveTangent=1.4274, interpMode=RCIM_Cubic, leaveTangent=1.4274, time=0.2352, value=0.15}
           {arriveTangent=6.3729, interpMode=RCIM_Cubic, leaveTangent=6.3729, time=0.2413, value=0.05}
           {arriveTangent=-6.0039, interpMode=RCIM_Cubic, leaveTangent=-6.0039, time=0.2509, value=0.25}
           {arriveTangent=5.2658, interpMode=RCIM_Cubic, leaveTangent=5.2658, time=0.2579, value=-0.05}
           {arriveTangent=1.8255, interpMode=RCIM_Cubic, leaveTangent=1.8255, time=0.2698, value=0.35}
           {arriveTangent=7.9737, interpMode=RCIM_Cubic, leaveTangent=7.9737, time=0.2853, value=0}
           {arriveTangent=7.8649, interpMode=RCIM_Cubic, leaveTangent=7.8649, time=0.2949, value=0.55}
           {arriveTangent=-11.5888, interpMode=RCIM_Cubic, leaveTangent=-11.5888, time=0.3044, value=0.15}
           {arriveTangent=1.6953, interpMode=RCIM_Cubic, leaveTangent=1.6953, time=0.3122, value=0.35}
           {arriveTangent=7.6817, interpMode=RCIM_Cubic, leaveTangent=7.6817, time=0.3339, value=0.2}
           <at 0.3382 0.55 RCIM_Cubic>
           {arriveTangent=2.4481, interpMode=RCIM_Cubic, leaveTangent=2.4481, time=0.3511, value=0.2}
           {arriveTangent=2.2096, interpMode=RCIM_Cubic, leaveTangent=2.2096, time=0.3587, value=0.6}
           {arriveTangent=11.0799, interpMode=RCIM_Cubic, leaveTangent=11.0799, time=0.3738, value=0.25}
           {arriveTangent=4.3466, interpMode=RCIM_Cubic, leaveTangent=4.3466, time=0.3812, value=0.85}
           {arriveTangent=-9.6875, interpMode=RCIM_Cubic, leaveTangent=-9.6875, time=0.3968, value=0.35}
           {arriveTangent=19.776, interpMode=RCIM_Cubic, leaveTangent=19.776, time=0.4225, value=0.45}
           <at 0.4246 0.9 RCIM_Cubic>
           {arriveTangent=-14.7338, interpMode=RCIM_Cubic, leaveTangent=-14.7338, time=0.436, value=0.45}
           {arriveTangent=33.3536, interpMode=RCIM_Cubic, leaveTangent=33.3536, time=0.4517, value=0.5}
           <at 0.454 1.05 RCIM_Cubic>
           {arriveTangent=-65.0091, interpMode=RCIM_Cubic, leaveTangent=-65.0091, time=0.4566, value=0.5}
           <at 0.4632 0.45 RCIM_Cubic>
           {arriveTangent=65.2447, interpMode=RCIM_Cubic, leaveTangent=65.2447, time=0.4699, value=0.5}
           {arriveTangent=3.946, interpMode=RCIM_Cubic, leaveTangent=3.946, time=0.4732, value=1.1}
           {arriveTangent=-68.3442, interpMode=RCIM_Cubic, leaveTangent=-68.3442, time=0.4825, value=0.55}
           <at 0.4893 0>
          ]}}}
       Scale Curve=0.25
      )
     )
     Scale Factor=<0.3323 1.6388 2 1>
    )
    Particles.DyMat20:Vector4f=&MakeVector4(
     X=&FloatFromCurve(FloatCurve=DataInterface() {propertyValues {Curve {keys [<at 0 1.5 RCIM_Cubic>, {interpMode=RCIM_Cubic, leaveTangent=-2.4626, time=1, value=0.3}]}}})
     Y=0.4
     Z=1
     W=1
    )
    Particles.SpriteSize1:Vector2f=&Multiply_Vector2D(
     A=<700 220>
     B=&Vector2DFromFloat(Value=&FloatFromCurve(FloatCurve=DataInterface() {propertyValues {Curve {keys [<at 0 1>,<at 15000 1.4>]}}}, CurveIndex=$Engine.Owner.LODDistance))
    )
    Particles.SpriteSize2=&Multiply_Vector2D(
     A=<500 500>
     B=&Multiply_Vector2D(
      A=&Vector2DFromFloat(Value=&FloatFromCurve(FloatCurve=DataInterface() {propertyValues {Curve {keys [<at 0 1 RCIM_Cubic>, {interpMode=RCIM_Cubic, leaveTangent=-2.0936, time=1, value=0}]}}}))
      B=&Vector2DFromFloat(Value=&FloatFromCurve(FloatCurve=DataInterface() {propertyValues {Curve {keys [<at 0 1>,<at 10000 2.25>]}}}, CurveIndex=$Engine.Owner.LODDistance))
     )
    )
    Particles.SpriteSize3:Vector2f=&Multiply_Vector2D(A=<800 800>, B=<1 1>)
   ) {isSetParameters=true}
  ]
  rendererClasses [&NiagaraSpriteRendererProperties]
  renderers [
   {rendererClass=&NiagaraSpriteRendererProperties}
   {rendererClass=&NiagaraSpriteRendererProperties, rendererIndex=1}
   {rendererClass=&NiagaraSpriteRendererProperties, rendererIndex=2}
   {rendererClass=&NiagaraSpriteRendererProperties, rendererIndex=3}
  ]
 }
 FloatingFrags {
  emitterUpdateScript [
   &EmitterState(
    Enable Distance Culling=true
    Max Distance Response=::KillAfterParticlesFinish
    MaxDistance=25000
    Module.bUseMaxDistance=true
    Module.bUseMinDistance=false
    Scalability Mode=::NewEnumerator1
   )
   &SpawnBurst_Instantaneous(Spawn Count=1, Spawn Time=0.25, Use Spawn Probability=false)
   &SpawnBurst_Instantaneous(Spawn Count=1, Spawn Time=0.7, Use Spawn Probability=false) {module=SpawnBurst_Instantaneous001}
   &SpawnBurst_Instantaneous(Spawn Count=1, Spawn Time=1.4, Use Spawn Probability=false) {module=SpawnBurst_Instantaneous002}
   &SpawnBurst_Instantaneous(Spawn Count=1, Spawn Group=1, Use Spawn Probability=false) {module=SpawnBurst_Instantaneous003}
   &SpawnBurst_Instantaneous(Spawn Count=1, Spawn Group=2, Spawn Time=0.12, Use Spawn Probability=false) {module=SpawnBurst_Instantaneous004}
  ]
  particleSpawnScript [
   /SetVariables(Particles.VisibilityTag=$Emitter.SpawnGroup) {isSetParameters=true}
   /SetVariables(Particles.MeshOrientation=&MakeQuatFromAxisAngle(
     Axis=&MaskVectorBySpawnGroup(Masked Vector=&RandomRangeVector(Minimum=<0 0 0>, Maximum=<1 1 1>), Passthrough Vector=<1 1 1>)
     AngleInDegrees=&MaskFloatBySpawnGroup(
      Masked Float=&RandomRangeFloat(Minimum=0, Maximum=180)
      Passthrough Float=&RandomRangeFloat(Minimum=-100, Maximum=150)
     )
    )) {isSetParameters=true}
   /SetVariables(Particles.SpawnRotationVector=&MaskVectorBySpawnGroup(
     Masked Vector=&RandomRangeVector(Minimum=<-0.01 -0.01 -0.01>, Maximum=<0.01 0.01 0.01>)
     Passthrough Vector=<0 0 0>
    )) {isSetParameters=true}
   &InitializeParticle(
    Lifetime=&FloatFromCurve(FloatCurve=DataInterface() {propertyValues {Curve {keys [<at 0 10>,<at 2 6>]}}}, CurveIndex=$System.Age)
    Mesh Scale=&MaskVectorBySpawnGroup(
     Masked Vector=&VectorFromFloat(Value=&RandomRangeFloat(Minimum=16, Maximum=20))
     Passthrough Vector=&VectorFromFloat(Value=&RandomRangeFloat(Minimum=6, Maximum=7))
    )
    Module.Write Color=true
    Module.Write Lifetime=true
    Module.Write Mass=true
    Module.Write Material Random=true
    Module.Write Position=false
    Module.Write Scale=true
    Module.Write SpriteRotation=false
    Module.Write SpriteSize=false
   )
   &Color(Color=&MaskLinearColorBySpawnGroup(
     Masked Linear Color=&MakeLinearColorFromVectorAndFloat("Vector (RGB)"=<1 0.1 0.1>, "Float (Alpha)"=&ClampFloat(Float=0.5))
     Passthrough Linear Color=&MakeLinearColorFromVectorAndFloat("Vector (RGB)"=<1.5 1.5 1.5>, "Float (Alpha)"=&ClampFloat(Float=1))
    ))
   &CameraOffset(Camera Offset Amount=100)
  ]
  particleUpdateScript [
   &ParticleState()
   &DynamicMaterialParameters(
    Index 0 Param 1=1
    Index 0 Param 2=1
    Index 0 Param 3=1
    Index 0 Param 4=&FloatFromCurve(FloatCurve=DataInterface() {propertyValues {Curve {keys [
         <at 0 -1.5 RCIM_Cubic>
         {arriveTangent=2.8362, interpMode=RCIM_Cubic, leaveTangent=2.8362, time=0.0658, value=-1.18}
         {arriveTangent=2.8159, interpMode=RCIM_Cubic, leaveTangent=2.8159, time=0.2547, value=-0.7776}
         {arriveTangent=2.7086, interpMode=RCIM_Cubic, leaveTangent=2.7086, time=0.2776, value=-0.5837}
         {arriveTangent=2.368, interpMode=RCIM_Cubic, leaveTangent=2.368, time=0.4377, value=-0.2819}
         {arriveTangent=1.7079, interpMode=RCIM_Cubic, leaveTangent=1.7079, time=0.4731, value=-0.1206}
         {arriveTangent=1.9231, interpMode=RCIM_Cubic, leaveTangent=1.9231, time=0.8103, value=0.3545}
         {arriveTangent=3.4034, interpMode=RCIM_Cubic, leaveTangent=3.4034, time=0.8496, value=0.6032}
         <at 1 1>
        ]}}})
    Module.Param0WriteEnabled=true
    Module.Param1WriteEnabled=true
    Module.Param2WriteEnabled=true
    Module.Param3WriteEnabled=true
   )
   &ScaleMeshSize(Scale Factor=&MaskVectorBySpawnGroup(
     Masked Vector=&VectorFromFloat(Value=&FloatFromCurve(FloatCurve=DataInterface() {propertyValues {Curve {keys [
           {arriveTangent=-0.0466, time=0, value=0.6}
           {arriveTangent=0.0413, interpMode=RCIM_Cubic, leaveTangent=0.0413, time=0.2603, value=0.5534}
           {arriveTangent=0.7472, interpMode=RCIM_Cubic, leaveTangent=0.7472, time=0.2832, value=0.6117}
           {arriveTangent=0.673, interpMode=RCIM_Cubic, leaveTangent=0.673, time=0.4449, value=0.6913}
           {arriveTangent=0.4479, interpMode=RCIM_Cubic, leaveTangent=0.4479, time=0.4705, value=0.7378}
           {arriveTangent=0.4597, interpMode=RCIM_Cubic, leaveTangent=0.4597, time=0.8146, value=0.8569}
           {arriveTangent=0.7716, interpMode=RCIM_Cubic, leaveTangent=0.7716, time=0.8441, value=0.9095}
           <at 1 1 RCIM_Cubic>
          ]}}}))
     Passthrough Vector=<1 1 1>
    ))
   &ScaleColor(
    Scale Alpha=&MaskFloatBySpawnGroup(
     Masked Float=&FloatFromCurve(FloatCurve=DataInterface() {propertyValues {Curve {keys [<at 0 0 RCIM_Cubic>, {arriveTangent=2, interpMode=RCIM_Cubic, leaveTangent=2, time=0.1598, value=0.3945}, <at 0.5 1>]}}})
     Passthrough Float=1
    )
    Scale RGB=&MaskVectorBySpawnGroup(
     Masked Vector=&VectorFromCurve(VectorCurve=DataInterface() {propertyValues {
        XCurve {keys [<at 0.5 2 RCIM_Cubic>,<at 1 6>]}
        YCurve {keys [<at 0.5 2 RCIM_Cubic>,<at 1 4.75>]}
        ZCurve {keys [<at 0.5 2 RCIM_Cubic>,<at 1 4>]}
       }})
     Passthrough Vector=<1 1 1>
    )
    ScaleA=true
    ScaleRGB=true
   )
   &UpdateMeshOrientation(
    Rotation Rate=&MaskFloatBySpawnGroup(Masked Float=1, Passthrough Float=0)
    Rotation Vector=$Particles.SpawnRotationVector
   ) {module=UpdateMeshOrientation001}
   &ScaleMeshSize(Scale Factor=&VectorFromFloat(Value=&FloatFromCurve(FloatCurve=DataInterface() {propertyValues {Curve {keys [<at 0 1>,<at 15000 2.5>]}}}, CurveIndex=$Engine.Owner.LODDistance))) {module=ScaleMeshSize001}
   /SetVariables(Particles.DyMat00:Vector4f=&MakeVector4(
     X=&FloatFromCurve(FloatCurve=DataInterface() {propertyValues {Curve {keys [
          {arriveTangent=0.249, time=0, value=0}
          {arriveTangent=0.0112, leaveTangent=0.0112, time=0.0429, value=0.249}
          {arriveTangent=0.015, leaveTangent=0.015, time=0.0478, value=0.2602}
          {arriveTangent=0.0229, leaveTangent=0.0229, time=0.161, value=0.2752}
          {arriveTangent=0.0047, leaveTangent=0.0047, time=0.1647, value=0.298}
          {arriveTangent=0.0502, leaveTangent=0.0502, time=0.2839, value=0.3027}
          {arriveTangent=-0.0006, leaveTangent=-0.0006, time=0.2998, value=0.3529}
          {arriveTangent=0.1459, leaveTangent=0.1459, time=0.4532, value=0.3524}
          {arriveTangent=0.0682, leaveTangent=0.0682, time=0.4728, value=0.4983}
          {arriveTangent=0.219, leaveTangent=0.219, time=0.6953, value=0.5665}
          {arriveTangent=2.1176, interpMode=RCIM_Cubic, leaveTangent=2.1176, time=0.7211, value=0.7854}
          {arriveTangent=-2.8157, interpMode=RCIM_Cubic, leaveTangent=-2.8157, time=0.9, value=1}
          <at 1 0 RCIM_Cubic>
         ]}}})
     Y=&FloatFromCurve(FloatCurve=DataInterface() {propertyValues {Curve {keys [
          <at 0.4 0.4 RCIM_Cubic>
          {arriveTangent=-0.5006, interpMode=RCIM_Cubic, leaveTangent=-0.5006, time=0.95, value=0.7}
          <at 0.9993 0.1 RCIM_Cubic>
         ]}}})
     Z=1
     W=&FloatFromCurve(FloatCurve=DataInterface() {propertyValues {Curve {keys [<at 0 1 RCIM_Cubic>,<at 1 1>]}}})
    )) {isSetParameters=true}
  ]
  rendererClasses [&NiagaraMeshRendererProperties]
  renderers [
   {rendererClass=&NiagaraMeshRendererProperties, rendererIndex=0}
   {rendererClass=&NiagaraMeshRendererProperties, rendererIndex=1}
  ]
 }
 Light002 {
  emitterUpdateScript [
   &EmitterState(Life Cycle Mode=::NewEnumerator1, Loop Behavior=::NewEnumerator1, Loop Delay=10, Loop Duration=11, UseLoopDelay=false)
   &SpawnBurst_Instantaneous(Spawn Count=1, Use Spawn Probability=false)
  ]
  particleSpawnScript [
   &InitializeParticle(
    Color=<rgba 1 0.1015 0.1072 1>
    Lifetime=10
    Module.Write Color=true
    Module.Write Lifetime=true
    Module.Write Mass=true
    Module.Write Material Random=true
    Module.Write Position=false
    Module.Write Scale=false
    Module.Write SpriteRotation=false
    Module.Write SpriteSize=true
    Sprite Size=<500 500>
   )
   /SetVariables(Particles.LightBrightness=1500) {isSetParameters=true}
  ]
  particleUpdateScript [
   /SetVariables(Particles.LightColorInit:LinearColor=&MultiplyLinearColor(
     Linear Color=&LerpLinearColors(StartColor=<rgba 1 4 8 1>, EndColor=<rgba 10 15 30 1>, LerpFactor=$User.Open)
     Scale Factor=&MakeVector4_FromFloat(Float=&FloatFromCurve(FloatCurve=DataInterface() {propertyValues {Curve {keys [
           <at 0 0 RCIM_Cubic>
           {arriveTangent=1.3333, leaveTangent=1.3333, time=0.05, value=1}
           {arriveTangent=-0.029, interpMode=RCIM_Cubic, leaveTangent=-75.0977, tangentMode=RCTM_Break, time=0.975, value=1}
           <at 1 0>
          ]}}}))
    )) {isSetParameters=true}
   /SetVariables(Particles.LightColor=&MultiplyLinearColor(
     Linear Color=$Particles.LightColorInit
     Scale Factor=&MakeVector4_FromLinearColor(LinearColor=$Particles.Color)
    )) {isSetParameters=true}
   /SetVariables(Particles.LightExponent=12) {isSetParameters=true}
   /SetVariables(Particles.LightRadius=500) {isSetParameters=true}
   &ParticleState()
   &ScaleColor(
    Scale Alpha=&FloatFromCurve(FloatCurve=DataInterface() {propertyValues {Curve {keys [
         <at 0 0 RCIM_Cubic>
         {arriveTangent=1.3333, leaveTangent=1.3333, time=0.01, value=1}
         {arriveTangent=-0.029, interpMode=RCIM_Cubic, leaveTangent=-75.0977, tangentMode=RCTM_Break, time=0.975, value=1}
         <at 1 0>
        ]}}})
    ScaleA=true
    ScaleRGB=true
   )
   /SetVariables(Particles.LightIntensityRamp:NiagaraFloat=&FloatFromCurve(
     FloatCurve=DataInterface() {propertyValues {Curve {keys [
         <at 0 0 RCIM_Cubic>
         {arriveTangent=4.1026, interpMode=RCIM_Cubic, leaveTangent=0.6001, tangentMode=RCTM_Break, time=0.05, value=0.5}
         {arriveTangent=11.2956, interpMode=RCIM_Cubic, leaveTangent=-75.0977, tangentMode=RCTM_Break, time=0.985, value=4}
         <at 1 0>
        ]}}}
     Scale Curve=&Lerp_Float(A=$Particles.LightBrightness, B=&Multiply_Float(A=$Particles.LightBrightness, B=5), Alpha=$User.Open)
    )) {isSetParameters=true}
  ]
  rendererClasses [
   &NiagaraSpriteRendererProperties
   &NiagaraLightRendererProperties
   &NiagaraComponentRendererProperties
  ]
  renderers [
   {rendererClass=&NiagaraSpriteRendererProperties}
   {rendererClass=&NiagaraLightRendererProperties, rendererIndex=1}
   {rendererClass=&NiagaraComponentRendererProperties, rendererIndex=2}
  ]
 }
 ShockWave {
  emitterUpdateScript [
   &EmitterState(Life Cycle Mode=::NewEnumerator1, Loop Behavior=::NewEnumerator1, Loop Delay=10, Loop Duration=1, UseLoopDelay=true)
   &SpawnBurst_Instantaneous(Spawn Count=1, Use Spawn Probability=false)
  ]
  particleSpawnScript [
   &InitializeParticle(
    Lifetime=0.3
    Mesh Scale=<1.5 1.5 1.5>
    Module.Write Color=true
    Module.Write Lifetime=true
    Module.Write Mass=true
    Module.Write Material Random=true
    Module.Write Position=false
    Module.Write Scale=true
    Module.Write SpriteRotation=false
    Module.Write SpriteSize=false
   )
   &Color(Color=&MakeLinearColorFromVectorAndFloat("Vector (RGB)"=<1 1 1>, "Float (Alpha)"=&ClampFloat(Float=0.1)))
  ]
  particleUpdateScript [
   &ParticleState()
   &ScaleColor(
    Scale Alpha=&FloatFromCurve(FloatCurve=DataInterface() {propertyValues {Curve {keys [{arriveTangent=-2.1694, interpMode=RCIM_Cubic, time=0, value=0}, <at 0.4 1 RCIM_Cubic>, <at 1 0 RCIM_Cubic>]}}})
    ScaleA=true
    ScaleRGB=true
   )
   &ScaleMeshSize(Scale Factor=&VectorFromFloat(Value=&FloatFromCurve(FloatCurve=DataInterface() {propertyValues {Curve {keys [<at 0 1 RCIM_Cubic>,<at 1 0>]}}})))
   &CameraOffset(Camera Offset Amount=400)
  ]
  rendererClasses [&NiagaraMeshRendererProperties]
  renderers [{rendererClass=&NiagaraMeshRendererProperties, rendererIndex=0}]
 }
 TinyFrags {
  emitterUpdateScript [
   &EmitterState(
    Enable Distance Culling=true
    Life Cycle Mode=::NewEnumerator1
    Loop Behavior=::NewEnumerator1
    Loop Duration=8
    MaxDistance=3500
    Module.bUseMaxDistance=true
    Module.bUseMinDistance=false
    Scalability Mode=::NewEnumerator1
    UseLoopDelay=true
   )
   &SpawnRate(SpawnRate=2, Use Spawn Probability=false)
   &SpawnBurst_Instantaneous(Spawn Count=&RandomRangeInt(Minimum=1, Maximum=2), Use Spawn Probability=false)
   &SpawnBurst_Instantaneous(Spawn Count=&RandomRangeInt(Minimum=0, Maximum=2), Spawn Time=0.25, Use Spawn Probability=false) {module=SpawnBurst_Instantaneous001}
   &SpawnBurst_Instantaneous(Spawn Count=&RandomRangeInt(Minimum=0, Maximum=2), Spawn Time=1.4, Use Spawn Probability=false) {module=SpawnBurst_Instantaneous002}
  ]
  particleSpawnScript [
   /SetVariables(Particles.MeshOrientation=&MakeQuatFromAxisAngle(
     Axis=&RandomRangeVector(Minimum=<0 0 0>, Maximum=<1 1 1>, Evaluation Type=::NewEnumerator1)
     AngleInDegrees=90
    )) {isSetParameters=true}
   /SetVariables(Particles.SpawnRotationVector=&RandomRangeVector(Minimum=<-1 -1 -1>, Maximum=<1 1 1>, Evaluation Type=::NewEnumerator1)) {isSetParameters=true}
   &InitializeParticle(
    Lifetime=&RandomRangeFloat(Minimum=2, Maximum=3, Evaluation Type=::NewEnumerator1)
    Mesh Scale=&RandomRangeVector(Minimum=<1 1 1.5>, Maximum=<2 2 1.5>, Evaluation Type=::NewEnumerator1)
    Module.Write Color=true
    Module.Write Lifetime=true
    Module.Write Mass=true
    Module.Write Material Random=true
    Module.Write Position=false
    Module.Write Scale=true
    Module.Write SpriteRotation=false
    Module.Write SpriteSize=false
   )
   &Color(Color=&MakeLinearColorFromVectorAndFloat("Vector (RGB)"=<10 1 1>, "Float (Alpha)"=&ClampFloat(Float=1)))
   &AddVelocity(Velocity=&RandomRangeVector(Minimum=<-8 -8 -8>, Maximum=<8 8 8>, Evaluation Type=::NewEnumerator1))
   &SphereLocation(
    MaskSpawnGroup=false
    Module.Override Randomness=false
    Sphere Radius=&RandomRangeFloat(Minimum=70, Maximum=90, Evaluation Type=::NewEnumerator1)
    Surface Only=false
   )
   &AddVelocity(Velocity=&Multiply_VectorByFloat(
     Vector=$Output.SphereLocation.SphereVector
     Float=&RandomRangeFloat(Minimum=-0.5, Maximum=0, Evaluation Type=::NewEnumerator1)
    )) {module=AddVelocity001}
  ]
  particleUpdateScript [
   &ParticleState()
   &Drag(Drag=0.2, Use Linear Drag=true, Use Rotational Drag=false)
   &SolveForcesAndVelocity(Clamp Velocity=false, Limit Acceleration=false)
   &ScaleColor(
    Scale Alpha=&FloatFromCurve(FloatCurve=DataInterface() {propertyValues {Curve {keys [
         <at 0 0 RCIM_Cubic>
         {arriveTangent=1.4286, interpMode=RCIM_Cubic, leaveTangent=1.4286, time=0.2, value=1}
         {arriveTangent=-1.25, interpMode=RCIM_Cubic, leaveTangent=-1.25, time=0.7, value=1}
         <at 1 0>
        ]}}})
    ScaleA=true
    ScaleRGB=true
   )
   &UpdateMeshOrientation(Rotation Vector=$Particles.SpawnRotationVector) {module=UpdateMeshOrientation001}
   &ScaleMeshSize(Scale Factor=&VectorFromFloat(Value=&FloatFromCurve(FloatCurve=DataInterface() {propertyValues {Curve {keys [<at 0 0 RCIM_Cubic>, <at 0.2 1>, {arriveTangent=-1.25, interpMode=RCIM_Cubic, leaveTangent=-1.25, time=0.8, value=1}, <at 1 0>]}}})))
   /Niagara/Modules/CascadeConversion/CascadeConversion_Orbit(
    Offset:Vector3f=&RandomRangeVector(Minimum=<100 -100 100>, Maximum=<100 100 100>)
    Rotation Rate:Vector3f=&RandomRangeVector(Minimum=<-0.2 -0.2 -0.2>, Maximum=<0.2 0.2 0.2>)
   )
   /Niagara/Modules/CascadeConversion/CascadeConversion_SolveOrbit()
  ]
  rendererClasses [&NiagaraMeshRendererProperties]
  renderers [{rendererClass=&NiagaraMeshRendererProperties, rendererIndex=0}]
 }
}
summary {
 systemName=NS_Rift_Portal_Closing_Red
 emitters [
  {bEnabled=false, emitterName=CrackSpline, rendererClasses [&NiagaraMeshRendererProperties]}
  {emitterName=FloatingFrags, rendererClasses [&NiagaraMeshRendererProperties]}
  {emitterName=ElectricArcsResidual, rendererClasses [&NiagaraMeshRendererProperties]}
  {emitterName=TinyFrags, rendererClasses [&NiagaraMeshRendererProperties]}
  {bEnabled=false, emitterName=WideFlare, rendererClasses [&NiagaraSpriteRendererProperties]}
  {bEnabled=false, emitterName=SharpFlare, rendererClasses [&NiagaraSpriteRendererProperties]}
  {
   bEnabled=false
   emitterName=Light
   rendererClasses [&NiagaraSpriteRendererProperties, &NiagaraLightRendererProperties]
  }
  {emitterName=Embers, simTarget=GPUComputeSim, rendererClasses [&NiagaraSpriteRendererProperties]}
  {bEnabled=false, emitterName=Corona, rendererClasses [&NiagaraSpriteRendererProperties]}
  {bEnabled=false, emitterName=RoundFlareEnd, rendererClasses [&NiagaraSpriteRendererProperties]}
  {bEnabled=false, emitterName=WideFlareEnd, rendererClasses [&NiagaraSpriteRendererProperties]}
  {emitterName=ShockWave, rendererClasses [&NiagaraMeshRendererProperties]}
  {bEnabled=false, emitterName=SmokeInhale, rendererClasses [&NiagaraSpriteRendererProperties]}
  {
   emitterName=Light002
   rendererClasses [
    &NiagaraSpriteRendererProperties
    &NiagaraLightRendererProperties
    &NiagaraComponentRendererProperties
   ]
  }
  {emitterName=Flares_Merged, rendererClasses [&NiagaraSpriteRendererProperties]}
  {emitterName=End_Flares_Merged, rendererClasses [&NiagaraSpriteRendererProperties]}
  {emitterName=EmbersCPU, rendererClasses [&NiagaraSpriteRendererProperties]}
  {bEnabled=false, emitterName=FloatingFrags001, rendererClasses [&NiagaraMeshRendererProperties]}
 ]
 userVariables [User.Open {type=NiagaraFloat, defaultValue=1, description=""}]
}
