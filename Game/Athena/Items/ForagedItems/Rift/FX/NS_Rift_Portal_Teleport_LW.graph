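# Niagara system for the Rift foraged item's portal teleport effect, judging by its name and path.
# It layers flares, corona, pickup beam, crack spline, residual electric arcs, shockwave, smoke and light emitters, alongside a disabled BigStreak sprite emitter.
# A user-exposed color array data interface is declared, but the file shows neither which emitters read it nor any renderers for the active emitters.
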NiagaraSystem /Game/Athena/Items/ForagedItems/Rift/FX/NS_Rift_Portal_Teleport_LW
dependencies=null
emitters {
 Corona {rendererClasses [], renderers []}
 CrackSpline {rendererClasses [], renderers []}
 ElectricArcsResidual {rendererClasses [], renderers []}
 Light {rendererClasses [], renderers []}
 PickupBeam {rendererClasses [], renderers []}
 RoundFlare {rendererClasses [], renderers []}
 ShockWave {rendererClasses [], renderers []}
 TransSmoke {rendererClasses [], renderers []}
 WideFlare {rendererClasses [], renderers []}
}
summary {
 systemName=NS_Rift_Portal_Teleport_LW
 emitters [
  {bEnabled=false, emitterName=BigStreak, rendererClasses [&NiagaraSpriteRendererProperties]}
  {emitterName=CrackSpline, rendererClasses []}
  {emitterName=WideFlare, rendererClasses []}
  {emitterName=RoundFlare, rendererClasses []}
  {emitterName=PickupBeam, rendererClasses []}
  {emitterName=Corona, rendererClasses []}
  {emitterName=Light, rendererClasses []}
  {emitterName=ElectricArcsResidual, rendererClasses []}
  {emitterName=ShockWave, rendererClasses []}
  {emitterName=TransSmoke, rendererClasses []}
 ]
 userVariables [User.Color Array {
   type=NiagaraDataInterfaceArrayColor
   description=""
   defaultValue {
    struct=NiagaraExt_VariableValue_DataInterface
    value {
     dataInterface=&self:NiagaraDataInterfaceArrayColor_0
     dataInterfaceClass=&NiagaraDataInterfaceArrayColor
    }
   }
  }]
}
